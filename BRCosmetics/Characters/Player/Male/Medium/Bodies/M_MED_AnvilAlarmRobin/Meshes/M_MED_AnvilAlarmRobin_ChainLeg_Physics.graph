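# Physics asset for the AnvilAlarmRobin medium male player body cosmetic, defining collision and joints for its dangling leg chains.
# Capsule bodies cover lower spine, both thighs, right calf and foot, plus chain bones.
# Front, back and short extra strands use twist-locked joints with swing limits; front and extra anchor to base bones.
# Front and back tips join through a free-rotation constraint, closing a loop; dyn_chain_back has no constraint to base_chain_back.
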PhysicsAsset /BRCosmetics/Characters/Player/Male/Medium/Bodies/M_MED_AnvilAlarmRobin/Meshes/M_MED_AnvilAlarmRobin_ChainLeg_Physics
bodies {
 base_chain_back [{shapeType=Capsule, length=0.505, radius=0.505, shapeName=base_chain_back_capsule, rotation=<rot 0 -90 0>}]
 base_chain_extra [{shapeType=Capsule, length=0.505, radius=0.505, shapeName=base_chain_extra_capsule, rotation=<rot 0 -90 0>}]
 base_chain_front [{shapeType=Capsule, length=0.505, radius=0.505, shapeName=base_chain_front_capsule, rotation=<rot 0 -90 0>}]
 calf_r [
  {
   shapeType=Capsule
   length=28.3426
   radius=8.9821
   shapeName=thigh_r_capsule
   center=<19.9807 0.8862 -0.5099>
   rotation=<rot 87.7355 180 -180>
  }
  {shapeType=Capsule, length=30.708, radius=6.6725, shapeName=calf_r_capsule, center=<16.7377 6.7415 -7.4875>, rotation=<rot 90 0 0>}
 ]
 dyn_chain_back [{shapeType=Capsule, length=3.1848, radius=0.8152, shapeName=dyn_chain_back_capsule, center=<2.6244 0 0.4262>, rotation=<rot 90 0 0>}]
 dyn_chain_back_01 [{shapeType=Capsule, length=4.7988, radius=2, shapeName=dyn_chain_back_01_capsule, center=<4.0596 0 0.5>, rotation=<rot 90 0 0>}]
 dyn_chain_back_02 [{shapeType=Capsule, length=3, radius=2, shapeName=dyn_chain_back_02_capsule, center=<2 0 0.5>, rotation=<rot 90 0 0>}]
 dyn_chain_back_03 [{shapeType=Capsule, length=3, radius=2, shapeName=dyn_chain_back_03_capsule, center=<2 0 0.5>, rotation=<rot 90 0 0>}]
 dyn_chain_back_04 [{shapeType=Capsule, length=3, radius=2, shapeName=dyn_chain_back_04_capsule, center=<2 0 0.5>, rotation=<rot 90 0 0>}]
 dyn_chain_back_05 [{shapeType=Capsule, length=6, radius=2, shapeName=dyn_chain_back_05_capsule, center=<2.8855 -0.498 -0.1058>, rotation=<rot 63.8732 0 6.4438>}]
 dyn_chain_extra_01 [{shapeType=Capsule, length=3, radius=1.2, shapeName=dyn_chain_extra_01_capsule, center=<3.6058 0.999 0>, rotation=<rot 90 0 0>}]
 dyn_chain_extra_02 [{shapeType=Capsule, length=3, radius=1, shapeName=dyn_chain_extra_02_capsule, center=<2.7374 0.549 0.3795>, rotation=<rot 90 4.6192 0>}]
 dyn_chain_front_01 [{shapeType=Capsule, length=3, radius=1.5, shapeName=dyn_chain_front_01_capsule, center=<2 -0.9749 -0.5>, rotation=<rot 90 0 0>}]
 dyn_chain_front_02 [{shapeType=Capsule, length=3, radius=1.5, shapeName=dyn_chain_front_02_capsule, center=<2 0 -0.5>, rotation=<rot 90 0 0>}]
 dyn_chain_front_03 [{shapeType=Capsule, length=3, radius=1.8, shapeName=dyn_chain_front_03_capsule, center=<2 0 -0.5>, rotation=<rot 90 0 0>}]
 dyn_chain_front_04 [{shapeType=Capsule, length=3, radius=2, shapeName=dyn_chain_front_04_capsule, center=<2 0 -0.5>, rotation=<rot 90 0 0>}]
 dyn_chain_front_05 [{shapeType=Capsule, length=6, radius=2, shapeName=dyn_chain_front_05_capsule, center=<4.2047 0 1.0281>, rotation=<rot 65 180 180>}]
 foot_r [{shapeType=Capsule, length=13.9064, radius=7.7974, shapeName=foot_r_capsule, center=<0.9158 -7.1022 0.3721>, rotation=<rot 0 4.5062 -89.9806>}]
 spine_01 [{shapeType=Capsule, length=9, radius=11, shapeName=spine_01_capsule}]
 spine_02 [{shapeType=Capsule, length=9, radius=11, shapeName=spine_02_capsule}]
 thigh_l [{
   shapeType=Capsule
   length=31.7677
   radius=7.6762
   shapeName=thigh_r_capsule
   center=<-21.0463 0.1696 2.531>
   rotation=<rot 0 -91.5117 -88.1966>
  }]
 thigh_r [{
   shapeType=Capsule
   length=30.0126
   radius=9.0689
   shapeName=thigh_r_capsule
   center=<21.0688 0.2399 -2.0428>
   rotation=<rot 0.0496 -94.1441 -88.924>
  }]
}
constraints [
 {
  bone1Name=dyn_chain_back_01
  bone2Name=dyn_chain_back
  swing1LimitDegrees=30
  swing2LimitDegrees=30
  twistLimitDegrees=45
  twistMotion=Locked
 }
 {
  bone1Name=dyn_chain_back_02
  bone2Name=dyn_chain_back_01
  swing1LimitDegrees=20
  swing2LimitDegrees=30
  twistLimitDegrees=45
  twistMotion=Locked
 }
 {
  bone1Name=dyn_chain_back_03
  bone2Name=dyn_chain_back_02
  swing1LimitDegrees=20
  swing2LimitDegrees=30
  twistLimitDegrees=45
  twistMotion=Locked
 }
 {
  bone1Name=dyn_chain_back_04
  bone2Name=dyn_chain_back_03
  swing1LimitDegrees=20
  swing2LimitDegrees=20
  twistLimitDegrees=45
  twistMotion=Locked
 }
 {
  bone1Name=dyn_chain_back_05
  bone2Name=dyn_chain_back_04
  swing1LimitDegrees=20
  swing2LimitDegrees=20
  twistLimitDegrees=45
  twistMotion=Locked
 }
 {
  bone1Name=dyn_chain_front_05
  bone2Name=dyn_chain_back_05
  swing1LimitDegrees=45
  swing1Motion=Free
  swing2LimitDegrees=45
  swing2Motion=Free
  twistLimitDegrees=45
  twistMotion=Free
 }
 {
  bone1Name=dyn_chain_front_01
  bone2Name=base_chain_front
  swing1LimitDegrees=20
  swing2LimitDegrees=30
  twistLimitDegrees=45
  twistMotion=Locked
 }
 {
  bone1Name=dyn_chain_front_02
  bone2Name=dyn_chain_front_01
  swing1LimitDegrees=20
  swing2LimitDegrees=30
  twistLimitDegrees=45
  twistMotion=Locked
 }
 {
  bone1Name=dyn_chain_front_03
  bone2Name=dyn_chain_front_02
  swing1LimitDegrees=20
  swing2LimitDegrees=30
  twistLimitDegrees=45
  twistMotion=Locked
 }
 {
  bone1Name=dyn_chain_front_04
  bone2Name=dyn_chain_front_03
  swing1LimitDegrees=20
  swing2LimitDegrees=20
  twistLimitDegrees=45
  twistMotion=Locked
 }
 {
  bone1Name=dyn_chain_front_05
  bone2Name=dyn_chain_front_04
  swing1LimitDegrees=20
  swing2LimitDegrees=20
  twistLimitDegrees=45
  twistMotion=Locked
 }
 {
  bone1Name=dyn_chain_extra_02
  bone2Name=dyn_chain_extra_01
  swing1LimitDegrees=35
  swing2LimitDegrees=20
  twistLimitDegrees=15
  twistMotion=Locked
 }
 {
  bone1Name=dyn_chain_extra_01
  bone2Name=base_chain_extra
  swing1LimitDegrees=35
  swing2LimitDegrees=35
  twistLimitDegrees=15
  twistMotion=Locked
 }
]
