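# Control Rig for the AnvilAlarmRobin male character body that procedurally corrects jacket, pants, collar and accessory bones.
# Weighted parent constraints attach chains, necklaces, buckle, laces, collar and jacket bones to body bones.
# Spherical pose readers on thighs and head drive additional constraints and corrective transform offsets.
# State variables such as crouching, sliding, emote and skydive gate or drive those offsets, many smoothed with alpha interpolation.
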ControlRigBlueprint /BRCosmetics/Characters/Player/Male/Medium/Bodies/M_MED_AnvilAlarmRobin/Meshes/M_MED_AnvilAlarmRobin_CtrlRig
graphs [
 Multiply_1_ContainedGraph {nodes [
   n1 = RigVMFunctionEntryNode()
   RigVMFunctionReturnNode(Result=&RigVMFunction_MathDoubleMul(B=@n1.C, A=&RigVMFunction_MathDoubleMul(A=@n1.A, B=@n1.B).Result).Result)
  ]}
 Multiply_1_ContainedGraph {nodes [
   n1 = RigVMFunctionEntryNode()
   RigVMFunctionReturnNode(Result=&RigVMFunction_MathDoubleMul(B=@n1.C, A=&RigVMFunction_MathDoubleMul(A=@n1.A, B=@n1.B).Result).Result)
  ]}
 Multiply_1_ContainedGraph {nodes [
   n1 = RigVMFunctionEntryNode()
   Multiply = &RigVMFunction_MathDoubleMul(A=@n1.A, B=@n1.B)
   RigVMFunctionReturnNode(Result=&RigVMFunction_MathDoubleMul(B=@n1.C, A=@Multiply.Result).Result)
  ]}
 Multiply_1_ContainedGraph {nodes [
   n1 = RigVMFunctionEntryNode()
   RigVMFunctionReturnNode(Result=&RigVMFunction_MathDoubleMul(B=@n1.C, A=&RigVMFunction_MathDoubleMul(A=@n1.A, B=@n1.B).Result).Result)
  ]}
 Multiply_1_ContainedGraph {nodes [
   n1 = RigVMFunctionEntryNode()
   RigVMFunctionReturnNode(Result=&RigVMFunction_MathDoubleMul(B=@n1.C, A=&RigVMFunction_MathDoubleMul(A=@n1.A, B=@n1.B).Result).Result)
  ]}
 Multiply_2_ContainedGraph {nodes [
   n1 = RigVMFunctionEntryNode()
   RigVMFunctionReturnNode(Result=&RigVMFunction_MathDoubleMul(B=@n1.C, A=&RigVMFunction_MathDoubleMul(A=@n1.A, B=@n1.B).Result).Result)
  ]}
 Multiply_2_ContainedGraph {nodes [
   n1 = RigVMFunctionEntryNode()
   RigVMFunctionReturnNode(Result=&RigVMFunction_MathDoubleMul(B=@n1.C, A=&RigVMFunction_MathDoubleMul(A=@n1.A, B=@n1.B).Result).Result)
  ]}
 Multiply_4_ContainedGraph {nodes [
   n1 = RigVMFunctionEntryNode()
   RigVMFunctionReturnNode(Result=&RigVMFunction_MathDoubleMul(B=@n1.C, A=&RigVMFunction_MathDoubleMul(A=@n1.A, B=@n1.B).Result).Result)
  ]}
 Multiply_4_ContainedGraph {nodes [
   n1 = RigVMFunctionEntryNode()
   RigVMFunctionReturnNode(Result=&RigVMFunction_MathDoubleMul(B=@n1.C, A=&RigVMFunction_MathDoubleMul(A=@n1.A, B=@n1.B).Result).Result)
  ]}
 Multiply_8_ContainedGraph {nodes [
   n1 = RigVMFunctionEntryNode()
   n2 = &RigVMFunction_MathDoubleMul(A=@n1.A, B=@n1.B)
   RigVMFunctionReturnNode(Result=&RigVMFunction_MathDoubleMul(B=@n1.D, A=&RigVMFunction_MathDoubleMul(B=@n1.C, A=@n2.Result).Result).Result)
  ]}
 Multiply_8_ContainedGraph {nodes [
   n1 = RigVMFunctionEntryNode()
   n2 = &RigVMFunction_MathDoubleMul(A=@n1.A, B=@n1.B)
   RigVMFunctionReturnNode(Result=&RigVMFunction_MathDoubleMul(B=@n1.D, A=&RigVMFunction_MathDoubleMul(B=@n1.C, A=@n2.Result).Result).Result)
  ]}
 New Function_ContainedGraph {nodes [
   RigVMFunctionEntryNode()
   > RigVMFunction_Sequence = RigVMAggregateNode()
   &RigUnit_ParentConstraint(
    AdvancedSettings=(RotationOrderForFilter=XZY)
    Child=(Type=Bone,Name="base_chain_front")
    Parents=((Item=(Type=Bone,Name="thigh_r"),Weight=0.1272),(Item=(Type=Bone,Name="deform_groin_r"),Weight=0.1313),(Item=(Type=Bone,Name="thigh_twist_01_r"),Weight=0.3381),(Item=(Type=Bone,Name="pelvis"),Weight=0.4035))
    Weight=1
    ExecutePin=@RigVMFunction_Sequence.A
   )
   > &RigUnit_ParentConstraint(
    AdvancedSettings=(RotationOrderForFilter=XZY)
    Child=(Type=Bone,Name="base_chain_back")
    Parents="((Item=(Type=Bone,Name=\"calf_r\"),Weight=0.0001),(Item=(Type=Bone,Name=\"deform_groin_r\"),Weight=0.0031),(Item=(Type=Bone,Name=\"thigh_r\"),Weight=0.0842),(Item=(Type=Bone,Name=\"thigh_twist_01_r\"),Weight=0.2611),(Item=(Type=Bone,Name=\"pelvis\"),Weight=0.3156),( ...(truncated)"
    Weight=1
   )
   > &RigUnit_ParentConstraint(
    AdvancedSettings=(RotationOrderForFilter=XZY)
    Child=(Type=Bone,Name="base_chain_extra")
    Parents=((Item=(Type=Bone,Name="thigh_r"),Weight=0.1242),(Item=(Type=Bone,Name="deform_groin_r"),Weight=0.161),(Item=(Type=Bone,Name="thigh_twist_01_r"),Weight=0.3053),(Item=(Type=Bone,Name="pelvis"),Weight=0.4095))
    Weight=1
   )
   > &RigUnit_ParentConstraint(
    AdvancedSettings=(RotationOrderForFilter=XZY)
    Child=(Type=Bone,Name="base_necklace_a")
    Parents="((Item=(Type=Bone,Name=\"clavicle_r\"),Weight=0.0164),(Item=(Type=Bone,Name=\"clavicle_l\"),Weight=0.0166),(Item=(Type=Bone,Name=\"spine_03\"),Weight=0.0254),(Item=(Type=Bone,Name=\"neck_01\"),Weight=0.0305),(Item=(Type=Bone,Name=\"neck_02\"),Weight=0.0695),(Item=(T ...(truncated)"
    Weight=1
   )
   > &RigUnit_ParentConstraint(
    AdvancedSettings=(RotationOrderForFilter=XZY)
    Child=(Type=Bone,Name="base_necklace_b")
    Parents="((Item=(Type=Bone,Name=\"clavicle_r\"),Weight=0.0164),(Item=(Type=Bone,Name=\"clavicle_l\"),Weight=0.0166),(Item=(Type=Bone,Name=\"spine_03\"),Weight=0.0254),(Item=(Type=Bone,Name=\"neck_01\"),Weight=0.0305),(Item=(Type=Bone,Name=\"neck_02\"),Weight=0.0695),(Item=(T ...(truncated)"
    Weight=1
   )
   > &RigUnit_ParentConstraint(
    AdvancedSettings=(RotationOrderForFilter=XZY)
    Child=(Type=Bone,Name="base_buckle")
    Parents="((Item=(Type=Bone,Name=\"thigh_twist_01_l\"),Weight=0.0073),(Item=(Type=Bone,Name=\"pants_corrective_02\"),Weight=0.0102),(Item=(Type=Bone,Name=\"thigh_twist_01_r\"),Weight=0.0119),(Item=(Type=Bone,Name=\"deform_groin_r\"),Weight=0.0405),(Item=(Type=Bone,Name=\"def ...(truncated)"
    Weight=1
   )
   > &RigUnit_ParentConstraint(
    AdvancedSettings=(RotationOrderForFilter=XZY)
    Child=(Type=Bone,Name="base_laces_l")
    Parents="((Item=(Type=Bone,Name=\"calf_r\"),Weight=0.0002),(Item=(Type=Bone,Name=\"dyn_boot_tongue_02\"),Weight=0.0003),(Item=(Type=Bone,Name=\"dyn_boot_tongue_01\"),Weight=0.0053),(Item=(Type=Bone,Name=\"foot_r\"),Weight=0.1208),(Item=(Type=Bone,Name=\"calf_twist_02_r\"),We ...(truncated)"
    Weight=1
   )
   > &RigUnit_ParentConstraint(
    AdvancedSettings=(RotationOrderForFilter=XZY)
    Child=(Type=Bone,Name="base_laces_r")
    Parents="((Item=(Type=Bone,Name=\"ball_r\"),Weight=0.0003),(Item=(Type=Bone,Name=\"dyn_boot_tongue_02\"),Weight=0.0004),(Item=(Type=Bone,Name=\"dyn_boot_tongue_01\"),Weight=0.0066),(Item=(Type=Bone,Name=\"foot_r\"),Weight=0.1337),(Item=(Type=Bone,Name=\"calf_twist_02_r\"),We ...(truncated)"
    Weight=1
   )
   > &RigUnit_ParentConstraint(
    AdvancedSettings=(RotationOrderForFilter=XZY)
    Child=(Type=Bone,Name="base_collar_fr_r")
    Parents="((Item=(Type=Bone,Name=\"upperarm_twist_01_r\"),Weight=0.0028),(Item=(Type=Bone,Name=\"neck_01\"),Weight=0.0099),(Item=(Type=Bone,Name=\"spine_03\"),Weight=0.0278),(Item=(Type=Bone,Name=\"neck_02\"),Weight=0.0295),(Item=(Type=Bone,Name=\"deform_pec_r\"),Weight=0.041 ...(truncated)"
    Weight=1
   )
   > &RigUnit_ParentConstraint(
    AdvancedSettings=(RotationOrderForFilter=XZY)
    Child=(Type=Bone,Name="base_collar_fr_l")
    Parents="((Item=(Type=Bone,Name=\"upperarm_twist_01_l\"),Weight=0.002),(Item=(Type=Bone,Name=\"neck_01\"),Weight=0.0102),(Item=(Type=Bone,Name=\"spine_03\"),Weight=0.0275),(Item=(Type=Bone,Name=\"neck_02\"),Weight=0.0301),(Item=(Type=Bone,Name=\"deform_pec_l\"),Weight=0.0407 ...(truncated)"
    Weight=1
   )
   RigVMFunctionReturnNode(ExecuteContext=@RigVMFunction_Sequence.C)
   &RigUnit_ParentConstraint(
    AdvancedSettings=(InterpolationType=Average,RotationOrderForFilter=XZY)
    Child=(Type=Bone,Name="MSTR_jacket_bk_l")
    Filter=(TranslationFilter=(bX=true,bY=true,bZ=true),RotationFilter=(bX=true,bY=true,bZ=true),ScaleFilter=(bX=True,bY=True,bZ=True))
    Parents=((Item=(Type=Bone,Name="spine_02"),Weight=1),(Item=(Type=Bone,Name="pelvis"),Weight=1))
    Weight=0.5
    bMaintainOffset=true
    ExecutePin=@RigVMFunction_Sequence.B
   )
   > &RigUnit_ParentConstraint(
    AdvancedSettings=(InterpolationType=Average,RotationOrderForFilter=XZY)
    Child=(Type=Bone,Name="MSTR_jacket_bk_r")
    Filter=(TranslationFilter=(bX=true,bY=true,bZ=true),RotationFilter=(bX=True,bY=true,bZ=true),ScaleFilter=(bX=True,bY=True,bZ=True))
    Parents=((Item=(Type=Bone,Name="spine_02"),Weight=1),(Item=(Type=Bone,Name="pelvis"),Weight=1))
    Weight=0.5
    bMaintainOffset=true
   )
   > &RigUnit_ParentConstraint(
    AdvancedSettings=(InterpolationType=Average,RotationOrderForFilter=XZY)
    Child=(Type=Bone,Name="MSTR_jacket_bk")
    Filter=(TranslationFilter=(bX=True,bY=true,bZ=true),RotationFilter=(bX=True,bY=true,bZ=true),ScaleFilter=(bX=True,bY=True,bZ=True))
    Parents=((Item=(Type=Bone,Name="spine_02"),Weight=1),(Item=(Type=Bone,Name="pelvis"),Weight=1))
    Weight=0.5
    bMaintainOffset=true
   )
   > &RigUnit_ParentConstraint(
    AdvancedSettings=(InterpolationType=Average,RotationOrderForFilter=XZY)
    Child=(Type=Bone,Name="MSTR_jacket_side_l")
    Filter=(TranslationFilter=(bX=true,bY=true,bZ=false),RotationFilter=(bX=true,bY=true,bZ=true),ScaleFilter=(bX=True,bY=True,bZ=True))
    Parents=((Item=(Type=Bone,Name="pelvis"),Weight=0.6),(Item=(Type=Bone,Name="spine_03"),Weight=0.2),(Item=(Type=Bone,Name="thigh_l"),Weight=0))
    Weight=0.5
    bMaintainOffset=true
   )
   > &RigUnit_ParentConstraint(
    AdvancedSettings=(InterpolationType=Average,RotationOrderForFilter=XZY)
    Child=(Type=Bone,Name="MSTR_jacket_side_r")
    Filter=(TranslationFilter=(bX=true,bY=true,bZ=false),RotationFilter=(bX=true,bY=true,bZ=true),ScaleFilter=(bX=True,bY=True,bZ=True))
    Parents=((Item=(Type=Bone,Name="pelvis"),Weight=0.6),(Item=(Type=Bone,Name="spine_03"),Weight=0.2),(Item=(Type=Bone,Name="thigh_r"),Weight=0))
    Weight=0.5
   )
   > &RigUnit_ParentConstraint(
    AdvancedSettings=(InterpolationType=Average,RotationOrderForFilter=XZY)
    Child=(Type=Bone,Name="MSTR_jacket_front_l")
    Filter=(TranslationFilter=(bX=true,bY=false,bZ=false),RotationFilter=(bX=true,bY=true,bZ=true),ScaleFilter=(bX=True,bY=True,bZ=True))
    Parents=((Item=(Type=Bone,Name="spine_03"),Weight=0.2),(Item=(Type=Bone,Name="pelvis"),Weight=0.8))
    Weight=0.5
    bMaintainOffset=true
   )
   > &RigUnit_ParentConstraint(
    AdvancedSettings=(InterpolationType=Average,RotationOrderForFilter=XZY)
    Child=(Type=Bone,Name="MSTR_jacket_front_r")
    Filter=(TranslationFilter=(bX=true,bY=false,bZ=false),RotationFilter=(bX=true,bY=true,bZ=true),ScaleFilter=(bX=True,bY=True,bZ=True))
    Parents=((Item=(Type=Bone,Name="spine_03"),Weight=0.2),(Item=(Type=Bone,Name="pelvis"),Weight=0.8))
    Weight=0.5
   )
  ]}
 New Function_ContainedGraph {nodes [
   RigVMFunctionEntryNode()
   > RigVMFunction_Sequence = RigVMAggregateNode()
   n1 = &RigUnit_SphericalPoseReader(
    ActiveRegionScaleFactors=(PositiveWidth=3.5,NegativeWidth=0.5,PositiveHeight=3,NegativeHeight=3)
    ActiveRegionSize=0.1
    Debug=(bDrawDebug=false,DebugScale=25,DebugSegments=20,DebugThickness=0.25)
    DriverAxis=(X=-1,Y=0,Z=0)
    DriverItem=(Type=Bone,Name="thigh_l")
    FalloffRegionScaleFactors=(PositiveWidth=1,NegativeWidth=0.8,PositiveHeight=1,NegativeHeight=1)
    FalloffSize=0.7
    OptionalParentItem=(Type=Bone,Name="spine_03")
    RotationOffset=(X=110,Y=0,Z=-30)
    ExecutePin=@RigVMFunction_Sequence.A
   )
   > RigVMVariableNode(Value=@n1.OutputParam) {variable=Thigh_L_Front}
   > n2 = &RigUnit_SphericalPoseReader(
    ActiveRegionScaleFactors=(PositiveWidth=3.5,NegativeWidth=0.5,PositiveHeight=3,NegativeHeight=3)
    ActiveRegionSize=0.1
    Debug=(bDrawDebug=false,DebugScale=25,DebugSegments=20,DebugThickness=0.25)
    DriverAxis=(X=1,Y=0,Z=0)
    DriverItem=(Type=Bone,Name="thigh_r")
    FalloffRegionScaleFactors=(PositiveWidth=1,NegativeWidth=0.8,PositiveHeight=1,NegativeHeight=1)
    FalloffSize=0.7
    FlipHeightScaling=false
    FlipWidthScaling=true
    OptionalParentItem=(Type=Bone,Name="spine_03")
    RotationOffset=(X=-70,Y=0,Z=-30)
   )
   > RigVMVariableNode(Value=@n2.OutputParam) {variable=Thigh_R_Front}
   > n3 = &RigUnit_SphericalPoseReader(
    ActiveRegionScaleFactors=(PositiveWidth=1,NegativeWidth=1,PositiveHeight=1,NegativeHeight=1)
    ActiveRegionSize=0.1
    Debug=(bDrawDebug=false,bDraw2D=False,bDrawLocalAxes=False,DebugScale=25,DebugSegments=20,DebugThickness=0.25)
    DriverAxis=(X=-1,Y=0,Z=0)
    DriverItem=(Type=Bone,Name="thigh_l")
    FalloffRegionScaleFactors=(PositiveWidth=1,NegativeWidth=1,PositiveHeight=1,NegativeHeight=1)
    FalloffSize=0.25
    FlipHeightScaling=False
    FlipWidthScaling=False
    OptionalParentItem=(Type=Bone,Name="None")
    RotationOffset=(X=165,Y=0,Z=0)
   )
   > RigVMVariableNode(Value=@n3.OutputParam) {variable=Thigh_L_Side}
   > n4 = &RigUnit_SphericalPoseReader(
    ActiveRegionScaleFactors=(PositiveWidth=1,NegativeWidth=1,PositiveHeight=1,NegativeHeight=1)
    ActiveRegionSize=0.1
    Debug=(bDrawDebug=false,bDraw2D=False,bDrawLocalAxes=False,DebugScale=25,DebugSegments=20,DebugThickness=0.25)
    DriverAxis=(X=1,Y=0,Z=0)
    DriverItem=(Type=Bone,Name="thigh_r")
    FalloffRegionScaleFactors=(PositiveWidth=1,NegativeWidth=1,PositiveHeight=1,NegativeHeight=1)
    FalloffSize=0.5
    FlipHeightScaling=true
    FlipWidthScaling=true
    OptionalParentItem=(Type=Bone,Name="None")
    RotationOffset=(X=0,Y=20,Z=0)
   )
   > RigVMVariableNode(Value=@n4.OutputParam) {variable=Thigh_R_Side}
   > n5 = &RigUnit_SphericalPoseReader(
    ActiveRegionScaleFactors=(PositiveWidth=1,NegativeWidth=1,PositiveHeight=1,NegativeHeight=1)
    ActiveRegionSize=0.05
    Debug=(bDrawDebug=false,bDraw2D=false,bDrawLocalAxes=False,DebugScale=25,DebugSegments=20,DebugThickness=0.25)
    DriverAxis=(X=1,Y=0,Z=0)
    DriverItem=(Type=Bone,Name="thigh_r")
    FalloffRegionScaleFactors=(PositiveWidth=1,NegativeWidth=1,PositiveHeight=1,NegativeHeight=1)
    FalloffSize=0.55
    FlipHeightScaling=true
    FlipWidthScaling=true
    OptionalParentItem=(Type=Bone,Name="None")
    RotationOffset=(X=90,Y=0,Z=20)
   )
   > RigVMVariableNode(Value=@n5.OutputParam) {variable=Thigh_R_Back}
   n6 = RigVMVariableNode() {variable=Thigh_L_Front}
   > &RigUnit_ParentConstraint(
    AdvancedSettings=(InterpolationType=Average,RotationOrderForFilter=XZY)
    Child=(Type=Bone,Name="pants_corrective_01")
    Parents=((Item=(Type=Bone,Name="thigh_l"),Weight=0.5),(Item=(Type=Bone,Name="thigh_r"),Weight=0.5))
    Weight=@n6.Value
   )
   > &RigUnit_ParentConstraint(
    AdvancedSettings=(InterpolationType=Average,RotationOrderForFilter=XZY)
    Child=(Type=Bone,Name="pants_corrective_02")
    Parents=((Item=(Type=Bone,Name="thigh_l"),Weight=0.5),(Item=(Type=Bone,Name="thigh_r"),Weight=0.5))
    Weight=@n6.Value
   )
   n7 = RigVMVariableNode() {variable=Thigh_R_Front}
   > &RigUnit_ParentConstraint(
    AdvancedSettings=(InterpolationType=Average,RotationOrderForFilter=XZY)
    Child=(Type=Bone,Name="pants_corrective_01")
    Parents=((Item=(Type=Bone,Name="thigh_l"),Weight=0.5),(Item=(Type=Bone,Name="thigh_r"),Weight=0.5))
    Weight=@n7.Value
   )
   > &RigUnit_ParentConstraint(
    AdvancedSettings=(InterpolationType=Average,RotationOrderForFilter=XZY)
    Child=(Type=Bone,Name="pants_corrective_02")
    Parents=((Item=(Type=Bone,Name="thigh_l"),Weight=0.5),(Item=(Type=Bone,Name="thigh_r"),Weight=0.5))
    Weight=@n7.Value
   )
   n8 = &RigVMFunction_AlphaInterp(
    InterpSpeedDecreasing=10
    InterpSpeedIncreasing=10
    Scale=1
    bClampResult=true
    Value=&RigVMFunction_MathBoolToFloat(Value=RigVMVariableNode() {variable=CrouchingBool}.Value).Result
   )
   > RigVMVariableNode(Value=@n8.Result) {variable=Crouching}
   RigVMFunctionReturnNode(ExecuteContext=@RigVMFunction_Sequence.E)
   &RigUnit_ModifyTransforms(
    ItemToModify="((Item=(Type=Bone,Name=\"MSTR_jacket_side_l\"),Transform=(Rotation=(X=0.4364,Y=-0.009,Z=-0.0632,W=0.8975),Translation=(X=1.7745,Y=4.5559,Z=-6.1624),Scale3D=(X=1,Y=1,Z=1))),(Item=(Type=Bone,Name=\"MSTR_jacket_bk_l\"),Transform=(Rotation=(X=0.228,Y=0.0205,Z=-0.1 ...(truncated)"
    WeightMinimum=0
    Weight=RigVMVariableNode() {variable=Thigh_L_Front}.Value
    ExecutePin=@RigVMFunction_Sequence.B
   )
   > &RigUnit_ModifyTransforms(
    ItemToModify="((Item=(Type=Bone,Name=\"MSTR_jacket_side_l\"),Transform=(Rotation=(X=0.0071,Y=0.077,Z=0.2078,W=-0.9751),Translation=(X=0,Y=0,Z=0),Scale3D=(X=1,Y=1,Z=1))),(Item=(Type=Bone,Name=\"MSTR_jacket_bk_l\"),Transform=(Rotation=(X=-0.0137,Y=0.1781,Z=0.1825,W=-0.9668),T ...(truncated)"
    WeightMinimum=0
    Weight=RigVMVariableNode() {variable=Thigh_L_Side}.Value
   )
   > &RigUnit_ModifyTransforms(
    ItemToModify="((Item=(Type=Bone,Name=\"base_chain_back\"),Transform=(Rotation=(X=0.1933,Y=-0.4739,Z=-0.1552,W=0.845),Translation=(X=0,Y=0,Z=0),Scale3D=(X=1,Y=1,Z=1))),(Item=(Type=Bone,Name=\"base_chain_front\"),Transform=(Rotation=(X=0,Y=0,Z=0,W=1),Translation=(X=0,Y=0,Z=0) ...(truncated)"
    WeightMinimum=0
    Weight=RigVMVariableNode() {variable=Thigh_R_Back}.Value
   )
   > &RigUnit_ModifyTransforms(
    ItemToModify="((Item=(Type=Bone,Name=\"MSTR_jacket_side_r\"),Transform=(Rotation=(X=-0.4342,Y=0.044,Z=-0.0462,W=0.8985),Translation=(X=1.7745,Y=2.5559,Z=0),Scale3D=(X=1,Y=1,Z=1))),(Item=(Type=Bone,Name=\"MSTR_jacket_bk_r\"),Transform=(Rotation=(X=-0.228,Y=-0.0205,Z=-0.1129, ...(truncated)"
    WeightMinimum=0
    Weight=&RigVMFunction_MathDoubleMul(A=RigVMVariableNode() {variable=Thigh_R_Front}.Value, B=RigVMVariableNode() {variable=NotCrouching}.Value).Result
   )
   RigVMFunction_MathBoolToFloat = &RigVMFunction_MathBoolToFloat(Value=&RigVMFunction_MathBoolNot(Value=RigVMVariableNode() {variable=Emote}.Value).Result)
   AlphaInterp = &RigVMFunction_AlphaInterp(InterpSpeedDecreasing=10, InterpSpeedIncreasing=10, Scale=1, bClampResult=true, Value=@RigVMFunction_MathBoolToFloat.Result)
   n9 = RigVMVariableNode() {variable=NotCrouching}
   Multiply_6 = RigVMAggregateNode(C=@AlphaInterp.Result, A=RigVMVariableNode() {variable=Thigh_L_Front}.Value, B=@n9.Value)
   > &RigUnit_ModifyTransforms(
    ItemToModify="((Item=(Type=Bone,Name=\"pants_corrective_01\"),Transform=(Rotation=(X=0,Y=0,Z=0,W=1),Translation=(X=4.5723,Y=0,Z=3.7852),Scale3D=(X=1,Y=1,Z=1))),(Item=(Type=Bone,Name=\"pants_corrective_02\"),Transform=(Rotation=(X=0.059,Y=0.527,Z=-0.2391,W=-0.8134),Translati ...(truncated)"
    WeightMinimum=0
    Weight=&RigVMFunction_MathDoubleRemap(SourceMaximum=1.5, SourceMinimum=0, TargetMaximum=1, TargetMinimum=0, Value=@Multiply_6.Result).Result
   )
   Multiply_7 = RigVMAggregateNode(A=RigVMVariableNode() {variable=Thigh_R_Front}.Value, B=@n9.Value, C=@AlphaInterp.Result)
   > &RigUnit_ModifyTransforms(
    ItemToModify="((Item=(Type=Bone,Name=\"pants_corrective_01\"),Transform=(Rotation=(X=0.0578,Y=-0.2831,Z=-0.1913,W=0.938),Translation=(X=2.0984,Y=0,Z=1.3936),Scale3D=(X=1,Y=1,Z=1))),(Item=(Type=Bone,Name=\"pants_corrective_02\"),Transform=(Rotation=(X=0.1177,Y=-0.6577,Z=-0.1 ...(truncated)"
    WeightMinimum=0
    Weight=&RigVMFunction_MathDoubleRemap(SourceMaximum=1.5, SourceMinimum=0, TargetMaximum=1, TargetMinimum=0, Value=@Multiply_7.Result).Result
   )
   Multiply_5 = RigVMAggregateNode(
    A=RigVMVariableNode() {variable=Thigh_L_Front}.Value
    B=RigVMVariableNode() {variable=Thigh_R_Front}.Value
    C=RigVMVariableNode() {variable=NotCrouching}.Value
    D=@AlphaInterp.Result
   )
   RigVMFunction_AnimEvalRichCurve = &RigVMFunction_AnimEvalRichCurve(
    Curve=(EditorCurveData=(Keys=((),(InterpMode=RCIM_Cubic,TangentMode=RCTM_User,Time=0.7096,Value=0.2588,ArriveTangent=0.6663,LeaveTangent=0.6663),(Time=1,Value=1))))
    SourceMinimum=0
    TargetMinimum=0
    Value=@Multiply_5.Result
   )
   > &RigUnit_ModifyTransforms(
    ItemToModify="((Item=(Type=Bone,Name=\"pants_corrective_01\"),Transform=(Rotation=(X=-0.5639,Y=0.1285,Z=0.0894,W=-0.8108),Translation=(X=-22.6373,Y=-4.8922,Z=-0.1741),Scale3D=(X=1,Y=1,Z=1))),(Item=(Type=Bone,Name=\"pants_corrective_02\"),Transform=(Rotation=(X=-0.0773,Y=0.4 ...(truncated)"
    WeightMinimum=0
    Weight=@RigVMFunction_AnimEvalRichCurve.Result
   )
   n10 = &RigVMFunction_AlphaInterp(
    InterpSpeedDecreasing=10
    InterpSpeedIncreasing=10
    Scale=1
    bClampResult=true
    Value=&RigVMFunction_MathBoolToFloat(Value=RigVMVariableNode() {variable=Emote}.Value).Result
   )
   Multiply_1 = RigVMAggregateNode(A=RigVMVariableNode() {variable=Thigh_L_Front}.Value, B=RigVMVariableNode() {variable=Thigh_R_Front}.Value, C=@n10.Result)
   > &RigUnit_ModifyTransforms(
    ItemToModify="((Item=(Type=Bone,Name=\"pants_corrective_01\"),Transform=(Rotation=(X=0,Y=0.6662,Z=0,W=0.7458),Translation=(X=-1.6336,Y=0,Z=-0.1848),Scale3D=(X=1,Y=1,Z=1))),(Item=(Type=Bone,Name=\"pants_corrective_02\"),Transform=(Rotation=(X=0,Y=0.1524,Z=0,W=-0.9883),Transl ...(truncated)"
    WeightMinimum=0
    Weight=@Multiply_1.Result
   )
   > &RigUnit_ModifyTransforms(
    ItemToModify="((Item=(Type=Bone,Name=\"base_buckle\"),Transform=(Rotation=(X=-0.0703,Y=0.2085,Z=-0.015,W=-0.9754),Translation=(X=-1.1036,Y=-0.3788,Z=2.3875),Scale3D=(X=1,Y=1,Z=1))),(Item=(Type=Bone,Name=\"corrective_pelvis_front\"),Transform=(Rotation=(X=0,Y=-0.0197,Z=0,W=- ...(truncated)"
    WeightMinimum=0
    Weight=&RigVMFunction_MathDoubleMul(A=RigVMVariableNode() {variable=Thigh_L_Front}.Value, B=RigVMVariableNode() {variable=NotCrouching}.Value).Result
   )
   > &RigUnit_ModifyTransforms(
    ItemToModify="((Item=(Type=Bone,Name=\"base_buckle\"),Transform=(Rotation=(X=-0.102,Y=-0.1993,Z=0.0209,W=-0.9744),Translation=(X=0.4285,Y=-0.2266,Z=0.9842),Scale3D=(X=1,Y=1,Z=1))),(Item=(Type=Bone,Name=\"corrective_pelvis_front\"),Transform=(Rotation=(X=-0.0025,Y=0,Z=0,W=-1 ...(truncated)"
    WeightMinimum=0
    Weight=&RigVMFunction_MathDoubleMul(A=RigVMVariableNode() {variable=Thigh_R_Front}.Value, B=RigVMVariableNode() {variable=NotCrouching}.Value).Result
   )
   Multiply = &RigVMFunction_MathDoubleMul(A=RigVMVariableNode() {variable=Thigh_L_Front}.Value, B=RigVMVariableNode() {variable=Crouching}.Value)
   &RigUnit_ModifyTransforms(
    ItemToModify="((Item=(Type=Bone,Name=\"MSTR_jacket_side_l\"),Transform=(Rotation=(X=0,Y=0,Z=0,W=1),Translation=(X=0,Y=0,Z=0),Scale3D=(X=1,Y=1,Z=1))),(Item=(Type=Bone,Name=\"MSTR_jacket_bk_l\"),Transform=(Rotation=(X=0,Y=0,Z=0,W=1),Translation=(X=0,Y=0,Z=0),Scale3D=(X=1,Y=1, ...(truncated)"
    WeightMinimum=0
    Weight=@Multiply.Result
    ExecutePin=@RigVMFunction_Sequence.C
   )
   n11 = RigVMVariableNode() {variable=Thigh_R_Side}
   n12 = RigVMVariableNode() {variable=Crouching}
   > &RigUnit_ModifyTransforms(
    ItemToModify="((Item=(Type=Bone,Name=\"MSTR_jacket_front_r\"),Transform=(Rotation=(X=-0.5055,Y=-0.306,Z=0.2045,W=0.7803),Translation=(X=-2.2924,Y=2.39,Z=1.084),Scale3D=(X=1,Y=1,Z=1))),(Item=(Type=Bone,Name=\"MSTR_jacket_side_r\"),Transform=(Rotation=(X=0.0221,Y=0.0904,Z=0.0 ...(truncated)"
    WeightMinimum=0
    Weight=&RigVMFunction_MathDoubleRemap(
     SourceMaximum=0.5
     SourceMinimum=0
     TargetMaximum=1
     TargetMinimum=0
     Value=&RigVMFunction_MathDoubleMul(A=@n11.Value, B=@n12.Value).Result
    ).Result
   )
   n13 = RigVMVariableNode() {variable=Moving2d}
   n14 = &RigVMFunction_AlphaInterp(
    InterpSpeedDecreasing=10
    InterpSpeedIncreasing=10
    Scale=1
    bClampResult=true
    Value=&RigVMFunction_MathBoolToFloat(Value=&RigVMFunction_MathBoolNot(Value=@n13.Value).Result).Result
   )
   RigVMFunction_MathBoolNot = &RigVMFunction_MathBoolNot(Value=RigVMVariableNode() {variable=Sliding}.Value)
   Multiply_3 = RigVMAggregateNode(
    B=@n14.Result
    C=&RigVMFunction_MathBoolToFloat(Value=@RigVMFunction_MathBoolNot.Result).Result
    A=RigVMVariableNode() {variable=Crouching}.Value
   )
   > &RigUnit_ModifyTransforms(
    ItemToModify="((Item=(Type=Bone,Name=\"base_chain_back\"),Transform=(Rotation=(X=-0.0858,Y=-0.2814,Z=-0.2787,W=0.9142),Translation=(X=0,Y=0,Z=0),Scale3D=(X=1,Y=1,Z=1))),(Item=(Type=Bone,Name=\"deform_glute_l\"),Transform=(Rotation=(X=0.0117,Y=0.0389,Z=0.1596,W=-0.9863),Tran ...(truncated)"
    WeightMinimum=0
    Weight=@Multiply_3.Result
   )
   n15 = RigVMVariableNode() {variable=Crouching}
   > &RigUnit_ModifyTransforms(
    ItemToModify="((Item=(Type=Bone,Name=\"pants_corrective_01\"),Transform=(Rotation=(X=0.0566,Y=0,Z=0,W=0.9984),Translation=(X=0.2673,Y=0.7589,Z=0),Scale3D=(X=1,Y=0.5447,Z=1))),(Item=(Type=Bone,Name=\"pants_corrective_02\"),Transform=(Rotation=(X=-0.0174,Y=0.3423,Z=-0.145,W=- ...(truncated)"
    WeightMinimum=0
    Weight=@n15.Value
   )
   > &RigUnit_ModifyTransforms(
    ItemToModify="((Item=(Type=Bone,Name=\"pants_corrective_01\"),Transform=(Rotation=(X=0,Y=0,Z=0,W=1),Translation=(X=0,Y=0,Z=0),Scale3D=(X=1,Y=1,Z=1))),(Item=(Type=Bone,Name=\"pants_corrective_02\"),Transform=(Rotation=(X=0,Y=-0.1246,Z=0,W=-0.9922),Translation=(X=-0.0885,Y=0, ...(truncated)"
    WeightMinimum=0
    Weight=&RigVMFunction_MathDoubleMul(B=@n15.Value, A=RigVMVariableNode() {variable=Thigh_R_Front}.Value).Result
   )
   n16 = &RigVMFunction_AlphaInterp(
    InterpSpeedDecreasing=10
    InterpSpeedIncreasing=10
    Scale=1
    bClampResult=true
    Value=&RigVMFunction_MathBoolToFloat(Value=RigVMVariableNode() {variable=Sliding}.Value).Result
   )
   > &RigUnit_ModifyTransforms(
    ItemToModify="((Item=(Type=Bone,Name=\"base_chain_back\"),Transform=(Rotation=(X=0,Y=0,Z=-0.3214,W=0.9469),Translation=(X=-1.689,Y=-2.2015,Z=0),Scale3D=(X=1,Y=1,Z=1))),(Item=(Type=Bone,Name=\"base_chain_front\"),Transform=(Rotation=(X=-0.0123,Y=-0.1504,Z=-0.2166,W=0.9645),T ...(truncated)"
    WeightMinimum=0
    Weight=@n16.Result
   )
   n17 = &RigVMFunction_AlphaInterp(
    InterpSpeedDecreasing=10
    InterpSpeedIncreasing=10
    Scale=1
    bClampResult=true
    Value=&RigVMFunction_MathBoolToFloat(Value=RigVMVariableNode() {variable=Moving2d}.Value).Result
   )
   > &RigUnit_ModifyTransforms(
    ItemToModify=((Item=(Type=Bone,Name="deform_glute_l"),Transform=(Rotation=(X=0,Y=0,Z=-0.1728,W=0.985),Translation=(X=0.6144,Y=0,Z=0),Scale3D=(X=1,Y=1,Z=1))))
    WeightMinimum=0
    Weight=&RigVMFunction_MathDoubleMul(B=@n17.Result, A=RigVMVariableNode() {variable=Crouching}.Value).Result
   )
   n18 = &RigVMFunction_AlphaInterp(
    InterpSpeedDecreasing=10
    InterpSpeedIncreasing=10
    Scale=1
    bClampResult=true
    Value=&RigVMFunction_MathBoolToFloat(Value=RigVMVariableNode() {variable=CrouchingBool}.Value).Result
   )
   Remap = &RigVMFunction_MathDoubleRemap(
    SourceMaximum=0.6
    SourceMinimum=0
    TargetMaximum=1
    TargetMinimum=0
    Value=&RigVMFunction_MathDoubleRemap(
     SourceMaximum=39
     SourceMinimum=43
     TargetMaximum=1
     TargetMinimum=0
     Value=&RigUnit_GetRelativeTransformForItem(Child=(Type=Bone,Name="pelvis"), Parent=(Type=Bone,Name="root"), bChildInitial=False, bParentInitial=False).RelativeTransform.Translation.Z
    ).Result
   )
   &RigUnit_ModifyTransforms(
    ItemToModify=((Item=(Type=Bone,Name="base_chain_back"),Transform=(Rotation=(X=0,Y=0,Z=0.1224,W=-0.9925),Translation=(X=0,Y=0,Z=0),Scale3D=(X=1,Y=1,Z=1))))
    WeightMinimum=0
    Weight=&RigVMFunction_MathDoubleMul(B=@n18.Result, A=@Remap.Result).Result
    ExecutePin=@RigVMFunction_Sequence.D
   )
   > DeltaY = RigVMFunctionReferenceNode(Child=Thigh_L, Parent=pelvis, SourceMax=-100, SourceMin=-30)
   > &RigUnit_ModifyTransforms(
    ItemToModify="((Item=(Type=Bone,Name=\"MSTR_jacket_front_l\"),Transform=(Rotation=(X=0,Y=0.5807,Z=0,W=0.8141),Translation=(X=0,Y=0,Z=0),Scale3D=(X=0.6331,Y=1,Z=1))),(Item=(Type=Bone,Name=\"MSTR_jacket_side_l\"),Transform=(Rotation=(X=0,Y=0,Z=0,W=1),Translation=(X=0,Y=0,Z=0) ...(truncated)"
    WeightMinimum=0
    Weight=&RigVMFunction_MathDoubleMul(A=@DeltaY.DeltaY, B=RigVMVariableNode() {variable=EasternBlocEmote}.Value).Result
   )
   > DeltaY_1 = RigVMFunctionReferenceNode(Child=thigh_r, Parent=pelvis, SourceMax=-80, SourceMin=-30)
   > &RigUnit_ModifyTransforms(
    ItemToModify="((Item=(Type=Bone,Name=\"MSTR_jacket_front_r\"),Transform=(Rotation=(X=0,Y=-0.6553,Z=0,W=0.7554),Translation=(X=-1.8738,Y=0,Z=1.3153),Scale3D=(X=0.3,Y=1,Z=1))),(Item=(Type=Bone,Name=\"MSTR_jacket_side_r\"),Transform=(Rotation=(X=0,Y=0.017,Z=0,W=-0.9999),Transl ...(truncated)"
    WeightMinimum=0
    Weight=&RigVMFunction_MathDoubleMul(A=@DeltaY_1.DeltaY, B=RigVMVariableNode() {variable=EasternBlocEmote}.Value).Result
   )
  ]}
 New Function_ContainedGraph {nodes [
   n1 = RigVMFunctionEntryNode()
   RigVMFunction_Sequence = &RigVMFunction_Sequence(ExecuteContext=@n1.ExecuteContext)
   n2 = &RigUnit_SphericalPoseReader(
    ActiveRegionScaleFactors=(PositiveWidth=1,NegativeWidth=1,PositiveHeight=2.4,NegativeHeight=1.4)
    ActiveRegionSize=0.06
    Debug=(bDrawDebug=false,bDraw2D=False,bDrawLocalAxes=False,DebugScale=25,DebugSegments=20,DebugThickness=0.25)
    DriverAxis=(X=0,Y=1,Z=0)
    DriverItem=(Type=Bone,Name="head")
    FalloffRegionScaleFactors=(PositiveWidth=1,NegativeWidth=1,PositiveHeight=6,NegativeHeight=1.5)
    FalloffSize=0.2
    FlipHeightScaling=true
    FlipWidthScaling=false
    OptionalParentItem=(Type=Bone,Name="spine_05")
    RotationOffset=(X=50,Y=0,Z=30)
    ExecutePin=@RigVMFunction_Sequence.A
   )
   > &RigUnit_ModifyTransforms(
    ItemToModify="((Item=(Type=Bone,Name=\"base_collar_sd_r\"),Transform=(Rotation=(X=-0.0754,Y=0.087,Z=0.0231,W=0.9931),Translation=(X=0,Y=0,Z=0),Scale3D=(X=1,Y=1,Z=1))),(Item=(Type=Bone,Name=\"base_collar_fr_r\"),Transform=(Rotation=(X=0.1487,Y=-0.0548,Z=-0.0934,W=-0.9829),Tr ...(truncated)"
    WeightMinimum=0
    Weight=@n2.OutputParam
   )
   > n3 = &RigUnit_SphericalPoseReader(
    ActiveRegionScaleFactors=(PositiveWidth=1,NegativeWidth=1,PositiveHeight=2.4,NegativeHeight=0.5)
    ActiveRegionSize=0.07
    Debug=(bDrawDebug=false,bDraw2D=False,bDrawLocalAxes=False,DebugScale=25,DebugSegments=20,DebugThickness=0.25)
    DriverAxis=(X=0,Y=1,Z=0)
    DriverItem=(Type=Bone,Name="head")
    FalloffRegionScaleFactors=(PositiveWidth=1,NegativeWidth=1,PositiveHeight=6,NegativeHeight=1)
    FalloffSize=0.22
    FlipHeightScaling=False
    FlipWidthScaling=False
    OptionalParentItem=(Type=Bone,Name="spine_05")
    RotationOffset=(X=148,Y=0,Z=37)
   )
   > &RigUnit_ModifyTransforms(
    ItemToModify="((Item=(Type=Bone,Name=\"base_collar_sd_l\"),Transform=(Rotation=(X=0.0611,Y=-0.2262,Z=-0.0377,W=-0.9714),Translation=(X=0,Y=0,Z=0),Scale3D=(X=1,Y=1,Z=1))),(Item=(Type=Bone,Name=\"base_collar_fr_l\"),Transform=(Rotation=(X=0.0877,Y=-0.1927,Z=-0.0486,W=-0.9761) ...(truncated)"
    WeightMinimum=0
    Weight=@n3.OutputParam
   )
   RigVMFunction_MathBoolToFloat = &RigVMFunction_MathBoolToFloat(Value=RigVMVariableNode() {variable=Floating}.Value)
   AlphaInterp = &RigVMFunction_AlphaInterp(InterpSpeedDecreasing=10, InterpSpeedIncreasing=10, Scale=1, bClampResult=true, Value=@RigVMFunction_MathBoolToFloat.Result)
   > &RigUnit_ModifyTransforms(
    ItemToModify="((Item=(Type=Bone,Name=\"base_necklace_a\"),Transform=(Rotation=(X=0,Y=0.2941,Z=0,W=-0.9558),Translation=(X=0,Y=0,Z=0),Scale3D=(X=1,Y=1,Z=1))),(Item=(Type=Bone,Name=\"base_necklace_b\"),Transform=(Rotation=(X=0,Y=-0.2429,Z=0,W=0.9701),Translation=(X=0,Y=0,Z=0) ...(truncated)"
    WeightMinimum=0
    Weight=@AlphaInterp.Result
   )
   RigVMFunctionReturnNode(ExecuteContext=@RigVMFunction_Sequence.B)
  ]}
 New Function_ContainedGraph {nodes [
   RigVMFunctionEntryNode()
   > RigVMFunction_Sequence = RigVMAggregateNode()
   Multiply = RigVMAggregateNode(
    A=RigVMDispatchNode(BoneName=root, Default=0, Name=deformRig_r_upperarm_up_alpha) {template="DISPATCH_RigDispatch_GetAnimAttribute(in Name,in BoneName,in Default,out Value,out Found)"}.Value
    C=RigVMVariableNode() {variable=OnGround}.Value
    B=RigVMVariableNode() {variable=Detected_Pistol}.Value
   )
   &RigUnit_ModifyTransforms(
    ItemToModify="((Item=(Type=Bone,Name=\"deform_upperarm_bk_base_r\"),Transform=(Rotation=(X=-0.0584,Y=0.2534,Z=0.1408,W=-0.9553),Translation=(X=3.7173,Y=0,Z=10.6083),Scale3D=(X=1,Y=1,Z=1))),(Item=(Type=Bone,Name=\"deform_upperarm_up_r\"),Transform=(Rotation=(X=0,Y=0,Z=0,W=-1 ...(truncated)"
    WeightMinimum=0
    Weight=@Multiply.Result
    ExecutePin=@RigVMFunction_Sequence.A
   )
   RigVMFunctionReturnNode(ExecuteContext=@RigVMFunction_Sequence.D)
   RigVMFunction_MathBoolNot = &RigVMFunction_MathBoolNot(Value=RigVMVariableNode() {variable=Moving2d}.Value)
   RigVMFunction_MathBoolToFloat = &RigVMFunction_MathBoolToFloat(Value=@RigVMFunction_MathBoolNot.Result)
   AlphaInterp = &RigVMFunction_AlphaInterp(InterpSpeedDecreasing=10, InterpSpeedIncreasing=10, Scale=1, bClampResult=true, Value=@RigVMFunction_MathBoolToFloat.Result)
   &RigUnit_ModifyTransforms(
    ItemToModify="((Item=(Type=Bone,Name=\"base_laces_l\"),Transform=(Rotation=(X=0.0849,Y=0.5363,Z=-0.2145,W=-0.8119),Translation=(X=0,Y=0,Z=0),Scale3D=(X=1,Y=1,Z=1))),(Item=(Type=Bone,Name=\"base_laces_r\"),Transform=(Rotation=(X=0.1567,Y=0.1647,Z=0.4163,W=-0.8804),Translatio ...(truncated)"
    WeightMinimum=0
    Weight=&RigVMFunction_MathDoubleMul(B=@AlphaInterp.Result, A=RigVMVariableNode() {variable=Crouching}.Value).Result
    ExecutePin=@RigVMFunction_Sequence.B
   )
   &RigUnit_ModifyTransforms(
    ItemToModify="((Item=(Type=Bone,Name=\"MSTR_jacket_bk_l\"),Transform=(Rotation=(X=0,Y=0,Z=-0.2229,W=0.9748),Translation=(X=0,Y=0,Z=0),Scale3D=(X=1,Y=1,Z=1))),(Item=(Type=Bone,Name=\"MSTR_jacket_bk\"),Transform=(Rotation=(X=0,Y=0,Z=-0.2487,W=0.9686),Translation=(X=0,Y=0,Z=0) ...(truncated)"
    WeightMinimum=0
    Weight=RigVMVariableNode() {variable=SkydiveUp}.Value
    ExecutePin=@RigVMFunction_Sequence.C
   )
   n1 = RigVMVariableNode() {variable=Moving2d}
   n2 = &RigVMFunction_AlphaInterp(
    InterpSpeedDecreasing=10
    InterpSpeedIncreasing=10
    Scale=1
    bClampResult=true
    Value=&RigVMFunction_MathBoolToFloat(Value=&RigVMFunction_MathBoolNot(Value=@n1.Value).Result).Result
   )
   > &RigUnit_ModifyTransforms(
    ItemToModify="((Item=(Type=Bone,Name=\"MSTR_jacket_bk_l\"),Transform=(Rotation=(X=0,Y=0,Z=0,W=-1),Translation=(X=2.5634,Y=0,Z=0),Scale3D=(X=1,Y=1,Z=1))),(Item=(Type=Bone,Name=\"MSTR_jacket_bk\"),Transform=(Rotation=(X=0,Y=0,Z=0,W=-1),Translation=(X=3.7471,Y=0,Z=0),Scale3D=( ...(truncated)"
    WeightMinimum=0
    Weight=&RigVMFunction_MathDoubleMul(A=RigVMVariableNode() {variable=Crouching}.Value, B=@n2.Result).Result
   )
   n3 = &RigVMFunction_AlphaInterp(
    InterpSpeedDecreasing=10
    InterpSpeedIncreasing=10
    Scale=1
    bClampResult=true
    Value=&RigVMFunction_MathBoolToFloat(Value=RigVMVariableNode() {variable=Moving2d}.Value).Result
   )
   > &RigUnit_ModifyTransforms(
    ItemToModify="((Item=(Type=Bone,Name=\"MSTR_jacket_bk_l\"),Transform=(Rotation=(X=0.0924,Y=0.0815,Z=-0.1481,W=-0.9813),Translation=(X=5.4262,Y=-0.0005,Z=0),Scale3D=(X=1,Y=1,Z=1))),(Item=(Type=Bone,Name=\"MSTR_jacket_bk\"),Transform=(Rotation=(X=-0.0113,Y=-0.0264,Z=-0.0652,W ...(truncated)"
    WeightMinimum=0
    Weight=&RigVMFunction_MathDoubleMul(A=RigVMVariableNode() {variable=Crouching}.Value, B=@n3.Result).Result
   )
  ]}
 New Function_ContainedGraph {nodes [
   RigVMFunctionEntryNode()
   > RigVMFunction_Sequence = RigVMAggregateNode()
   n1 = &RigUnit_SphericalPoseReader(
    ActiveRegionScaleFactors=(PositiveWidth=3.5,NegativeWidth=0.5,PositiveHeight=3,NegativeHeight=3)
    ActiveRegionSize=0.1
    Debug=(bDrawDebug=false,DebugScale=25,DebugSegments=20,DebugThickness=0.25)
    DriverAxis=(X=-1,Y=0,Z=0)
    DriverItem=(Type=Bone,Name="thigh_l")
    FalloffRegionScaleFactors=(PositiveWidth=1,NegativeWidth=0.8,PositiveHeight=1,NegativeHeight=1)
    FalloffSize=0.7
    OptionalParentItem=(Type=Bone,Name="spine_03")
    RotationOffset=(X=110,Y=0,Z=-30)
    ExecutePin=@RigVMFunction_Sequence.A
   )
   > RigVMVariableNode(Value=@n1.OutputParam) {variable=Thigh_L_Front}
   > n2 = &RigUnit_SphericalPoseReader(
    ActiveRegionScaleFactors=(PositiveWidth=3.5,NegativeWidth=0.5,PositiveHeight=3,NegativeHeight=3)
    ActiveRegionSize=0.1
    Debug=(bDrawDebug=false,DebugScale=25,DebugSegments=20,DebugThickness=0.25)
    DriverAxis=(X=1,Y=0,Z=0)
    DriverItem=(Type=Bone,Name="thigh_r")
    FalloffRegionScaleFactors=(PositiveWidth=1,NegativeWidth=0.8,PositiveHeight=1,NegativeHeight=1)
    FalloffSize=0.7
    FlipHeightScaling=false
    FlipWidthScaling=true
    OptionalParentItem=(Type=Bone,Name="spine_03")
    RotationOffset=(X=-70,Y=0,Z=-30)
   )
   > RigVMVariableNode(Value=@n2.OutputParam) {variable=Thigh_R_Front}
   > n3 = &RigUnit_SphericalPoseReader(
    ActiveRegionScaleFactors=(PositiveWidth=1,NegativeWidth=1,PositiveHeight=1,NegativeHeight=1)
    ActiveRegionSize=0.1
    Debug=(bDrawDebug=false,bDraw2D=False,bDrawLocalAxes=False,DebugScale=25,DebugSegments=20,DebugThickness=0.25)
    DriverAxis=(X=-1,Y=0,Z=0)
    DriverItem=(Type=Bone,Name="thigh_l")
    FalloffRegionScaleFactors=(PositiveWidth=1,NegativeWidth=1,PositiveHeight=1,NegativeHeight=1)
    FalloffSize=0.25
    FlipHeightScaling=False
    FlipWidthScaling=False
    OptionalParentItem=(Type=Bone,Name="None")
    RotationOffset=(X=165,Y=0,Z=0)
   )
   > RigVMVariableNode(Value=@n3.OutputParam) {variable=Thigh_L_Side}
   > n4 = &RigUnit_SphericalPoseReader(
    ActiveRegionScaleFactors=(PositiveWidth=1,NegativeWidth=1,PositiveHeight=1,NegativeHeight=1)
    ActiveRegionSize=0.05
    Debug=(bDrawDebug=false,bDraw2D=False,bDrawLocalAxes=False,DebugScale=25,DebugSegments=20,DebugThickness=0.25)
    DriverAxis=(X=1,Y=0,Z=0)
    DriverItem=(Type=Bone,Name="thigh_r")
    FalloffRegionScaleFactors=(PositiveWidth=1,NegativeWidth=1.2,PositiveHeight=1,NegativeHeight=1)
    FalloffSize=0.55
    FlipHeightScaling=true
    FlipWidthScaling=true
    OptionalParentItem=(Type=Bone,Name="None")
    RotationOffset=(X=0,Y=20,Z=0)
   )
   > RigVMVariableNode(Value=@n4.OutputParam) {variable=Thigh_R_Side}
   > n5 = &RigUnit_SphericalPoseReader(
    ActiveRegionScaleFactors=(PositiveWidth=1,NegativeWidth=1,PositiveHeight=1,NegativeHeight=1)
    ActiveRegionSize=0.05
    Debug=(bDrawDebug=false,bDraw2D=false,bDrawLocalAxes=False,DebugScale=25,DebugSegments=20,DebugThickness=0.25)
    DriverAxis=(X=1,Y=0,Z=0)
    DriverItem=(Type=Bone,Name="thigh_r")
    FalloffRegionScaleFactors=(PositiveWidth=1,NegativeWidth=1,PositiveHeight=1,NegativeHeight=1)
    FalloffSize=0.55
    FlipHeightScaling=true
    FlipWidthScaling=true
    OptionalParentItem=(Type=Bone,Name="None")
    RotationOffset=(X=90,Y=0,Z=-40)
   )
   > RigVMVariableNode(Value=@n5.OutputParam) {variable=Thigh_R_Back}
   n6 = RigVMVariableNode() {variable=Thigh_L_Front}
   > &RigUnit_ParentConstraint(
    AdvancedSettings=(InterpolationType=Average,RotationOrderForFilter=XZY)
    Child=(Type=Bone,Name="pants_corrective_01")
    Parents=((Item=(Type=Bone,Name="thigh_l"),Weight=0.5),(Item=(Type=Bone,Name="thigh_r"),Weight=0.5))
    Weight=@n6.Value
   )
   > &RigUnit_ParentConstraint(
    AdvancedSettings=(InterpolationType=Average,RotationOrderForFilter=XZY)
    Child=(Type=Bone,Name="pants_corrective_02")
    Parents=((Item=(Type=Bone,Name="thigh_l"),Weight=0.5),(Item=(Type=Bone,Name="thigh_r"),Weight=0.5))
    Weight=@n6.Value
   )
   n7 = RigVMVariableNode() {variable=Thigh_R_Front}
   > &RigUnit_ParentConstraint(
    AdvancedSettings=(InterpolationType=Average,RotationOrderForFilter=XZY)
    Child=(Type=Bone,Name="pants_corrective_01")
    Parents=((Item=(Type=Bone,Name="thigh_l"),Weight=0.5),(Item=(Type=Bone,Name="thigh_r"),Weight=0.5))
    Weight=@n7.Value
   )
   > &RigUnit_ParentConstraint(
    AdvancedSettings=(InterpolationType=Average,RotationOrderForFilter=XZY)
    Child=(Type=Bone,Name="pants_corrective_02")
    Parents=((Item=(Type=Bone,Name="thigh_l"),Weight=0.5),(Item=(Type=Bone,Name="thigh_r"),Weight=0.5))
    Weight=@n7.Value
   )
   n8 = &RigVMFunction_AlphaInterp(
    InterpSpeedDecreasing=10
    InterpSpeedIncreasing=10
    Scale=1
    bClampResult=true
    Value=&RigVMFunction_MathBoolToFloat(Value=RigVMVariableNode() {variable=CrouchingBool}.Value).Result
   )
   > RigVMVariableNode(Value=@n8.Result) {variable=Crouching}
   RigVMFunctionReturnNode(ExecuteContext=@RigVMFunction_Sequence.E)
   &RigUnit_ModifyTransforms(
    ItemToModify="((Item=(Type=Bone,Name=\"MSTR_jacket_side_l\"),Transform=(Rotation=(X=0,Y=0,Z=0,W=1),Translation=(X=0,Y=0,Z=0),Scale3D=(X=1,Y=1,Z=1))),(Item=(Type=Bone,Name=\"MSTR_jacket_bk_l\"),Transform=(Rotation=(X=0,Y=0,Z=0,W=1),Translation=(X=0,Y=0,Z=0),Scale3D=(X=1,Y=1, ...(truncated)"
    WeightMinimum=0
    Weight=&RigVMFunction_MathDoubleMul(B=RigVMVariableNode() {variable=NotCrouching}.Value, A=RigVMVariableNode() {variable=Thigh_L_Front}.Value).Result
    ExecutePin=@RigVMFunction_Sequence.B
   )
   > &RigUnit_ModifyTransforms(
    ItemToModify="((Item=(Type=Bone,Name=\"MSTR_jacket_side_l\"),Transform=(Rotation=(X=0.0071,Y=0.077,Z=0.2078,W=-0.9751),Translation=(X=0,Y=0,Z=0),Scale3D=(X=1,Y=1,Z=1))),(Item=(Type=Bone,Name=\"MSTR_jacket_bk_l\"),Transform=(Rotation=(X=-0.0137,Y=0.1781,Z=0.1825,W=-0.9668),T ...(truncated)"
    WeightMinimum=0
    Weight=&RigVMFunction_MathDoubleMul(A=RigVMVariableNode() {variable=Thigh_L_Side}.Value, B=RigVMVariableNode() {variable=NotCrouching}.Value).Result
   )
   > &RigUnit_ModifyTransforms(
    ItemToModify="((Item=(Type=Bone,Name=\"base_chain_front\"),Transform=(Rotation=(X=-0.0171,Y=0.0043,Z=0.4268,W=-0.9042),Translation=(X=0,Y=0,Z=0),Scale3D=(X=1,Y=1,Z=1))),(Item=(Type=Bone,Name=\"base_chain_back\"),Transform=(Rotation=(X=0,Y=0,Z=-0.2143,W=0.9768),Translation=( ...(truncated)"
    WeightMinimum=0
    Weight=&RigVMFunction_MathDoubleMul(A=RigVMVariableNode() {variable=Thigh_R_Side}.Value, B=RigVMVariableNode() {variable=NotCrouching}.Value).Result
   )
   > &RigUnit_ModifyTransforms(
    ItemToModify="((Item=(Type=Bone,Name=\"MSTR_jacket_side_r\"),Transform=(Rotation=(X=0,Y=0,Z=0,W=1),Translation=(X=0,Y=0,Z=0),Scale3D=(X=1,Y=1,Z=1))),(Item=(Type=Bone,Name=\"MSTR_jacket_bk_r\"),Transform=(Rotation=(X=0,Y=0,Z=0,W=1),Translation=(X=0,Y=0,Z=0),Scale3D=(X=1,Y=1, ...(truncated)"
    WeightMinimum=0
    Weight=&RigVMFunction_MathDoubleMul(A=RigVMVariableNode() {variable=Thigh_R_Front}.Value, B=RigVMVariableNode() {variable=NotCrouching}.Value).Result
   )
   RigVMFunction_MathBoolToFloat = &RigVMFunction_MathBoolToFloat(Value=&RigVMFunction_MathBoolNot(Value=RigVMVariableNode() {variable=Emote}.Value).Result)
   AlphaInterp = &RigVMFunction_AlphaInterp(InterpSpeedDecreasing=10, InterpSpeedIncreasing=10, Scale=1, bClampResult=true, Value=@RigVMFunction_MathBoolToFloat.Result)
   n9 = RigVMVariableNode() {variable=NotCrouching}
   Multiply_6 = RigVMAggregateNode(C=@AlphaInterp.Result, A=RigVMVariableNode() {variable=Thigh_L_Front}.Value, B=@n9.Value)
   > &RigUnit_ModifyTransforms(
    ItemToModify="((Item=(Type=Bone,Name=\"pants_corrective_01\"),Transform=(Rotation=(X=0,Y=0,Z=0,W=1),Translation=(X=4.5723,Y=0,Z=3.7852),Scale3D=(X=1,Y=1,Z=1))),(Item=(Type=Bone,Name=\"pants_corrective_02\"),Transform=(Rotation=(X=0.059,Y=0.527,Z=-0.2391,W=-0.8134),Translati ...(truncated)"
    WeightMinimum=0
    Weight=@Multiply_6.Result
   )
   Multiply_7 = RigVMAggregateNode(A=RigVMVariableNode() {variable=Thigh_R_Front}.Value, B=@n9.Value, C=@AlphaInterp.Result)
   > &RigUnit_ModifyTransforms(
    ItemToModify="((Item=(Type=Bone,Name=\"pants_corrective_01\"),Transform=(Rotation=(X=0.0578,Y=-0.2831,Z=-0.1913,W=0.938),Translation=(X=2.0984,Y=0,Z=1.3936),Scale3D=(X=1,Y=1,Z=1))),(Item=(Type=Bone,Name=\"pants_corrective_02\"),Transform=(Rotation=(X=0.1177,Y=-0.6577,Z=-0.1 ...(truncated)"
    WeightMinimum=0
    Weight=@Multiply_7.Result
   )
   Multiply_5 = RigVMAggregateNode(
    A=RigVMVariableNode() {variable=Thigh_L_Front}.Value
    B=RigVMVariableNode() {variable=Thigh_R_Front}.Value
    C=RigVMVariableNode() {variable=NotCrouching}.Value
    D=@AlphaInterp.Result
   )
   > &RigUnit_ModifyTransforms(
    ItemToModify="((Item=(Type=Bone,Name=\"pants_corrective_01\"),Transform=(Rotation=(X=0,Y=0,Z=0,W=-1),Translation=(X=-21.0996,Y=-5.3612,Z=0),Scale3D=(X=1,Y=1,Z=1))),(Item=(Type=Bone,Name=\"pants_corrective_02\"),Transform=(Rotation=(X=0,Y=0.4237,Z=0,W=0.9058),Translation=(X= ...(truncated)"
    WeightMinimum=0
    Weight=@Multiply_5.Result
   )
   n10 = &RigVMFunction_AlphaInterp(
    InterpSpeedDecreasing=10
    InterpSpeedIncreasing=10
    Scale=1
    bClampResult=true
    Value=&RigVMFunction_MathBoolToFloat(Value=RigVMVariableNode() {variable=Emote}.Value).Result
   )
   Multiply_1 = RigVMAggregateNode(A=RigVMVariableNode() {variable=Thigh_L_Front}.Value, B=RigVMVariableNode() {variable=Thigh_R_Front}.Value, C=@n10.Result)
   > &RigUnit_ModifyTransforms(
    ItemToModify="((Item=(Type=Bone,Name=\"pants_corrective_01\"),Transform=(Rotation=(X=0,Y=0.6662,Z=0,W=0.7458),Translation=(X=-1.6336,Y=0,Z=-0.1848),Scale3D=(X=1,Y=1,Z=1))),(Item=(Type=Bone,Name=\"pants_corrective_02\"),Transform=(Rotation=(X=0,Y=0.1524,Z=0,W=-0.9883),Transl ...(truncated)"
    WeightMinimum=0
    Weight=@Multiply_1.Result
   )
   > &RigUnit_ModifyTransforms(
    ItemToModify="((Item=(Type=Bone,Name=\"base_buckle\"),Transform=(Rotation=(X=-0.0703,Y=0.2085,Z=-0.015,W=-0.9754),Translation=(X=-1.1036,Y=-0.3788,Z=2.3875),Scale3D=(X=1,Y=1,Z=1))),(Item=(Type=Bone,Name=\"corrective_pelvis_front\"),Transform=(Rotation=(X=0,Y=-0.0197,Z=0,W=- ...(truncated)"
    WeightMinimum=0
    Weight=&RigVMFunction_MathDoubleMul(A=RigVMVariableNode() {variable=Thigh_L_Front}.Value, B=RigVMVariableNode() {variable=NotCrouching}.Value).Result
   )
   > &RigUnit_ModifyTransforms(
    ItemToModify="((Item=(Type=Bone,Name=\"base_buckle\"),Transform=(Rotation=(X=-0.102,Y=-0.1993,Z=0.0209,W=-0.9744),Translation=(X=0.4285,Y=-0.2266,Z=0.9842),Scale3D=(X=1,Y=1,Z=1))),(Item=(Type=Bone,Name=\"corrective_pelvis_front\"),Transform=(Rotation=(X=-0.0025,Y=0,Z=0,W=-1 ...(truncated)"
    WeightMinimum=0
    Weight=&RigVMFunction_MathDoubleMul(A=RigVMVariableNode() {variable=Thigh_R_Front}.Value, B=RigVMVariableNode() {variable=NotCrouching}.Value).Result
   )
   Multiply = &RigVMFunction_MathDoubleMul(A=RigVMVariableNode() {variable=Thigh_L_Front}.Value, B=RigVMVariableNode() {variable=Crouching}.Value)
   &RigUnit_ModifyTransforms(
    ItemToModify="((Item=(Type=Bone,Name=\"MSTR_jacket_side_l\"),Transform=(Rotation=(X=0,Y=0,Z=0,W=1),Translation=(X=0,Y=0,Z=0),Scale3D=(X=1,Y=1,Z=1))),(Item=(Type=Bone,Name=\"MSTR_jacket_bk_l\"),Transform=(Rotation=(X=0,Y=0,Z=0,W=1),Translation=(X=0,Y=0,Z=0),Scale3D=(X=1,Y=1, ...(truncated)"
    WeightMinimum=0
    Weight=@Multiply.Result
    ExecutePin=@RigVMFunction_Sequence.C
   )
   n11 = RigVMVariableNode() {variable=Thigh_R_Side}
   n12 = RigVMVariableNode() {variable=Crouching}
   > &RigUnit_ModifyTransforms(
    ItemToModify="((Item=(Type=Bone,Name=\"MSTR_jacket_front_r\"),Transform=(Rotation=(X=0,Y=0,Z=0,W=1),Translation=(X=0,Y=0,Z=0),Scale3D=(X=1,Y=1,Z=1))),(Item=(Type=Bone,Name=\"MSTR_jacket_side_r\"),Transform=(Rotation=(X=0,Y=0,Z=0,W=1),Translation=(X=0,Y=0,Z=0),Scale3D=(X=1,Y ...(truncated)"
    WeightMinimum=0
    Weight=&RigVMFunction_MathDoubleRemap(
     SourceMaximum=0.5
     SourceMinimum=0
     TargetMaximum=1
     TargetMinimum=0
     Value=&RigVMFunction_MathDoubleMul(A=@n11.Value, B=@n12.Value).Result
    ).Result
   )
   > &RigUnit_ModifyTransforms(
    ItemToModify="((Item=(Type=Bone,Name=\"base_chain_front\"),Transform=(Rotation=(X=0,Y=0,Z=0.2501,W=-0.9682),Translation=(X=0,Y=0,Z=0),Scale3D=(X=1,Y=1,Z=1))),(Item=(Type=Bone,Name=\"base_chain_back\"),Transform=(Rotation=(X=0,Y=0,Z=-0.28,W=0.96),Translation=(X=0,Y=0,Z=0),Sc ...(truncated)"
    WeightMinimum=0
    Weight=&RigVMFunction_MathDoubleMul(A=RigVMVariableNode() {variable=Thigh_R_Back}.Value, B=RigVMVariableNode() {variable=Crouching}.Value).Result
   )
   > &RigUnit_ModifyTransforms(
    ItemToModify=((Item=(Type=Bone,Name="MSTR_jacket_side_r"),Transform=(Rotation=(X=0,Y=0,Z=0,W=1),Translation=(X=0,Y=0,Z=0),Scale3D=(X=1,Y=1,Z=1))))
    WeightMinimum=0
    Weight=&RigVMFunction_MathDoubleMul(A=RigVMVariableNode() {variable=Thigh_R_Front}.Value, B=RigVMVariableNode() {variable=Crouching}.Value).Result
   )
   n13 = RigVMVariableNode() {variable=Moving2d}
   n14 = &RigVMFunction_AlphaInterp(
    InterpSpeedDecreasing=10
    InterpSpeedIncreasing=10
    Scale=1
    bClampResult=true
    Value=&RigVMFunction_MathBoolToFloat(Value=&RigVMFunction_MathBoolNot(Value=@n13.Value).Result).Result
   )
   RigVMFunction_MathBoolNot = &RigVMFunction_MathBoolNot(Value=RigVMVariableNode() {variable=Sliding}.Value)
   Multiply_3 = RigVMAggregateNode(
    B=@n14.Result
    C=&RigVMFunction_MathBoolToFloat(Value=@RigVMFunction_MathBoolNot.Result).Result
    A=RigVMVariableNode() {variable=Crouching}.Value
   )
   > &RigUnit_ModifyTransforms(
    ItemToModify="((Item=(Type=Bone,Name=\"base_chain_back\"),Transform=(Rotation=(X=-0.0858,Y=-0.2814,Z=-0.2787,W=0.9142),Translation=(X=0,Y=0,Z=0),Scale3D=(X=1,Y=1,Z=1))),(Item=(Type=Bone,Name=\"deform_glute_l\"),Transform=(Rotation=(X=0.0117,Y=0.0389,Z=0.1596,W=-0.9863),Tran ...(truncated)"
    WeightMinimum=0
    Weight=@Multiply_3.Result
   )
   n15 = RigVMVariableNode() {variable=Crouching}
   > &RigUnit_ModifyTransforms(
    ItemToModify="((Item=(Type=Bone,Name=\"pants_corrective_01\"),Transform=(Rotation=(X=0.0566,Y=0,Z=0,W=0.9984),Translation=(X=0.2673,Y=0.7589,Z=0),Scale3D=(X=1,Y=0.5447,Z=1))),(Item=(Type=Bone,Name=\"pants_corrective_02\"),Transform=(Rotation=(X=-0.0174,Y=0.3423,Z=-0.145,W=- ...(truncated)"
    WeightMinimum=0
    Weight=@n15.Value
   )
   > &RigUnit_ModifyTransforms(
    ItemToModify="((Item=(Type=Bone,Name=\"pants_corrective_01\"),Transform=(Rotation=(X=0,Y=0,Z=0,W=1),Translation=(X=0,Y=0,Z=0),Scale3D=(X=1,Y=1,Z=1))),(Item=(Type=Bone,Name=\"pants_corrective_02\"),Transform=(Rotation=(X=0,Y=-0.1246,Z=0,W=-0.9922),Translation=(X=-0.0885,Y=0, ...(truncated)"
    WeightMinimum=0
    Weight=&RigVMFunction_MathDoubleMul(B=@n15.Value, A=RigVMVariableNode() {variable=Thigh_R_Front}.Value).Result
   )
   n16 = &RigVMFunction_AlphaInterp(
    InterpSpeedDecreasing=10
    InterpSpeedIncreasing=10
    Scale=1
    bClampResult=true
    Value=&RigVMFunction_MathBoolToFloat(Value=RigVMVariableNode() {variable=Moving2d}.Value).Result
   )
   > &RigUnit_ModifyTransforms(
    ItemToModify=((Item=(Type=Bone,Name="deform_glute_l"),Transform=(Rotation=(X=0,Y=0,Z=-0.1728,W=0.985),Translation=(X=0.6144,Y=0,Z=0),Scale3D=(X=1,Y=1,Z=1))))
    WeightMinimum=0
    Weight=&RigVMFunction_MathDoubleMul(B=@n16.Result, A=RigVMVariableNode() {variable=Crouching}.Value).Result
   )
   n17 = &RigVMFunction_AlphaInterp(
    InterpSpeedDecreasing=10
    InterpSpeedIncreasing=10
    Scale=1
    bClampResult=true
    Value=&RigVMFunction_MathBoolToFloat(Value=RigVMVariableNode() {variable=CrouchingBool}.Value).Result
   )
   Remap = &RigVMFunction_MathDoubleRemap(
    SourceMaximum=0.6
    SourceMinimum=0
    TargetMaximum=1
    TargetMinimum=0
    Value=&RigVMFunction_MathDoubleRemap(
     SourceMaximum=39
     SourceMinimum=43
     TargetMaximum=1
     TargetMinimum=0
     Value=&RigUnit_GetRelativeTransformForItem(Child=(Type=Bone,Name="pelvis"), Parent=(Type=Bone,Name="root"), bChildInitial=False, bParentInitial=False).RelativeTransform.Translation.Z
    ).Result
   )
   &RigUnit_ModifyTransforms(
    ItemToModify=((Item=(Type=Bone,Name="base_chain_back"),Transform=(Rotation=(X=0,Y=0,Z=0.1224,W=-0.9925),Translation=(X=0,Y=0,Z=0),Scale3D=(X=1,Y=1,Z=1))))
    WeightMinimum=0
    Weight=&RigVMFunction_MathDoubleMul(B=@n17.Result, A=@Remap.Result).Result
    ExecutePin=@RigVMFunction_Sequence.D
   )
  ]}
 New Function_ContainedGraph {nodes [
   n1 = RigVMFunctionEntryNode()
   RigVMFunction_MathQuaternionToEuler = &RigVMFunction_MathQuaternionToEuler(
    RotationOrder=YXZ
    Value=&RigUnit_GetRelativeTransformForItem(
     Child=(Type=Bone,Name="None")
     Parent=(Type=Bone,Name="None")
     bChildInitial=False
     bParentInitial=False
     Child.Name=@n1.Child
     Parent.Name=@n1.Parent
    ).RelativeTransform.Rotation
   )
   Remap = &RigVMFunction_MathDoubleRemap(
    TargetMaximum=1
    TargetMinimum=0
    Value=@RigVMFunction_MathQuaternionToEuler.Result.Y
    SourceMaximum=@n1.SourceMax
    SourceMinimum=@n1.SourceMin
   )
   > RigVMFunctionReturnNode(DeltaY=@Remap.Result)
  ]}
 RigVMFunctionLibrary {nodes [
   "Collar  and NecklaceSetup" = RigVMCollapseNode(ExecuteContext=())
   Constraints Setup = RigVMCollapseNode(ExecuteContext=())
   DeltaY = RigVMCollapseNode(ExecuteContext=())
   Extra Setup = RigVMCollapseNode(ExecuteContext=())
   Thighs NO RB Setup = RigVMCollapseNode(ExecuteContext=())
   Thighs RB Setup = RigVMCollapseNode(ExecuteContext=())
  ]}
 RigVMFunction_Sequence_1_ContainedGraph {nodes [
   n1 = RigVMFunctionEntryNode()
   RigVMFunction_Sequence = &RigVMFunction_Sequence(ExecuteContext=@n1.ExecuteContext)
   n2 = &RigVMFunction_Sequence(ExecuteContext=@RigVMFunction_Sequence.B)
   RigVMFunctionReturnNode(A=@RigVMFunction_Sequence.A, B=@n2.A, C=@n2.B)
  ]}
 RigVMFunction_Sequence_1_ContainedGraph {nodes [
   n1 = RigVMFunctionEntryNode()
   RigVMFunction_Sequence = &RigVMFunction_Sequence(ExecuteContext=@n1.ExecuteContext)
   n2 = &RigVMFunction_Sequence(ExecuteContext=@RigVMFunction_Sequence.B)
   n3 = &RigVMFunction_Sequence(ExecuteContext=@n2.B)
   n4 = &RigVMFunction_Sequence(ExecuteContext=@n3.B)
   RigVMFunctionReturnNode(A=@RigVMFunction_Sequence.A, B=@n2.A, C=@n3.A, D=@n4.A, E=@n4.B)
  ]}
 RigVMFunction_Sequence_1_ContainedGraph {nodes [
   n1 = RigVMFunctionEntryNode()
   RigVMFunction_Sequence = &RigVMFunction_Sequence(ExecuteContext=@n1.ExecuteContext)
   n2 = &RigVMFunction_Sequence(ExecuteContext=@RigVMFunction_Sequence.B)
   n3 = &RigVMFunction_Sequence(ExecuteContext=@n2.B)
   RigVMFunctionReturnNode(A=@RigVMFunction_Sequence.A, B=@n2.A, C=@n3.A, D=@n3.B)
  ]}
 RigVMFunction_Sequence_1_ContainedGraph {nodes [
   n1 = RigVMFunctionEntryNode()
   RigVMFunction_Sequence = &RigVMFunction_Sequence(ExecuteContext=@n1.ExecuteContext)
   n2 = &RigVMFunction_Sequence(ExecuteContext=@RigVMFunction_Sequence.B)
   n3 = &RigVMFunction_Sequence(ExecuteContext=@n2.B)
   n4 = &RigVMFunction_Sequence(ExecuteContext=@n3.B)
   RigVMFunctionReturnNode(A=@RigVMFunction_Sequence.A, B=@n2.A, C=@n3.A, D=@n4.A, E=@n4.B)
  ]}
 RigVMModel {nodes [
   RigUnit_BeginExecution = &RigUnit_BeginExecution()
   /Script/FortniteGame.RigUnit_FortDeformRig(
    Bool=(bUpperArm=true,bElbow=true,bWrist=true,bGroin=true,bGlute=true,bKnee=true,bPec=true,bBicep=true,bLat=False)
    Curve="(bLUpperArmUpAlpha=false,bRUpperArmUpAlpha=true,bLUpperArmFrAlpha=False,bRUpperArmFrAlpha=False,bLUpperArmBkAlpha=False,bRUpperArmBkAlpha=False,bLWristUpAlpha=False,bRWristUpAlpha=False,bLWristDnAlpha=False,bRWristDnAlpha=False,bLKneeAlpha=False,bRKneeAlph ...(truncated)"
    Debug=(bUpperArmDebug=false,bWristDebug=false,bLatDebug=False,bGroinGluteDebug=False)
    Value="(UpperArmUp=(Rotation=(X=0.0,Y=-0.0,Z=0.0,W=1.0),Translation=(X=0.0,Y=0.0,Z=8),Scale3D=(X=0.70,Y=1.0,Z=1.0)),UpperArmFwdFr=(Rotation=(X=0.0,Y=-0.0,Z=0.0,W=1.0),Translation=(X=0.0,Y=10.0,Z=0.0),Scale3D=(X=0.50,Y=1.0,Z=1.0)),UpperArmFwdBk=(Rotation=(X=0.0,Y= ...(truncated)"
    ExecutePin=@RigUnit_BeginExecution.ExecutePin
   ) {is=struct RigVMUnitNode}
   > Constraints Setup = RigVMFunctionReferenceNode()
   > "Collar  and NecklaceSetup" = RigVMFunctionReferenceNode()
   > Extra Setup = RigVMFunctionReferenceNode()
   RigVMFunction_ControlFlowBranch = &RigVMFunction_ControlFlowBranch(ExecuteContext=@"Extra Setup".ExecuteContext, Condition=RigVMVariableNode() {variable=RB Enable}.Value)
   Thighs NO RB Setup = RigVMFunctionReferenceNode(ExecuteContext=@RigVMFunction_ControlFlowBranch.False)
   Thighs RB Setup = RigVMFunctionReferenceNode(ExecuteContext=@RigVMFunction_ControlFlowBranch.True)
   &RigUnit_PrepareForExecution()
  ]}
]
hierarchy [
 root {loc=<0 0 0>, rot=<0 0 0 1>}
 pelvis {parent=root, loc=<0 -0.7071 93.1191>, rot=<0.0127 0.707 -0.0127 -0.707>}
 spine_01 {parent=pelvis, loc=<6.7494 -0.6388 0>, rot=<0 0 0.034 -0.9994>}
 spine_02 {parent=spine_01, loc=<7.2518 -0.035 0>, rot=<0 0 0.0238 -0.9997>}
 spine_03 {parent=spine_02, loc=<6.0104 0 0>, rot=<0 0 0.0149 -0.9999>}
 spine_04 {parent=spine_03, loc=<8.1024 0 0>, rot=<0 0 -0.0132 -0.9999>}
 spine_05 {parent=spine_04, loc=<10.0146 0 0>, rot=<0 0 -0.0015 -1>}
 neck_01 {parent=spine_05, loc=<11.3138 0.1529 0>, rot=<0 0 -0.2003 -0.9797>}
 neck_02 {parent=neck_01, loc=<4.5627 0 0>, rot=<0 0 0.0504 -0.9987>}
 head {parent=neck_02, loc=<5.3914 0 0>, rot=<0 0 0.115 -0.9934>}
 attach_cape {parent=spine_05, loc=<9.0169 -11.7111 -0.0007>, rot=<0.0616 -0.7044 -0.0616 -0.7044>}
 cape {parent=attach_cape, type=Null, loc=<0 1 0.2>, rot=<0.0436 0 0 0.999>}
 attach_backpack {parent=spine_05, loc=<-0.1221 -13.4699 0.0003>, rot=<-0.0093 -0.707 0.0093 -0.707>}
 Backpack_BR {parent=attach_backpack, type=Null, loc=<0 1 0>, rot=<-0.0523 0 0 0.9986>}
 clavicle_l {parent=spine_05, loc=<4.3575 -0.1466 -2.3223>, rot=<0.0465 -0.7289 -0.009 -0.6829>}
 upperarm_l {parent=clavicle_l, loc=<15.5111 0 0>, rot=<0.0226 -0.3234 -0.0088 -0.9459>}
 lowerarm_l {parent=upperarm_l, loc=<26.6514 0 0>, rot=<0.0391 -0.0099 -0.2453 -0.9686>}
 hand_l {parent=lowerarm_l, loc=<22.3491 0 0>, rot=<-0.7081 -0.0507 0.0975 -0.6975>}
 thumb_01_l {parent=hand_l, loc=<1.3118 -0.3191 -2.459>, rot=<0.6546 -0.4042 0.0664 -0.6354>}
 thumb_02_l {parent=thumb_01_l, loc=<5.9336 -0.0448 0.0401>, rot=<-0.0037 0.0213 0.298 -0.9543>}
 thumb_03_l {parent=thumb_02_l, loc=<2.2223 0.3682 -0.0907>, rot=<0.0118 -0.013 0.0066 -0.9998>}
 index_metacarpal_l {parent=hand_l, loc=<1.6535 0.0183 -2.96>, rot=<0.0158 -0.0124 -0.0173 -0.9996>}
 index_01_l {parent=index_metacarpal_l, loc=<8.3008 0.0357 -0.0524>, rot=<-0.0044 0.0485 0.0924 -0.9945>}
 index_02_l {parent=index_01_l, loc=<3.3587 -0.3004 0.0452>, rot=<-0.002 0.0027 0.3225 -0.9466>}
 index_03_l {parent=index_02_l, loc=<2.5578 -0.123 -0.0252>, rot=<-0.002 0.0067 0.2558 -0.9667>}
 middle_metacarpal_l {parent=hand_l, loc=<2.1856 -0.1725 -0.5984>, rot=<0.006 0.0168 -0.0329 -0.9993>}
 middle_01_l {parent=middle_metacarpal_l, loc=<7.3275 -0.2447 0.0129>, rot=<-0.0208 0.0042 0.0858 -0.9961>}
 middle_02_l {parent=middle_01_l, loc=<4.0123 0.1401 0.1175>, rot=<-0.0218 0.0258 0.3512 -0.9357>}
 middle_03_l {parent=middle_02_l, loc=<2.6077 0.0534 0.1373>, rot=<0.0082 0.0188 0.1708 -0.9851>}
 ring_metacarpal_l {parent=hand_l, loc=<2.2074 -0.3625 1.3638>, rot=<-0.0284 0.0722 -0.0458 -0.9959>}
 ring_01_l {parent=ring_metacarpal_l, loc=<7.0295 -0.6197 -0.1446>, rot=<-0.0489 -0.0252 0.1892 -0.9804>}
 ring_02_l {parent=ring_01_l, loc=<3.6174 0.4647 -0.1027>, rot=<0 0 0.2806 -0.9598>}
 ring_03_l {parent=ring_02_l, loc=<2.4152 0.114 0.0199>, rot=<0 0 0.1189 -0.9929>}
 pinky_metacarpal_l {parent=hand_l, loc=<1.7616 -0.9753 3.3814>, rot=<-0.0397 0.0814 -0.069 -0.9935>}
 pinky_01_l {parent=pinky_metacarpal_l, loc=<7.1709 -0.4831 -0.1094>, rot=<-0.0654 -0.0035 0.2242 -0.9723>}
 pinky_02_l {parent=pinky_01_l, loc=<2.8274 0.0893 -0.1043>, rot=<0.0037 0.0011 0.2854 -0.9584>}
 pinky_03_l {parent=pinky_02_l, loc=<2.2046 0.0307 -0.0078>, rot=<-0.0002 0.0098 0.1445 -0.9895>}
 deform_wrist_dn_base_l {parent=hand_l, loc=<0 -2.2342 0>, rot=<0 0 0 -1>}
 deform_wrist_dn_l {parent=deform_wrist_dn_base_l, loc=<0 0 0>, rot=<0 0 0 -1>}
 deform_wrist_up_base_l {parent=hand_l, loc=<0 2.2682 0>, rot=<0 0 0 -1>}
 deform_wrist_up_l {parent=deform_wrist_up_base_l, loc=<0 0 0>, rot=<0 0 0 -1>}
 lowerarm_twist_02_l {parent=lowerarm_l, loc=<9.6199 0 0>, rot=<0 0 0 -1>}
 lowerarm_twist_01_l {parent=lowerarm_l, loc=<17.8549 0 0>, rot=<0 0 0 -1>}
 deform_elbow_in_base_l {parent=lowerarm_l, loc=<0.2112 5.259 0.8274>, rot=<0 0 0 -1>}
 deform_elbow_in_l {parent=deform_elbow_in_base_l, loc=<0 0 0>, rot=<0 0 0 -1>}
 deform_elbow_out_base_l {parent=lowerarm_l, loc=<0 -4 0>, rot=<0 0 0 -1>}
 deform_elbow_out_l {parent=deform_elbow_out_base_l, loc=<0 0 0>, rot=<0 0 0 1>}
 upperarm_twist_01_l {parent=upperarm_l, loc=<0.0041 0 0>, rot=<0 0 0 -1>}
 deform_upperarm_bk_base_l {parent=upperarm_twist_01_l, loc=<1.0813 -7.2193 -1.9453>, rot=<0 0 0 1>}
 deform_upperarm_bk_l {parent=deform_upperarm_bk_base_l, loc=<0 0 0>, rot=<0 0 0 1>}
 deform_upperarm_fr_base_l {parent=upperarm_twist_01_l, loc=<-0.3691 7.0506 -2.765>, rot=<0 0 0 1>}
 deform_upperarm_fr_l {parent=deform_upperarm_fr_base_l, loc=<0 0 0>, rot=<0 0 0 1>}
 deform_upperarm_up_base_l {parent=upperarm_twist_01_l, loc=<-3.2216 -1.1391 3.3311>, rot=<0 0 0 1>}
 deform_upperarm_up_l {parent=deform_upperarm_up_base_l, loc=<0 0 0>, rot=<0 0 0 1>}
 upperarm_twist_02_l {parent=upperarm_l, loc=<13.7941 0 0>, rot=<0 0 0 -1>}
 clavicle_r {parent=spine_05, loc=<4.3575 -0.1466 2.3223>, rot=<0.7289 0.0465 -0.6829 0.009>}
 upperarm_r {parent=clavicle_r, loc=<-15.5111 0 0>, rot=<0.0226 -0.3234 -0.0088 -0.9459>}
 lowerarm_r {parent=upperarm_r, loc=<-26.6514 0 0>, rot=<0.0391 -0.0099 -0.2453 -0.9686>}
 hand_r {parent=lowerarm_r, loc=<-22.3491 0 0>, rot=<-0.7081 -0.0507 0.0975 -0.6975>}
 thumb_01_r {parent=hand_r, loc=<-1.3118 0.3191 2.459>, rot=<0.6546 -0.4042 0.0664 -0.6354>}
 thumb_02_r {parent=thumb_01_r, loc=<-5.9336 0.0448 -0.0401>, rot=<-0.0037 0.0213 0.298 -0.9543>}
 thumb_03_r {parent=thumb_02_r, loc=<-2.2223 -0.3682 0.0907>, rot=<0.0118 -0.013 0.0066 -0.9998>}
 index_metacarpal_r {parent=hand_r, loc=<-1.6535 -0.0183 2.96>, rot=<0.0158 -0.0124 -0.0173 -0.9996>}
 index_01_r {parent=index_metacarpal_r, loc=<-8.3008 -0.0357 0.0524>, rot=<-0.0044 0.0485 0.0924 -0.9945>}
 index_02_r {parent=index_01_r, loc=<-3.3587 0.3004 -0.0452>, rot=<-0.002 0.0027 0.3225 -0.9466>}
 index_03_r {parent=index_02_r, loc=<-2.5578 0.123 0.0252>, rot=<-0.002 0.0067 0.2558 -0.9667>}
 middle_metacarpal_r {parent=hand_r, loc=<-2.1856 0.1725 0.5984>, rot=<0.006 0.0168 -0.0329 -0.9993>}
 middle_01_r {parent=middle_metacarpal_r, loc=<-7.3275 0.2447 -0.0129>, rot=<-0.0208 0.0042 0.0858 -0.9961>}
 middle_02_r {parent=middle_01_r, loc=<-4.0123 -0.1401 -0.1175>, rot=<-0.0218 0.0258 0.3512 -0.9357>}
 middle_03_r {parent=middle_02_r, loc=<-2.6077 -0.0534 -0.1373>, rot=<0.0082 0.0188 0.1708 -0.9851>}
 ring_metacarpal_r {parent=hand_r, loc=<-2.2074 0.3625 -1.3638>, rot=<-0.0284 0.0722 -0.0458 -0.9959>}
 ring_01_r {parent=ring_metacarpal_r, loc=<-7.0295 0.6197 0.1446>, rot=<-0.0489 -0.0252 0.1892 -0.9804>}
 ring_02_r {parent=ring_01_r, loc=<-3.6174 -0.4647 0.1027>, rot=<0 0 0.2806 -0.9598>}
 ring_03_r {parent=ring_02_r, loc=<-2.4152 -0.114 -0.0199>, rot=<0 0 0.1189 -0.9929>}
 pinky_metacarpal_r {parent=hand_r, loc=<-1.7616 0.9753 -3.3814>, rot=<-0.0397 0.0814 -0.069 -0.9935>}
 pinky_01_r {parent=pinky_metacarpal_r, loc=<-7.1709 0.4831 0.1094>, rot=<-0.0654 -0.0035 0.2242 -0.9723>}
 pinky_02_r {parent=pinky_01_r, loc=<-2.8274 -0.0893 0.1043>, rot=<0.0037 0.0011 0.2854 -0.9584>}
 pinky_03_r {parent=pinky_02_r, loc=<-2.2046 -0.0307 0.0078>, rot=<-0.0002 0.0098 0.1445 -0.9895>}
 deform_wrist_dn_base_r {parent=hand_r, loc=<-0.0004 2.2345 -0.0001>, rot=<0 0 0 -1>}
 deform_wrist_dn_r {parent=deform_wrist_dn_base_r, loc=<0 0 0>, rot=<0 0 0 -1>}
 deform_wrist_up_base_r {parent=hand_r, loc=<0.0003 -2.2684 0.0001>, rot=<0 0 0 -1>}
 deform_wrist_up_r {parent=deform_wrist_up_base_r, loc=<0 0 0>, rot=<0 0 0 -1>}
 lowerarm_twist_02_r {parent=lowerarm_r, loc=<-9.6199 0 0>, rot=<0 0 0 -1>}
 lowerarm_twist_01_r {parent=lowerarm_r, loc=<-17.8549 0 0>, rot=<0 0 0 -1>}
 deform_elbow_in_base_r {parent=lowerarm_r, loc=<-0.2111 -5.2591 -0.8275>, rot=<0 0 0 -1>}
 deform_elbow_in_r {parent=deform_elbow_in_base_r, loc=<-0.0001 0 0.0002>, rot=<0 0 0 1>}
 deform_elbow_out_base_r {parent=lowerarm_r, loc=<0 4 -0.0001>, rot=<0 0 0 -1>}
 deform_elbow_out_r {parent=deform_elbow_out_base_r, loc=<0 0 0>, rot=<0 0 0 1>}
 upperarm_twist_01_r {parent=upperarm_r, loc=<-0.0041 0 0>, rot=<0 0 0 -1>}
 deform_upperarm_bk_base_r {parent=upperarm_twist_01_r, loc=<-1.0814 7.2192 1.9455>, rot=<0 0 0 1>}
 deform_upperarm_bk_r {parent=deform_upperarm_bk_base_r, loc=<0 0 0>, rot=<0 0 0 1>}
 deform_upperarm_fr_base_r {parent=upperarm_twist_01_r, loc=<0.3689 -7.0506 2.7652>, rot=<0 0 0 1>}
 deform_upperarm_fr_r {parent=deform_upperarm_fr_base_r, loc=<0 0 0>, rot=<0 0 0 1>}
 deform_upperarm_up_base_r {parent=upperarm_twist_01_r, loc=<3.2214 1.1391 -3.3308>, rot=<0 0 0 1>}
 deform_upperarm_up_r {parent=deform_upperarm_up_base_r, loc=<0 0 0>, rot=<0 0 0 -1>}
 upperarm_twist_02_r {parent=upperarm_r, loc=<-13.7941 0 0>, rot=<0 0 0 -1>}
 base_collar_bk_l {parent=spine_05, loc=<13.1857 -2.2696 -3.6226>, rot=<0.4977 0.02 -0.0348 -0.8664>}
 dyn_collar_bk_01_l {parent=base_collar_bk_l, loc=<0 0 0>, rot=<0 0 0 1>}
 base_collar_fr_l {parent=spine_05, loc=<7.793 5.7082 -7.6456>, rot=<0 0 -0.0401 -0.9992>}
 dyn_collar_fr_01_l {parent=base_collar_fr_l, loc=<0 0 0>, rot=<0 0 0 1>}
 base_collar_sd_l {parent=spine_05, loc=<10.6417 1.3473 -6.0657>, rot=<0.1967 0.0079 -0.0393 -0.9796>}
 dyn_collar_sd_01_l {parent=base_collar_sd_l, loc=<0 0 0>, rot=<0 0 0 1>}
 base_collar_bk_c {parent=spine_05, loc=<13.4739 -3.6365 0>, rot=<0.7065 0.0284 -0.0284 -0.7065>}
 dyn_collar_bk_01_c {parent=base_collar_bk_c, loc=<0 0 0>, rot=<0 0 0 1>}
 base_collar_bk_r {parent=spine_05, loc=<13.1853 -2.2696 3.6226>, rot=<0.02 -0.4977 0.8664 -0.0348>}
 dyn_collar_bk_01_r {parent=base_collar_bk_r, loc=<0 0 0>, rot=<0 0 0 1>}
 base_collar_fr_r {parent=spine_05, loc=<7.7926 5.7082 7.6456>, rot=<0 0 0.9992 -0.0401>}
 dyn_collar_fr_01_r {parent=base_collar_fr_r, loc=<0 0 0>, rot=<0 0 0 1>}
 base_collar_sd_r {parent=spine_05, loc=<10.6413 1.3473 6.0657>, rot=<0.0079 -0.1967 0.9796 -0.0393>}
 dyn_collar_sd_01_r {parent=base_collar_sd_r, loc=<0 0 0>, rot=<0 0 0 1>}
 base_necklace_a {parent=spine_05, loc=<8.7105 6.1362 0>, rot=<0.1767 0.6847 0.6847 0.1767>}
 dyn_necklace_01_a {parent=base_necklace_a, loc=<0 0 0>, rot=<0 0 0 1>}
 dyn_necklace_02_a {parent=dyn_necklace_01_a, loc=<5.2405 0 0>, rot=<0 0 0 1>}
 base_necklace_b {parent=spine_05, loc=<9.3578 5.7784 0>, rot=<0.1767 0.6847 0.6847 0.1767>}
 dyn_necklace_01_b {parent=base_necklace_b, loc=<0 0 0>, rot=<0 0 0 1>}
 dyn_necklace_02_b {parent=dyn_necklace_01_b, loc=<5.2405 0 0>, rot=<0 0 0 1>}
 deform_pec_aim_base_l {parent=spine_04, loc=<7.8732 12.6388 -2.4123>, rot=<0.0102 -0.4602 0.3979 -0.7936>}
 deform_pec_aim_l {parent=deform_pec_aim_base_l, loc=<0 0 0>, rot=<0 0 0 1>}
 deform_pec_base_l {parent=deform_pec_aim_l, loc=<10.766 0 0>, rot=<0 0 0 1>}
 deform_pec_l {parent=deform_pec_base_l, loc=<0.055 0 0>, rot=<0 0 0 -1>}
 deform_pec_aim_base_r {parent=spine_04, loc=<7.8727 12.6388 2.4123>, rot=<-0.4602 -0.0102 0.7936 0.3979>}
 deform_pec_aim_r {parent=deform_pec_aim_base_r, loc=<0 0 0>, rot=<0 0 0 1>}
 deform_pec_base_r {parent=deform_pec_aim_r, loc=<-10.7662 -0.0006 -0.0007>, rot=<0 0 0 1>}
 deform_pec_r {parent=deform_pec_base_r, loc=<0 0 0>, rot=<0 0 0 -1>}
 MSTR_jacket_side_r {parent=spine_03, loc=<5.6856 -0.5973 14.3536>, rot=<0.0417 -0.7061 0.7057 -0.0401>}
 dyn_jacket_side_01_r {parent=MSTR_jacket_side_r, loc=<0 0 0>, rot=<0 0 0 1>}
 dyn_jacket_side_02_r {parent=dyn_jacket_side_01_r, loc=<11.0432 0 0>, rot=<0 -0.0033 -0.012 -0.9999>}
 dyn_jacket_side_03_r {parent=dyn_jacket_side_02_r, loc=<8.6386 0.0001 0>, rot=<0 0 0.0205 -0.9998>}
 MSTR_jacket_front_r {parent=spine_03, loc=<6.3324 8.3373 10.6025>, rot=<0.0232 -0.2575 0.9656 -0.0273>}
 dyn_jacket_front_01_r {parent=MSTR_jacket_front_r, loc=<0 0 0>, rot=<0 0 0 1>}
 dyn_jacket_front_02_r {parent=dyn_jacket_front_01_r, loc=<12.6721 -0.1261 0.0099>, rot=<-0.0005 -0.0024 0.0087 -1>}
 dyn_jacket_front_03_r {parent=dyn_jacket_front_02_r, loc=<8.655 -0.0797 -0.0004>, rot=<0 0 0.0497 -0.9988>}
 MSTR_jacket_side_l {parent=spine_03, loc=<5.6851 -0.5974 -14.3536>, rot=<0.0417 -0.7061 -0.7057 0.0401>}
 dyn_jacket_side_01_l {parent=MSTR_jacket_side_l, loc=<0 0 0>, rot=<0 0 0 1>}
 dyn_jacket_side_02_l {parent=dyn_jacket_side_01_l, loc=<11.0425 0 0>, rot=<0 0.0033 -0.012 -0.9999>}
 dyn_jacket_side_03_l {parent=dyn_jacket_side_02_l, loc=<8.6388 0 0>, rot=<0 0 0.0205 -0.9998>}
 MSTR_jacket_front_l {parent=spine_03, loc=<6.332 8.3373 -10.6025>, rot=<-0.0232 0.2575 0.9656 -0.0273>}
 dyn_jacket_front_01_l {parent=MSTR_jacket_front_l, loc=<0 0 0>, rot=<0 0 0 1>}
 dyn_jacket_front_02_l {parent=dyn_jacket_front_01_l, loc=<12.672 -0.1261 -0.0099>, rot=<0.0005 0.0024 0.0087 -1>}
 dyn_jacket_front_03_l {parent=dyn_jacket_front_02_l, loc=<8.6547 -0.0797 0.0004>, rot=<0 0 0.0497 -0.9988>}
 MSTR_jacket_bk_l {parent=spine_01, loc=<8.565 -8.7805 -8.7605>, rot=<-0.0173 -0.9239 -0.3822 0.0032>}
 dyn_jacket_bk_01_l {parent=MSTR_jacket_bk_l, loc=<0 0 0>, rot=<0 0 0 1>}
 dyn_jacket_bk_02_l {parent=dyn_jacket_bk_01_l, loc=<9.4897 0 0>, rot=<0 0 0.0201 -0.9998>}
 MSTR_jacket_bk_r {parent=spine_01, loc=<8.565 -8.7806 8.7605>, rot=<-0.0173 -0.9239 0.3822 -0.0032>}
 dyn_jacket_bk_01_r {parent=MSTR_jacket_bk_r, loc=<0 0 0>, rot=<0 0 0 1>}
 dyn_jacket_bk_02_r {parent=dyn_jacket_bk_01_r, loc=<9.4898 0 0>, rot=<0 0 0.0201 -0.9998>}
 MSTR_jacket_bk {parent=spine_01, loc=<8.4235 -10.3322 0>, rot=<0.0152 -0.9999 0 0>}
 dyn_jacket_bk_01 {parent=MSTR_jacket_bk, loc=<0 0 0>, rot=<0 0 0 1>}
 dyn_jacket_bk_02 {parent=dyn_jacket_bk_01, loc=<9.4959 0 0>, rot=<0 0 0 1>}
 thigh_l {parent=pelvis, loc=<-2.5693 -0.0109 -7.7501>, rot=<0.0012 0.0572 0.0155 -0.9982>}
 calf_l {parent=thigh_l, loc=<-39.0484 0 0>, rot=<0.0006 -0.0161 -0.0391 -0.9991>}
 foot_l {parent=calf_l, loc=<-42.561 0.1077 0.4292>, rot=<0.0418 -0.0455 0.0343 -0.9975>}
 ball_l {parent=foot_l, loc=<-6.1599 13.8264 0.6806>, rot=<0.0001 -0.0001 -0.7063 -0.7079>}
 calf_twist_02_l {parent=calf_l, loc=<-18.8064 -0.0967 0.0397>, rot=<0.0237 0.0095 -0.0035 -0.9997>}
 calf_twist_01_l {parent=calf_l, loc=<-35.813 -0.0967 0.0397>, rot=<0.0237 0.0095 -0.0035 -0.9997>}
 base_bagginess_fr_l {parent=calf_twist_01_l, loc=<-1.3777 4.2736 -0.0895>, rot=<0.5603 0.4885 0.4636 0.4821>}
 dyn_bagginess_fr_l {parent=base_bagginess_fr_l, loc=<0 0 0>, rot=<0 0 0 1>}
 base_bagginess_sd_l {parent=calf_twist_01_l, loc=<0.8223 0.5932 -1.8853>, rot=<-0.0508 -0.7417 0.0131 -0.6687>}
 dyn_bagginess_sd_l {parent=base_bagginess_sd_l, loc=<0 0 0>, rot=<0 0 0 1>}
 base_bagginess_bk_l {parent=calf_twist_01_l, loc=<-1.1769 -2.2343 1.2908>, rot=<0.4885 -0.5603 0.4821 -0.4636>}
 dyn_bagginess_bk_l {parent=base_bagginess_bk_l, loc=<0 0 0>, rot=<0 0 0 1>}
 base_bagginess_in_l {parent=calf_twist_01_l, loc=<4.9602 1.1221 2.27>, rot=<0.7417 -0.0508 0.6687 0.0131>}
 dyn_bagginess_in_l {parent=base_bagginess_in_l, loc=<0 0 0>, rot=<0 0 0 1>}
 deform_knee_in_base_l {parent=calf_l, loc=<1.298 -5.2957 -0.0496>, rot=<0 0 0 -1>}
 deform_knee_in_l {parent=deform_knee_in_base_l, loc=<0 0 0>, rot=<0 0 0 1>}
 deform_knee_out_base_l {parent=calf_l, loc=<0 4 0>, rot=<0 0 0 -1>}
 deform_knee_out_l {parent=deform_knee_out_base_l, loc=<0 0 0>, rot=<0 0 0 1>}
 thigh_twist_01_l {parent=thigh_l, loc=<-21.6303 0 0>, rot=<0.0013 0 0 -1>}
 deform_groin_base_l {parent=thigh_l, loc=<-2.1908 8.1142 -1.7442>, rot=<0 0 0 -1>}
 deform_groin_l {parent=deform_groin_base_l, loc=<0 0 0>, rot=<0 0 0 -1>}
 thigh_r {parent=pelvis, loc=<-2.5693 -0.0109 7.7501>, rot=<0.0572 -0.0012 0.9982 0.0155>}
 calf_r {parent=thigh_r, loc=<39.0484 0 0>, rot=<0.0006 -0.0161 -0.0391 -0.9991>}
 foot_r {parent=calf_r, loc=<42.561 -0.1077 -0.4292>, rot=<0.0418 -0.0455 0.0343 -0.9975>}
 ball_r {parent=foot_r, loc=<6.1599 -13.8264 -0.6806>, rot=<0.0001 -0.0001 -0.7063 -0.7079>}
 base_laces_l {parent=foot_r, loc=<-6.8793 -4.4112 -4.4125>, rot=<0.9147 0.096 -0.1424 0.3658>}
 dyn_laces_01_l {parent=base_laces_l, loc=<0 0 0>, rot=<0 0 0 1>}
 dyn_laces_02_l {parent=dyn_laces_01_l, loc=<1.7774 0 0>, rot=<0 -0.1945 0 -0.9809>}
 dyn_laces_03_l {parent=dyn_laces_02_l, loc=<2.9004 0 0>, rot=<0 0 0 1>}
 base_laces_r {parent=foot_r, loc=<-7.2749 -4.7189 2.9914>, rot=<-0.3166 0.1591 -0.0607 -0.9331>}
 dyn_laces_01_r {parent=base_laces_r, loc=<0 0 0>, rot=<0 0 0 1>}
 dyn_laces_02_r {parent=dyn_laces_01_r, loc=<2.2724 0 0>, rot=<0 -0.1028 -0.042 -0.9938>}
 dyn_laces_03_r {parent=dyn_laces_02_r, loc=<1.8763 0 0>, rot=<0 0 0 1>}
 calf_twist_02_r {parent=calf_r, loc=<18.8064 0.0967 -0.0397>, rot=<0.0237 0.0095 -0.0035 -0.9997>}
 base_bagginess_fr_r {parent=calf_twist_02_r, loc=<4.671 -2.825 0.3458>, rot=<-0.4642 0.4816 0.5597 -0.4892>}
 dyn_bagginess_fr_r {parent=base_bagginess_fr_r, loc=<0 0 0>, rot=<0 0 0 -1>}
 base_bagginess_sd_r {parent=calf_twist_02_r, loc=<3.0242 0.9434 -4.0436>, rot=<-0.6688 0.0123 0.7417 0.0498>}
 dyn_bagginess_sd_r {parent=base_bagginess_sd_r, loc=<0 0 0>, rot=<0 0 0 -1>}
 base_bagginess_bk_r {parent=calf_twist_02_r, loc=<3.6456 3.5314 0.3511>, rot=<0.4816 0.4642 -0.4892 -0.5597>}
 dyn_bagginess_bk_r {parent=base_bagginess_bk_r, loc=<0 0 0>, rot=<0 0 0 -1>}
 base_bagginess_in_r {parent=calf_twist_02_r, loc=<-0.056 1.0456 4.1271>, rot=<0.0123 0.6688 0.0498 -0.7417>}
 dyn_bagginess_in_r {parent=base_bagginess_in_r, loc=<0 0 0>, rot=<0 0 0 -1>}
 calf_twist_01_r {parent=calf_r, loc=<35.813 0.0967 -0.0397>, rot=<0.0237 0.0095 -0.0035 -0.9997>}
 base_boot_tongue {parent=calf_twist_01_r, loc=<-4.276 -4.4707 -0.013>, rot=<-0.6071 0.3182 0.4002 -0.6083>}
 dyn_boot_tongue_01 {parent=base_boot_tongue, loc=<0 0 0>, rot=<0 0 0 1>}
 dyn_boot_tongue_02 {parent=dyn_boot_tongue_01, loc=<0 0 6.0516>, rot=<0 0 0 1>}
 deform_knee_in_base_r {parent=calf_r, loc=<-1.298 5.2958 0.0496>, rot=<0 0 0 1>}
 deform_knee_in_r {parent=deform_knee_in_base_r, loc=<0 0 0>, rot=<0 0 0 -1>}
 deform_knee_out_base_r {parent=calf_r, loc=<0 -4 0>, rot=<0 0 0 1>}
 deform_knee_out_r {parent=deform_knee_out_base_r, loc=<0 0 0>, rot=<0 0 0 -1>}
 thigh_twist_01_r {parent=thigh_r, loc=<21.6303 0 0>, rot=<0.0013 0 0 -1>}
 deform_groin_base_r {parent=thigh_r, loc=<2.1908 -8.1142 1.7442>, rot=<0 0 0 -1>}
 deform_groin_r {parent=deform_groin_base_r, loc=<0 0 0>, rot=<0 0 0 -1>}
 deform_glute_base_l {parent=pelvis, loc=<-8.8064 -8.8222 -8.4287>, rot=<0.0275 0.0501 -0.4472 -0.8926>}
 deform_glute_l {parent=deform_glute_base_l, loc=<0 0 0>, rot=<0 0 0 -1>}
 deform_glute_base_r {parent=pelvis, loc=<-8.8062 -8.8222 8.4287>, rot=<0.0501 -0.0275 0.8926 -0.4472>}
 deform_glute_r {parent=deform_glute_base_r, loc=<0 0 0>, rot=<0 0 0 1>}
 base_chain_front {parent=pelvis, loc=<-7.8321 6.5643 13.3978>, rot=<0.1064 -0.6363 0.7585 0.0921>}
 dyn_chain_front_01 {parent=base_chain_front, loc=<0 0 0>, rot=<0 0 0 1>}
 dyn_chain_front_02 {parent=dyn_chain_front_01, loc=<7.8087 0 0>, rot=<0.0017 0.0465 -0.0364 -0.9983>}
 dyn_chain_front_03 {parent=dyn_chain_front_02, loc=<4.6994 -0.005 -0.0286>, rot=<0.0861 0.0446 -0.0039 -0.9953>}
 dyn_chain_front_04 {parent=dyn_chain_front_03, loc=<4.9952 0.0001 -0.0257>, rot=<-0.0028 0.2605 0.0049 -0.9654>}
 dyn_chain_front_05 {parent=dyn_chain_front_04, loc=<4.6207 0 0>, rot=<0 0 0 1>}
 base_chain_back {parent=pelvis, loc=<-8.2427 -10.8982 9.761>, rot=<0.1113 -0.653 0.7482 0.0381>}
 dyn_chain_back {parent=base_chain_back, loc=<0 0 0>, rot=<0 0 0 1>}
 dyn_chain_back_01 {parent=dyn_chain_back, loc=<6.2567 0 -0.7498>, rot=<0 -0.1575 0 -0.9875>}
 dyn_chain_back_02 {parent=dyn_chain_back_01, loc=<5.0463 0 -0.0829>, rot=<0.0133 0.0024 0.0707 -0.9974>}
 dyn_chain_back_03 {parent=dyn_chain_back_02, loc=<5.3531 -0.0685 0>, rot=<-0.0119 -0.1393 0.0843 -0.9866>}
 dyn_chain_back_04 {parent=dyn_chain_back_03, loc=<4.9541 -0.2443 0.2809>, rot=<-0.0114 -0.053 -0.0408 -0.9977>}
 dyn_chain_back_05 {parent=dyn_chain_back_04, loc=<4.1745 -0.7532 -0.0971>, rot=<0 0 0 1>}
 pants_corrective {parent=pelvis, loc=<-14.4777 7.6927 0>, rot=<-0.5089 -0.491 -0.491 -0.5089>}
 base_chain_extra {parent=pelvis, loc=<-6.9313 6.5918 12.3104>, rot=<-0.1764 -0.9827 0.0081 0.0565>}
 dyn_chain_extra_01 {parent=base_chain_extra, loc=<0 0 0>, rot=<0 0 0 1>}
 dyn_chain_extra_02 {parent=dyn_chain_extra_01, loc=<4.697 0 0>, rot=<0 0 0 1>}
 pants_corrective_01 {parent=pelvis, loc=<-6.8398 7.4189 0>, rot=<-0.5089 -0.491 -0.491 -0.5089>}
 pants_corrective_02 {parent=pelvis, loc=<-14.4777 7.6927 0>, rot=<-0.5089 -0.491 -0.491 -0.5089>}
 base_buckle {parent=pelvis, loc=<-3.3689 9.5015 -0.3195>, rot=<-0.0127 -0.707 0.0127 -0.707>}
 dyn_buckle {parent=base_buckle, loc=<0 0 0>, rot=<0 0 0 1>}
 corrective_pelvis_front {parent=pelvis, loc=<-3.8855 6.5918 0>, rot=<-0.0127 -0.707 0.0127 -0.707>}
]
variables [
 Crouching {type=double, default=0}
 CrouchingBool {type=bool, default=False}
 Detected_Pistol {type=double, default=0}
 EasternBlocEmote {type=double, default=0}
 Emote {type=bool, default=False}
 Floating {type=bool, default=False}
 Moving2d {type=bool, default=False}
 NotCrouching {type=double, default=0}
 OnGround {type=double, default=0}
 RB Enable {type=bool, default=False}
 SkydiveUp {type=double, default=0}
 Sliding {type=bool, default=False}
 Thigh_L_Front {type=double, default=0}
 Thigh_L_Side {type=double, default=0}
 Thigh_R_Back {type=double, default=0}
 Thigh_R_Front {type=double, default=0}
 Thigh_R_Side {type=double, default=0}
]
