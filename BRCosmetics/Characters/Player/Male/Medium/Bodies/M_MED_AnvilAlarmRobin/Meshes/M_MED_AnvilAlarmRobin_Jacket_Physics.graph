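# Physics asset for the jacket of the M_MED_AnvilAlarmRobin male player body, defining dangling secondary motion for its dyn_jacket bones.
# Capsule chains (two-link back, three-link front and side) hang from tiny MSTR anchor bodies with tight swing limits.
# Twist is locked on back and upper side links, while an all-free angular constraint joins only the right front and side tips.
# Pelvis, lower-spine and thigh capsules provide body collision shapes.
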PhysicsAsset /BRCosmetics/Characters/Player/Male/Medium/Bodies/M_MED_AnvilAlarmRobin/Meshes/M_MED_AnvilAlarmRobin_Jacket_Physics
bodies {
 MSTR_jacket_bk [{shapeType=Capsule, length=0.505, radius=0.505, shapeName=MSTR_jacket_bk_capsule, rotation=<rot 0 -90 0>}]
 MSTR_jacket_bk_l [{shapeType=Capsule, length=0.505, radius=0.505, shapeName=MSTR_jacket_bk_l_capsule, rotation=<rot 0 -90 0>}]
 MSTR_jacket_bk_r [{shapeType=Capsule, length=0.505, radius=0.505, shapeName=MSTR_jacket_bk_r_capsule, rotation=<rot 0 -90 0>}]
 MSTR_jacket_front_l [{shapeType=Sphere, radius=0.505, shapeName=MSTR_jacket_front_l_sphere}]
 MSTR_jacket_front_r [{shapeType=Sphere, radius=0.505, shapeName=MSTR_jacket_front_r_sphere}]
 MSTR_jacket_side_l [{shapeType=Sphere, radius=0.505, shapeName=MSTR_jacket_side_l_sphere}]
 MSTR_jacket_side_r [{shapeType=Sphere, radius=0.505, shapeName=MSTR_jacket_side_r_sphere}]
 dyn_jacket_bk_01 [{shapeType=Capsule, length=10, radius=4, shapeName=dyn_jacket_bk_01_capsule, center=<3.9151 -4.1593 0>}]
 dyn_jacket_bk_01_l [{shapeType=Capsule, length=10, radius=4, shapeName=dyn_jacket_bk_01_l_capsule, center=<3.9151 -4.1593 0>}]
 dyn_jacket_bk_01_r [{shapeType=Capsule, length=10, radius=4, shapeName=dyn_jacket_bk_01_r_capsule, center=<3.9151 -4.1593 0>}]
 dyn_jacket_bk_02 [{shapeType=Capsule, length=10, radius=4, shapeName=dyn_jacket_bk_02_capsule, center=<3.9151 -4.1593 0>}]
 dyn_jacket_bk_02_l [{shapeType=Capsule, length=10, radius=4, shapeName=dyn_jacket_bk_02_l_capsule, center=<3.9151 -4.1593 0>}]
 dyn_jacket_bk_02_r [{shapeType=Capsule, length=10, radius=4, shapeName=dyn_jacket_bk_02_r_capsule, center=<3.9151 -4.1593 0>}]
 dyn_jacket_front_01_l [{shapeType=Capsule, length=9.3426, radius=5, shapeName=dyn_jacket_fr_01_r_capsule, center=<7.8651 -3.4967 0>, rotation=<rot -90 178.9092 0>}]
 dyn_jacket_front_01_r [{
   shapeType=Capsule
   length=9.3426
   radius=5
   shapeName=dyn_jacket_fr_01_l_capsule
   center=<7.8651 -3.4967 0>
   rotation=<rot 90 -0.8952 0.1956>
  }]
 dyn_jacket_front_02_l [{
   shapeType=Capsule
   length=13.3865
   radius=4.6026
   shapeName=dyn_jacket_fr_01_l_capsule
   center=<5.073 -2.7945 0.6604>
   rotation=<rot 0.304 175.0556 -160.7242>
  }]
 dyn_jacket_front_02_r [{
   shapeType=Capsule
   length=13.3865
   radius=4.6026
   shapeName=dyn_jacket_fr_01_r_capsule
   center=<4.9075 -4.6666 -0.0031>
   rotation=<rot -0.3041 175.0557 0>
  }]
 dyn_jacket_front_03_l [{
   shapeType=Capsule
   length=13.6563
   radius=4.9678
   shapeName=dyn_jacket_front_03_r_capsule
   center=<3.7862 -2.1021 0.8223>
   rotation=<rot 10.7198 91.4142 175.8801>
  }]
 dyn_jacket_front_03_r [{
   shapeType=Capsule
   length=13.6563
   radius=4.9678
   shapeName=dyn_jacket_front_03_l_capsule
   center=<4.4369 -2.556 0.392>
   rotation=<rot -6.3091 87.2787 5.4099>
  }]
 dyn_jacket_side_01_l [{shapeType=Capsule, length=5.9621, radius=5.5945, shapeName=dyn_jacket_fr_01_l_capsule, center=<5.9634 -4.5641 0>, rotation=<rot -90 0 0>}]
 dyn_jacket_side_01_r [{shapeType=Capsule, length=5.9621, radius=5.5945, shapeName=dyn_jacket_fr_01_r_capsule, center=<5.9634 -4.5641 0>, rotation=<rot 90 180 0>}]
 dyn_jacket_side_02_l [{
   shapeType=Capsule
   length=11.9319
   radius=5.4897
   shapeName=dyn_jacket_front_02_l_capsule
   center=<3.2815 -4.911 -0.304>
   rotation=<rot -4.2341 -90.0047 0.0647>
  }]
 dyn_jacket_side_02_r [{
   shapeType=Capsule
   length=11.9319
   radius=5.4897
   shapeName=dyn_jacket_front_02_r_capsule
   center=<3.2815 -4.911 0.304>
   rotation=<rot 4.2341 -90.0047 -179.9353>
  }]
 dyn_jacket_side_03_l [{
   shapeType=Capsule
   length=12.1459
   radius=4.5373
   shapeName=dyn_jacket_front_03_l_capsule
   center=<3.6393 -4.3989 0.1335>
   rotation=<rot -2.7572 -90.3 -0.3794>
  }]
 dyn_jacket_side_03_r [{
   shapeType=Capsule
   length=12.1459
   radius=4.5373
   shapeName=dyn_jacket_front_03_r_capsule
   center=<3.6385 -4.538 -0.1268>
   rotation=<rot 2.7572 -90.3 179.6206>
  }]
 pelvis [{shapeType=Capsule, length=6.1382, radius=11.0327, shapeName=pelvis_capsule, center=<-1.0028 -0.849 0.017>, rotation=<rot -0.5854 -0.0001 0>}]
 spine_01 [{shapeType=Capsule, length=6.1916, radius=11.258, shapeName=pelvis_capsule, center=<0.6596 -0.7248 0>, rotation=<rot -0.5854 -0.0001 0>}]
 thigh_twist_01_l [{
   shapeType=Capsule
   length=21.8164
   radius=10.5943
   shapeName=thigh_twist_01_l_capsule
   center=<10.4333 -0.3863 2.0779>
   rotation=<rot -89.914 -179.9873 -1.3895>
  }]
 thigh_twist_01_r [{
   shapeType=Capsule
   length=24.4066
   radius=9.2649
   shapeName=thigh_twist_01_r_capsule
   center=<-10.4323 0.3341 -1.4347>
   rotation=<rot -89.914 -179.9963 179.9963>
  }]
}
constraints [
 {bone1Name=dyn_jacket_bk_01, bone2Name=MSTR_jacket_bk, swing1LimitDegrees=10, swing2LimitDegrees=15, twistLimitDegrees=45, twistMotion=Locked}
 {
  bone1Name=dyn_jacket_bk_02
  bone2Name=dyn_jacket_bk_01
  swing1LimitDegrees=10
  swing2LimitDegrees=15
  twistLimitDegrees=45
  twistMotion=Locked
 }
 {
  bone1Name=dyn_jacket_bk_01_l
  bone2Name=MSTR_jacket_bk_l
  swing1LimitDegrees=10
  swing2LimitDegrees=15
  twistLimitDegrees=45
  twistMotion=Locked
 }
 {
  bone1Name=dyn_jacket_bk_02_l
  bone2Name=dyn_jacket_bk_01_l
  swing1LimitDegrees=10
  swing2LimitDegrees=15
  twistLimitDegrees=45
  twistMotion=Locked
 }
 {
  bone1Name=dyn_jacket_bk_01_r
  bone2Name=MSTR_jacket_bk_r
  swing1LimitDegrees=10
  swing2LimitDegrees=15
  twistLimitDegrees=45
  twistMotion=Locked
 }
 {
  bone1Name=dyn_jacket_bk_02_r
  bone2Name=dyn_jacket_bk_01_r
  swing1LimitDegrees=10
  swing2LimitDegrees=15
  twistLimitDegrees=45
  twistMotion=Locked
 }
 {bone1Name=dyn_jacket_front_01_l, bone2Name=MSTR_jacket_front_l, swing1LimitDegrees=10, swing2LimitDegrees=15, twistLimitDegrees=10}
 {bone1Name=dyn_jacket_front_02_l, bone2Name=dyn_jacket_front_01_l, swing1LimitDegrees=10, swing2LimitDegrees=15, twistLimitDegrees=10}
 {bone1Name=dyn_jacket_front_03_l, bone2Name=dyn_jacket_front_02_l, swing1LimitDegrees=15, swing2LimitDegrees=10, twistLimitDegrees=10}
 {bone1Name=dyn_jacket_front_01_r, bone2Name=MSTR_jacket_front_r, swing1LimitDegrees=10, swing2LimitDegrees=15, twistLimitDegrees=10}
 {bone1Name=dyn_jacket_front_02_r, bone2Name=dyn_jacket_front_01_r, swing1LimitDegrees=15, swing2LimitDegrees=15, twistLimitDegrees=10}
 {bone1Name=dyn_jacket_front_03_r, bone2Name=dyn_jacket_front_02_r, swing1LimitDegrees=15, swing2LimitDegrees=10, twistLimitDegrees=10}
 {
  bone1Name=dyn_jacket_front_03_r
  bone2Name=dyn_jacket_side_03_r
  swing1LimitDegrees=45
  swing1Motion=Free
  swing2LimitDegrees=45
  swing2Motion=Free
  twistLimitDegrees=45
  twistMotion=Free
 }
 {
  bone1Name=dyn_jacket_side_01_l
  bone2Name=MSTR_jacket_side_l
  swing1LimitDegrees=7
  swing2LimitDegrees=10
  twistLimitDegrees=5
  twistMotion=Locked
 }
 {
  bone1Name=dyn_jacket_side_02_l
  bone2Name=dyn_jacket_side_01_l
  swing1LimitDegrees=30
  swing2LimitDegrees=15
  twistLimitDegrees=5
  twistMotion=Locked
 }
 {bone1Name=dyn_jacket_side_03_l, bone2Name=dyn_jacket_side_02_l, swing1LimitDegrees=15, swing2LimitDegrees=5, twistLimitDegrees=10}
 {
  bone1Name=dyn_jacket_side_01_r
  bone2Name=MSTR_jacket_side_r
  swing1LimitDegrees=7
  swing2LimitDegrees=10
  twistLimitDegrees=5
  twistMotion=Locked
 }
 {
  bone1Name=dyn_jacket_side_02_r
  bone2Name=dyn_jacket_side_01_r
  swing1LimitDegrees=30
  swing2LimitDegrees=5
  twistLimitDegrees=5
  twistMotion=Locked
 }
 {bone1Name=dyn_jacket_side_03_r, bone2Name=dyn_jacket_side_02_r, swing1LimitDegrees=15, swing2LimitDegrees=5, twistLimitDegrees=10}
]
