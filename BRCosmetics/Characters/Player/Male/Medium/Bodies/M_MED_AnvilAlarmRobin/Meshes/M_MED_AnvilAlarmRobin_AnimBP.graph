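# Post-process style AnimBlueprint for the AnvilAlarmRobin male body cosmetic, adding secondary motion to its jacket, chains, necklaces, collar, pants and boots.
# A parent-copied pose runs through a state-driven Control Rig, then RigidBody and AnimDynamics simulations, with constraints averaging center jacket-back bones.
# Additive bone fixes follow, bRigidBodyEnabled selects simulated or rig-only poses, and attachment offsets and pants correctives finish.
# EventGraph updates gravity overrides, RBAN settings and state-based chain alpha.
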AnimBlueprint /BRCosmetics/Characters/Player/Male/Medium/Bodies/M_MED_AnvilAlarmRobin/Meshes/M_MED_AnvilAlarmRobin_AnimBP
target_skeleton=&/M_MED_AnvilAlarmRobin_Skeleton
graphs [
 AnimGraph {nodes [
   n1 = "Transform (Modify) Bone - Bone: base_chain_front"!(
    ComponentPose=Local To Component(LocalPose=Use cached pose 'CtrlRigCache'())
    bAlphaBoolEnabled:bool=True
    Translation="0,0,0"
    Rotation="0.000,36.236,0.000"
    Scale="1,1,1"
   ) -> (Pose) {
    alphaInputType=Bool
    rotationMode=BMM_Additive
    rotationSpace=BCS_BoneSpace
    translationMode=BMM_Additive
    translationSpace=BCS_BoneSpace
    alphaBoolBlend {blendInTime=0.6, blendOption=CircularIn, blendOutTime=0.05}
    alphaScaleBiasClamp {bInterpResult=true}
    boneToModify {boneName=base_chain_front}
    rotation {yaw=36.236}
   }
   n2 = Component To Local(ComponentPose=@n1)
   n3 = "Transform (Modify) Bone - Bone: MSTR_jacket_front_l"!(
    ComponentPose=Local To Component(LocalPose=Use cached pose 'DynamicsCache'())
    bAlphaBoolEnabled:bool=True
    Translation="0,0,0"
    Rotation="7.212,6.696,1.089"
    Scale="1,1,1"
   ) -> (Pose) {
    alphaInputType=Bool
    rotationMode=BMM_Additive
    rotationSpace=BCS_BoneSpace
    translationMode=BMM_Additive
    translationSpace=BCS_BoneSpace
    alphaBoolBlend {blendInTime=0.6, blendOutTime=0.2}
    boneToModify {boneName=MSTR_jacket_front_l}
    rotation=<rot 7.212 6.696 1.089>
   }
   n4 = "Transform (Modify) Bone - Bone: MSTR_jacket_bk_l"!(ComponentPose=@n3, bAlphaBoolEnabled:bool=True, Translation="-1.190,-2.142,-0.039", Rotation="-6.453,-25.682,-0.526", Scale="1,1,1") -> (Pose) {
    alphaInputType=Bool
    rotationMode=BMM_Additive
    rotationSpace=BCS_BoneSpace
    translationMode=BMM_Additive
    translationSpace=BCS_BoneSpace
    alphaBoolBlend {blendInTime=0.6, blendOutTime=0.2}
    boneToModify {boneName=MSTR_jacket_bk_l}
    rotation=<rot -6.453 -25.682 -0.526>
    translation=<-1.19 -2.142 -0.039>
   }
   n5 = "Transform (Modify) Bone - Bone: MSTR_jacket_bk"!(ComponentPose=@n4, bAlphaBoolEnabled:bool=True, Translation="-1.891,-2.524,0.000", Rotation="0.112,-37.222,-0.026", Scale="1,1,1") -> (Pose) {
    alphaInputType=Bool
    rotationMode=BMM_Additive
    rotationSpace=BCS_BoneSpace
    translationMode=BMM_Additive
    translationSpace=BCS_BoneSpace
    alphaBoolBlend {blendInTime=0.6, blendOutTime=0.2}
    boneToModify {boneName=MSTR_jacket_bk}
    rotation=<rot 0.112 -37.222 -0.026>
    translation=<-1.891 -2.524>
   }
   n6 = "Transform (Modify) Bone - Bone: MSTR_jacket_bk_r"!(ComponentPose=@n5, bAlphaBoolEnabled:bool=True, Translation="-2.171,-2.661,0.185", Rotation="5.680,-35.723,3.470", Scale="1,1,1") -> (Pose) {
    alphaInputType=Bool
    rotationMode=BMM_Additive
    rotationSpace=BCS_BoneSpace
    translationMode=BMM_Additive
    translationSpace=BCS_BoneSpace
    alphaBoolBlend {blendInTime=0.6, blendOutTime=0.2}
    boneToModify {boneName=MSTR_jacket_bk_r}
    rotation=<rot 5.68 -35.723 3.47>
    translation=<-2.171 -2.661 0.185>
   }
   n7 = "Transform (Modify) Bone - Bone: base_chain_front"(ComponentPose=@n6, bAlphaBoolEnabled=True, Translation="0,0,0", Rotation="-37.424,23.132,-9.274", Scale="1,1,1") {
    alphaInputType=Bool
    rotationMode=BMM_Additive
    rotationSpace=BCS_BoneSpace
    translationMode=BMM_Additive
    translationSpace=BCS_BoneSpace
    alphaBoolBlend {blendInTime=1, blendOutTime=0.2}
    boneToModify {boneName=base_chain_front}
    rotation=<rot -37.424 23.132 -9.274>
   }
   n8 = "Transform (Modify) Bone - Bone: dyn_chain_back_01"!(ComponentPose=@n7, bAlphaBoolEnabled:bool=True, Translation="1.515,-2.488,-0.616", Rotation="6.383,39.478,-9.697", Scale="1,1,1") -> (Pose) {
    alphaInputType=Bool
    rotationMode=BMM_Additive
    rotationSpace=BCS_BoneSpace
    translationMode=BMM_Additive
    translationSpace=BCS_BoneSpace
    alphaBoolBlend {blendInTime=0.6, blendOutTime=0.2}
    boneToModify {boneName=dyn_chain_back_01}
    rotation=<rot 6.383 39.478 -9.697>
    translation=<1.515 -2.488 -0.616>
   }
   n9 = "Transform (Modify) Bone - Bone: MSTR_jacket_side_l"!(ComponentPose=@n8, bAlphaBoolEnabled:bool=True, Translation="0,0,0", Rotation="11.401,-0.013,0.017", Scale="0.832,1.000,1.000") -> (Pose) {
    alphaInputType=Bool
    rotationMode=BMM_Additive
    rotationSpace=BCS_BoneSpace
    scaleMode=BMM_Additive
    scaleSpace=BCS_BoneSpace
    translationMode=BMM_Additive
    translationSpace=BCS_BoneSpace
    alphaBoolBlend {blendInTime=1, blendOutTime=0.2}
    boneToModify {boneName=MSTR_jacket_side_l}
    rotation=<rot 11.401 -0.013 0.017>
    scale {x=0.832}
   }
   n10 = "Transform (Modify) Bone - Bone: base_chain_extra"(ComponentPose=@n9, Alpha=1, Translation="-0.459,-1.069,0.057", Rotation="20.527,-15.819,-25.920", Scale="1,1,1") {
    rotationMode=BMM_Additive
    rotationSpace=BCS_BoneSpace
    translationMode=BMM_Additive
    translationSpace=BCS_BoneSpace
    alphaBoolBlend {blendInTime=0.6, blendOutTime=0.2}
    alphaScaleBiasClamp {bInterpResult=true}
    boneToModify {boneName=base_chain_extra}
    rotation=<rot 20.527 -15.819 -25.92>
    translation=<-0.4593 -1.0689 0.0569>
   }
   n11 = "Transform (Modify) Bone - Bone: dyn_chain_extra_02"(ComponentPose=@n10, Alpha=1, Translation="-0.298,-0.150,0.000", Rotation="-8.973,25.832,0.172", Scale="1,1,1") {
    rotationMode=BMM_Additive
    rotationSpace=BCS_BoneSpace
    translationMode=BMM_Additive
    translationSpace=BCS_BoneSpace
    alphaBoolBlend {blendInTime=0.6, blendOutTime=0.2}
    alphaScaleBiasClamp {bInterpResult=true}
    boneToModify {boneName=dyn_chain_extra_02}
    rotation=<rot -8.973 25.832 0.172>
    translation=<-0.2979 -0.1499 0>
   }
   n12 = "Transform (Modify) Bone - Bone: base_chain_front"!(ComponentPose=@n11, Alpha:float=1, Translation="0.205,-0.528,0.000", Rotation="0.011,18.611,-0.007", Scale="1,1,1") -> (Pose) {
    rotationMode=BMM_Additive
    rotationSpace=BCS_BoneSpace
    translationMode=BMM_Additive
    translationSpace=BCS_BoneSpace
    alphaBoolBlend {blendInTime=0.6, blendOutTime=0.2}
    alphaScaleBiasClamp {bInterpResult=true, interpSpeedIncreasing=2.5}
    boneToModify {boneName=base_chain_front}
    rotation=<rot 0.011 18.611 -0.007>
    translation=<0.205 -0.528>
   }
   n13 = "Transform (Modify) Bone - Bone: MSTR_jacket_front_r"(ComponentPose=@n12, Alpha=1, Translation="0,0,0", Rotation="-0.000,4.184,0.000", Scale="1,1,1") {
    rotationMode=BMM_Additive
    rotationSpace=BCS_BoneSpace
    translationMode=BMM_Additive
    translationSpace=BCS_BoneSpace
    alphaBoolBlend {blendInTime=0.6, blendOutTime=0.2}
    alphaScaleBiasClamp {bInterpResult=true, interpSpeedIncreasing=2.5}
    boneToModify {boneName=MSTR_jacket_front_r}
    rotation=<rot -0.0001 4.184 0>
   }
   n14 = "Transform (Modify) Bone - Bone: MSTR_jacket_front_r"(ComponentPose=@n13, Alpha=1, Translation="0,0,0", Rotation="3.957,3.870,0.721", Scale="1,1,1") {
    rotationMode=BMM_Additive
    rotationSpace=BCS_BoneSpace
    translationMode=BMM_Additive
    translationSpace=BCS_BoneSpace
    alphaBoolBlend {blendInTime=0.6, blendOutTime=0.2}
    alphaScaleBiasClamp {bInterpResult=true}
    boneToModify {boneName=MSTR_jacket_front_r}
    rotation=<rot 3.957 3.87 0.721>
   }
   n15 = "Transform (Modify) Bone - Bone: dyn_jacket_front_02_r"(ComponentPose=@n14, Alpha=1, Translation="0,0,0", Rotation="7.399,0.269,-0.006", Scale="1,1,1") {
    rotationMode=BMM_Additive
    rotationSpace=BCS_BoneSpace
    translationMode=BMM_Additive
    translationSpace=BCS_BoneSpace
    alphaBoolBlend {blendInTime=0.6, blendOutTime=0.2}
    alphaScaleBiasClamp {bInterpResult=true}
    boneToModify {boneName=dyn_jacket_front_02_r}
    rotation=<rot 7.399 0.269 -0.006>
   }
   n16 = "Transform (Modify) Bone - Bone: MSTR_jacket_bk_l"!(ComponentPose=@n15, Alpha:float=1, Translation="-0.855,-0.273,-0.070", Rotation="8.313,17.000,-11.680", Scale="1,1,1") -> (Pose) {
    rotationMode=BMM_Additive
    rotationSpace=BCS_BoneSpace
    translationMode=BMM_Additive
    translationSpace=BCS_BoneSpace
    alphaBoolBlend {blendInTime=0.6, blendOutTime=0.2}
    alphaScaleBiasClamp {bInterpResult=true, interpSpeedIncreasing=5}
    boneToModify {boneName=MSTR_jacket_bk_l}
    rotation=<rot 8.313 17 -11.68>
    translation=<-0.855 -0.273 -0.07>
   }
   n17 = "Transform (Modify) Bone - Bone: MSTR_jacket_bk"!(ComponentPose=@n16, Alpha:float=1, Translation="-0.795,-0.273,0.064", Rotation="-4.289,18.745,-3.425", Scale="1,1,1") -> (Pose) {
    rotationMode=BMM_Additive
    rotationSpace=BCS_BoneSpace
    translationMode=BMM_Additive
    translationSpace=BCS_BoneSpace
    alphaBoolBlend {blendInTime=0.6, blendOutTime=0.2}
    alphaScaleBiasClamp {bInterpResult=true, interpSpeedIncreasing=5}
    boneToModify {boneName=MSTR_jacket_bk}
    rotation=<rot -4.289 18.745 -3.425>
    translation=<-0.795 -0.273 0.064>
   }
   n18 = "Transform (Modify) Bone - Bone: MSTR_jacket_bk_r"!(ComponentPose=@n17, Alpha:float=1, Translation="-1.421,-0.440,0.000", Rotation="-8.397,14.930,-9.510", Scale="1,1,1") -> (Pose) {
    rotationMode=BMM_Additive
    rotationSpace=BCS_BoneSpace
    translationMode=BMM_Additive
    translationSpace=BCS_BoneSpace
    alphaBoolBlend {blendInTime=0.6, blendOutTime=0.2}
    alphaScaleBiasClamp {bInterpResult=true, interpSpeedIncreasing=5}
    boneToModify {boneName=MSTR_jacket_bk_r}
    rotation=<rot -8.397 14.93 -9.51>
    translation=<-1.421 -0.44>
   }
   n19 = Look At - Bone: dyn_jacket_front_01_l!(ComponentPose=@n18, bAlphaBoolEnabled:bool=True) -> (Pose) {
    alphaInputType=Bool
    alphaBoolBlend {blendInTime=0.2, blendOutTime=0.2}
    boneToModify {boneName=dyn_jacket_front_01_l}
    lookAtLocation=<0 11.2269 -7.7357>
    lookAtTarget {boneReference {boneName=thigh_l}}
    lookAt_Axis {axis=<1 0>}
   }
   n20 = Look At - Bone: dyn_jacket_side_01_l!(ComponentPose=@n19, bAlphaBoolEnabled:bool=True) -> (Pose) {
    alphaInputType=Bool
    alphaBoolBlend {blendInTime=0.2, blendOutTime=0.2}
    boneToModify {boneName=dyn_jacket_side_01_l}
    lookAtLocation=<-3.1318 0 -9.4129>
    lookAtTarget {boneReference {boneName=thigh_l}}
    lookAt_Axis {axis=<1 0>}
   }
   n21 = Component To Local(ComponentPose=@n20)
   n22 = Get bRigidBodyEnabled()
   n23 = "Transform (Modify) Bone - Bone: attach_backpack"(
    ComponentPose=Local To Component(LocalPose=Blend Poses by bool(BlendPose_0=@n21, BlendPose_1=@n2, bActiveValue=@n22))
    Alpha=1
    Translation="0,0.5,0"
    Rotation="0,0,-7.5"
    Scale="1,1,1"
   ) {
    rotationMode=BMM_Additive
    rotationSpace=BCS_BoneSpace
    translationMode=BMM_Additive
    translationSpace=BCS_BoneSpace
    boneToModify {boneName=attach_backpack}
    rotation {roll=-7.5}
    translation {y=0.5}
   }
   n24 = "Transform (Modify) Bone - Bone: attach_cape"(ComponentPose=@n23, Alpha=1, Translation="0,0,-2.5", Rotation="0,0,0", Scale="1,1,1") {
    rotationMode=BMM_Additive
    rotationSpace=BCS_BoneSpace
    translationMode=BMM_Additive
    translationSpace=BCS_BoneSpace
    boneToModify {boneName=attach_cape}
    translation {z=-2.5}
   }
   n25 = Look At - Bone: pants_corrective_01(ComponentPose=@n24, Alpha=1) {
    boneToModify {boneName=pants_corrective_01}
    lookAtLocation {x=0}
    lookAtTarget {boneReference {boneName=root}}
    lookAt_Axis {axis {y=0, z=-1}}
   }
   n26 = Look At - Bone: pants_corrective_02(ComponentPose=@n25, Alpha=1) {
    boneToModify {boneName=pants_corrective_02}
    lookAtLocation {x=0}
    lookAtTarget {boneReference {boneName=root}}
    lookAt_Axis {axis {y=0, z=-1}}
   }
   n27 = "Transform (Modify) Bone - Bone: pants_corrective_01"(ComponentPose=@n26, Alpha=1, Translation="0.322,-2.812,-0.024", Rotation="0,0,0", Scale="1,1,1") {
    rotationMode=BMM_Additive
    rotationSpace=BCS_BoneSpace
    translationMode=BMM_Additive
    translationSpace=BCS_ParentBoneSpace
    boneToModify {boneName=pants_corrective_01}
    translation=<0.322 -2.812 -0.024>
   }
   n28 = "Transform (Modify) Bone - Bone: pants_corrective_02"(ComponentPose=@n27, Alpha=1, Translation="0.700,-2.076,-0.067", Rotation="-10.821,0.000,0.000", Scale="1,1,1") {
    rotationMode=BMM_Additive
    rotationSpace=BCS_BoneSpace
    translationMode=BMM_Additive
    translationSpace=BCS_ParentBoneSpace
    boneToModify {boneName=pants_corrective_02}
    rotation {pitch=-10.821}
    translation=<0.7 -2.076 -0.067>
   }
   Output Pose(Result=Component To Local(ComponentPose=@n28)) {layerGroup=None}
   n29 = Control Rig!(
    Source=Copy Pose From Mesh() {bUseAttachedParent=true}
    Alpha
    CrouchingBool:bool=Get bIsCrouching()
    Sliding:bool=Get bIsSliding()
    Moving2d:bool=Get bIsMoving2D()
    NotCrouching:double=Get NotCrouching() -> (NotCrouching:double)
    Emote:bool=Get bIsPlayingEmote()
    Detected_Pistol:double=Get DetectedPistol() -> (DetectedPistol:double)
    OnGround:double=Get OnGround() -> (OnGround:double)
    RB Enable:bool=Get bRigidBodyEnabled()
    SkydiveUp:double=Get Alpha_isSkidiveUp() -> (Alpha_isSkidiveUp:float)
    Floating:bool=Get bIsSkydiveFloating()
    EasternBlocEmote:double=Get EasternBlocEmote() -> (EasternBlocEmote:double)
   ) -> (Pose) {controlRigAssetReference {blueprintRigClass=&/M_MED_AnvilAlarmRobin_CtrlRig.M_MED_AnvilAlarmRobin_CtrlRig_C}}
   Save cached pose 'CtrlRigCache'(Pose=@n29) {cacheName=CtrlRigCache}
   n30 = "Transform (Modify) Bone - Bone: MSTR_jacket_front_l"(
    ComponentPose=Local To Component(LocalPose=Use cached pose 'CtrlRigCache'())
    bAlphaBoolEnabled=True
    Translation="0,0,-2"
    Rotation="0,0,0"
    Scale="1,1,1"
   ) {
    alphaInputType=Bool
    rotationMode=BMM_Additive
    rotationSpace=BCS_BoneSpace
    translationMode=BMM_Additive
    translationSpace=BCS_BoneSpace
    alphaBoolBlend {blendInTime=0.2, blendOutTime=0.2}
    boneToModify {boneName=MSTR_jacket_front_l}
    translation {z=-2}
   }
   n31 = "Transform (Modify) Bone - Bone: MSTR_jacket_front_r"!(ComponentPose=@n30, bAlphaBoolEnabled:bool=True, Translation="0,0,0", Rotation="5,-5,0", Scale="1,1,1") -> (Pose) {
    alphaInputType=Bool
    rotationMode=BMM_Additive
    rotationSpace=BCS_BoneSpace
    translationMode=BMM_Additive
    translationSpace=BCS_BoneSpace
    alphaBoolBlend {blendInTime=0.2, blendOutTime=0.2}
    boneToModify {boneName=MSTR_jacket_front_r}
    rotation {pitch=5, yaw=-5}
   }
   Save cached pose 'PreAdjustCache'(Pose=Component To Local(ComponentPose=@n31)) {cacheName=PreAdjustCache}
   n32 = RigidBody!(
    ComponentPose=Local To Component(LocalPose=Use cached pose 'PreAdjustCache'())
    Alpha
    OverrideWorldGravity:Vector="0,0,0"
    ExternalForce
    SimSpaceSettings:Sim Space Settings Structure=(WorldAlpha=1,VelocityScaleZ=1,DampingAlpha=0,MaxLinearVelocity=350,MaxAngularVelocity=5,MaxLinearAcceleration=500,MaxAngularAcceleration=100,ExternalLinearDragV=(X=1,Y=1,Z=1),ExternalLinearVelocity=(X=0,Y=0,Z=0),ExternalAngularVelocity=(X=0,Y=0,Z=0))
   ) -> (Pose) {
    bClampLinearTranslationLimitToRefPose=true
    bOverrideWorldGravity=true
    overridePhysicsAsset=&/M_MED_AnvilAlarmRobin_ChainLeg_Physics
    simulationSpace=BaseBoneSpace
    alphaBoolBlend {blendOption=Linear}
    alphaScaleBiasClamp {bInterpResult=true}
    baseBoneRef {boneName=thigh_r}
    componentAppliedLinearAccClamp=<3000 3000 3000>
    simSpaceSettings {
     dampingAlpha=0
     maxAngularAcceleration=100
     maxAngularVelocity=5
     maxLinearAcceleration=500
     maxLinearVelocity=350
     worldAlpha=1
     externalLinearDragV=<1 1 1>
    }
   }
   n33 = RigidBody!(ComponentPose=@n32, Alpha, OverrideWorldGravity="0,0,0", ExternalForce) -> (Pose) {
    bClampLinearTranslationLimitToRefPose=true
    bOverrideWorldGravity=true
    overridePhysicsAsset=&/M_MED_AnvilAlarmRobin_Jacket_Physics
    simulationSpace=BaseBoneSpace
    alphaBoolBlend {blendOption=Linear}
    baseBoneRef {boneName=pelvis}
    componentAppliedLinearAccClamp=<3000 3000 3000>
    simSpaceSettings {
     dampingAlpha=0
     maxAngularAcceleration=25
     maxAngularVelocity=5
     maxLinearAcceleration=500
     maxLinearVelocity=350
     worldAlpha=1
     externalLinearDragV=<2 2 2>
    }
   }
   n34 = "Constraint - dyn_jacket_bk_01 (RR)"!(ComponentPose=@n33, Alpha=1, ConstraintWeights_0=0.5, ConstraintWeights_1:float=0.5) -> (Pose) {
    boneToModify {boneName=dyn_jacket_bk_01}
    constraintSetup [{transformType=Rotation, targetBone {boneName=dyn_jacket_bk_01_r}}, {transformType=Rotation, targetBone {boneName=dyn_jacket_bk_01_l}}]
    constraintWeights [1,1]
   }
   n35 = "Constraint - dyn_jacket_bk_02 (RR)"!(ComponentPose=@n34, Alpha=1, ConstraintWeights_0=0.5, ConstraintWeights_1:float=0.5) -> (Pose) {
    boneToModify {boneName=dyn_jacket_bk_02}
    constraintSetup [{transformType=Rotation, targetBone {boneName=dyn_jacket_bk_02_r}}, {transformType=Rotation, targetBone {boneName=dyn_jacket_bk_02_l}}]
    constraintWeights [1,1]
   }
   n36 = Anim Dynamics - Bone: dyn_collar_fr_01_l!(SimulationSpace:AnimPhysSimSpaceType Enum, ComponentPose=@n35, Alpha=1, ExternalForce="0,0,0") -> (Pose) {
    angularDampingOverride=1
    angularSpringConstant=10000
    bAngularSpring=true
    bOverrideAngularDamping=true
    bUseGravityOverride=true
    gravityScale=0.35
    simulationSpace=RootRelative
    boundBone {boneName=dyn_collar_fr_01_l}
    componentAppliedLinearAccClamp=<2000 2000 2000>
    componentLinearAccScale=<1 1 1>
    componentLinearVelScale=<1 1 1>
    gravityOverride {z=-980}
    physicsBodyDefinitions [{
      boundBone {boneName=dyn_collar_fr_01_l}
      boxExtents=<20 30 30>
      constraintSetup {angularLimitsMax=<2 10 3>, angularLimitsMin=<-2 0 -3>, angularTarget=<1 0 0>}
      localJointOffset=<4 0 0>
     }]
    simSpaceSettings {maxAngularAcceleration=100, maxAngularVelocity=50, simSpaceAngularAlpha=0.4}
   }
   n37 = Anim Dynamics - Bone: dyn_collar_fr_01_r!(SimulationSpace:AnimPhysSimSpaceType Enum, ComponentPose=@n36, Alpha=1, ExternalForce="0,0,0") -> (Pose) {
    angularDampingOverride=1
    angularSpringConstant=10000
    bAngularSpring=true
    bOverrideAngularDamping=true
    bUseGravityOverride=true
    gravityScale=0.35
    simulationSpace=RootRelative
    boundBone {boneName=dyn_collar_fr_01_r}
    componentAppliedLinearAccClamp=<2000 2000 2000>
    componentLinearAccScale=<1 1 1>
    componentLinearVelScale=<1 1 1>
    gravityOverride {z=-980}
    physicsBodyDefinitions [{
      boundBone {boneName=dyn_collar_fr_01_r}
      boxExtents=<20 30 30>
      constraintSetup {angularLimitsMax=<2 10 15>, angularLimitsMin=<-2 0 -3>, angularTarget=<1 0 0>}
      localJointOffset=<-4 0 0>
     }]
    simSpaceSettings {maxAngularAcceleration=100, maxAngularVelocity=50, simSpaceAngularAlpha=0.4}
   }
   n38 = Anim Dynamics - Bone: dyn_necklace_01_a!(SimulationSpace:AnimPhysSimSpaceType Enum, ComponentPose=@n37, Alpha=1, ExternalForce="0,0,0") -> (Pose) {
    angularDampingOverride=1
    angularSpringConstant=15000
    bAngularSpring=true
    bOverrideAngularDamping=true
    bUseGravityOverride=true
    bUseSphericalLimits=true
    gravityScale=0.35
    simulationSpace=RootRelative
    boundBone {boneName=dyn_necklace_01_a}
    componentAppliedLinearAccClamp=<2000 2000 2000>
    componentLinearAccScale=<1 1 1>
    componentLinearVelScale=<1 1 1>
    gravityOverride {z=-980}
    physicsBodyDefinitions [{
      boundBone {boneName=dyn_necklace_01_a}
      boxExtents=<20 30 30>
      constraintSetup {angularLimitsMax=<0 0 6>, angularLimitsMin=<0 -15 -6>, angularTarget=<1 0 0>}
      localJointOffset=<4 0 0>
     }]
    simSpaceSettings {maxAngularAcceleration=100, maxAngularVelocity=50}
    sphericalLimits [{limitRadius=23.7162, drivingBone {boneName=spine_05}, sphereLocalOffset=<-6.8713 -11.8624 1.1398>}]
   }
   n39 = Anim Dynamics - Bone: dyn_necklace_02_a!(SimulationSpace:AnimPhysSimSpaceType Enum, ComponentPose=@n38, Alpha=1, ExternalForce="0,0,0") -> (Pose) {
    angularDampingOverride=1
    angularSpringConstant=15000
    bAngularSpring=true
    bOverrideAngularDamping=true
    bUseGravityOverride=true
    bUseSphericalLimits=true
    gravityScale=0.35
    simulationSpace=RootRelative
    boundBone {boneName=dyn_necklace_02_a}
    componentAppliedLinearAccClamp=<2000 2000 2000>
    componentLinearAccScale=<1 1 1>
    componentLinearVelScale=<1 1 1>
    gravityOverride {z=-980}
    physicsBodyDefinitions [{
      boundBone {boneName=dyn_necklace_02_a}
      boxExtents=<20 30 30>
      constraintSetup {angularLimitsMax=<0 0 6>, angularLimitsMin=<0 0 -6>, angularTarget=<1 0 0>}
      localJointOffset=<4 0 0>
     }]
    simSpaceSettings {maxAngularAcceleration=100, maxAngularVelocity=50}
    sphericalLimits [{limitRadius=22.9718, drivingBone {boneName=spine_05}, sphereLocalOffset=<-8.187 -9.3605 0>}]
   }
   n40 = Anim Dynamics - Bone: dyn_necklace_01_b!(SimulationSpace:AnimPhysSimSpaceType Enum, ComponentPose=@n39, Alpha=1, ExternalForce="0,0,0") -> (Pose) {
    angularDampingOverride=1
    angularSpringConstant=14000
    bAngularSpring=true
    bOverrideAngularDamping=true
    bUseGravityOverride=true
    bUseSphericalLimits=true
    gravityScale=0.35
    simulationSpace=RootRelative
    boundBone {boneName=dyn_necklace_01_b}
    componentAppliedLinearAccClamp=<2000 2000 2000>
    componentLinearAccScale=<0.9 0.9 0.9>
    componentLinearVelScale=<0.9 0.9 9>
    gravityOverride {z=-980}
    physicsBodyDefinitions [{
      boundBone {boneName=dyn_necklace_01_b}
      boxExtents=<20 30 30>
      constraintSetup {angularLimitsMax=<0 0 6>, angularLimitsMin=<0 -15 -6>, angularTarget=<1 0 0>}
      localJointOffset=<4 0 0>
     }]
    simSpaceSettings {maxAngularAcceleration=100, maxAngularVelocity=50}
    sphericalLimits [{limitRadius=20, drivingBone {boneName=spine_05}, sphereLocalOffset=<-4.3885 -9.5423 0>}]
   }
   n41 = Anim Dynamics - Bone: dyn_necklace_02_b!(SimulationSpace:AnimPhysSimSpaceType Enum, ComponentPose=@n40, Alpha=1, ExternalForce="0,0,0") -> (Pose) {
    angularDampingOverride=1
    angularSpringConstant=14000
    bAngularSpring=true
    bOverrideAngularDamping=true
    bUseGravityOverride=true
    bUseSphericalLimits=true
    gravityScale=0.35
    simulationSpace=RootRelative
    boundBone {boneName=dyn_necklace_02_b}
    componentAppliedLinearAccClamp=<2000 2000 2000>
    componentLinearAccScale=<0.9 0.9 0.9>
    componentLinearVelScale=<0.9 0.9 9>
    gravityOverride {z=-980}
    physicsBodyDefinitions [{
      boundBone {boneName=dyn_necklace_02_b}
      boxExtents=<20 30 30>
      constraintSetup {angularLimitsMax=<0 0 6>, angularLimitsMin=<0 0 -6>, angularTarget=<1 0 0>}
      localJointOffset=<4 0 0>
     }]
    simSpaceSettings {maxAngularAcceleration=100, maxAngularVelocity=50}
    sphericalLimits [{limitRadius=20, drivingBone {boneName=spine_05}, sphereLocalOffset=<-8.187 -7.1912 0>}]
   }
   n42 = "Transform (Modify) Bone - Bone: dyn_necklace_01_a"(ComponentPose=@n41, Alpha=0.5, Translation="0,0,0", Rotation="0,0,0", Scale="1,1,1") {rotationMode=BMM_Additive, rotationSpace=BCS_BoneSpace, alphaScaleBiasClamp {bInterpResult=true}, boneToModify {boneName=dyn_necklace_01_a}}
   n43 = "Transform (Modify) Bone - Bone: dyn_necklace_02_a"(ComponentPose=@n42, Alpha=0.5, Translation="0,0,0", Rotation="0,0,0", Scale="1,1,1") {rotationMode=BMM_Additive, rotationSpace=BCS_BoneSpace, alphaScaleBiasClamp {bInterpResult=true}, boneToModify {boneName=dyn_necklace_02_a}}
   n44 = "Transform (Modify) Bone - Bone: dyn_necklace_01_b"(ComponentPose=@n43, Alpha=1, Translation="0,0,0", Rotation="0,0,0", Scale="1,1,1") {rotationMode=BMM_Additive, rotationSpace=BCS_BoneSpace, alphaScaleBiasClamp {bInterpResult=true}, boneToModify {boneName=dyn_necklace_01_b}}
   n45 = "Transform (Modify) Bone - Bone: dyn_necklace_02_b"(ComponentPose=@n44, Alpha=1, Translation="0,0,0", Rotation="0,0,0", Scale="1,1,1") {rotationMode=BMM_Additive, rotationSpace=BCS_BoneSpace, alphaScaleBiasClamp {bInterpResult=true}, boneToModify {boneName=dyn_necklace_02_b}}
   n46 = Anim Dynamics - Bone: dyn_bagginess_fr_l(ComponentPose=@n45, Alpha=1, ExternalForce="0,0,0") {
    angularDampingOverride=1
    angularSpringConstant=25000
    bAngularSpring=true
    bOverrideAngularDamping=true
    bUseGravityOverride=true
    gravityScale=0.35
    linearDampingOverride=0.95
    linearSpringConstant=300
    simulationSpace=RootRelative
    boundBone {boneName=dyn_bagginess_fr_l}
    componentAppliedLinearAccClamp=<3000 3000 3000>
    componentLinearAccScale=<1 1 1>
    componentLinearVelScale=<1 1 1>
    gravityOverride {z=-980}
    physicsBodyDefinitions [{
      boundBone {boneName=dyn_bagginess_fr_l}
      boxExtents=<20 30 30>
      constraintSetup {angularLimitsMax=<0 20 20>, angularLimitsMin=<0 -20 -20>, angularTarget=<1 0 0>}
      localJointOffset=<6 0 0>
     }]
    simSpaceSettings {maxAngularAcceleration=100, maxAngularVelocity=50, simSpaceAngularAlpha=1}
   }
   n47 = Anim Dynamics - Bone: dyn_bagginess_sd_l(ComponentPose=@n46, Alpha=1, ExternalForce="0,0,0") {
    angularDampingOverride=1
    angularSpringConstant=25000
    bAngularSpring=true
    bOverrideAngularDamping=true
    bUseGravityOverride=true
    gravityScale=0.35
    linearDampingOverride=0.95
    linearSpringConstant=300
    simulationSpace=RootRelative
    boundBone {boneName=dyn_bagginess_sd_l}
    componentAppliedLinearAccClamp=<3000 3000 3000>
    componentLinearAccScale=<1 1 1>
    componentLinearVelScale=<1 1 1>
    gravityOverride {z=-980}
    physicsBodyDefinitions [{
      boundBone {boneName=dyn_bagginess_sd_l}
      boxExtents=<20 30 30>
      constraintSetup {angularLimitsMax=<0 20 20>, angularLimitsMin=<0 -20 -20>, angularTarget=<1 0 0>}
      localJointOffset=<6 0 0>
     }]
    simSpaceSettings {maxAngularAcceleration=100, maxAngularVelocity=50, simSpaceAngularAlpha=1}
   }
   n48 = Anim Dynamics - Bone: dyn_bagginess_bk_l(ComponentPose=@n47, Alpha=1, ExternalForce="0,0,0") {
    angularDampingOverride=1
    angularSpringConstant=25000
    bAngularSpring=true
    bOverrideAngularDamping=true
    bUseGravityOverride=true
    gravityScale=0.35
    linearDampingOverride=0.95
    linearSpringConstant=300
    simulationSpace=RootRelative
    boundBone {boneName=dyn_bagginess_bk_l}
    componentAppliedLinearAccClamp=<3000 3000 3000>
    componentLinearAccScale=<1 1 1>
    componentLinearVelScale=<1 1 1>
    gravityOverride {z=-980}
    physicsBodyDefinitions [{
      boundBone {boneName=dyn_bagginess_bk_l}
      boxExtents=<20 30 30>
      constraintSetup {angularLimitsMax=<0 20 20>, angularLimitsMin=<0 -20 -20>, angularTarget=<1 0 0>}
      localJointOffset=<6 0 0>
     }]
    simSpaceSettings {maxAngularAcceleration=100, maxAngularVelocity=50, simSpaceAngularAlpha=1}
   }
   n49 = Anim Dynamics - Bone: dyn_bagginess_in_l(ComponentPose=@n48, Alpha=1, ExternalForce="0,0,0") {
    angularDampingOverride=1
    angularSpringConstant=25000
    bAngularSpring=true
    bOverrideAngularDamping=true
    bUseGravityOverride=true
    gravityScale=0.35
    linearDampingOverride=0.95
    linearSpringConstant=300
    simulationSpace=RootRelative
    boundBone {boneName=dyn_bagginess_in_l}
    componentAppliedLinearAccClamp=<3000 3000 3000>
    componentLinearAccScale=<1 1 1>
    componentLinearVelScale=<1 1 1>
    gravityOverride {z=-980}
    physicsBodyDefinitions [{
      boundBone {boneName=dyn_bagginess_in_l}
      boxExtents=<20 30 30>
      constraintSetup {angularLimitsMax=<0 20 20>, angularLimitsMin=<0 -20 -20>, angularTarget=<1 0 0>}
      localJointOffset=<6 0 0>
     }]
    simSpaceSettings {maxAngularAcceleration=100, maxAngularVelocity=50, simSpaceAngularAlpha=1}
   }
   n50 = Anim Dynamics - Bone: dyn_bagginess_fr_r(ComponentPose=@n49, Alpha=1, ExternalForce="0,0,0") {
    angularDampingOverride=1
    angularSpringConstant=25000
    bAngularSpring=true
    bOverrideAngularDamping=true
    bUseGravityOverride=true
    gravityScale=0.35
    linearDampingOverride=0.95
    linearSpringConstant=300
    simulationSpace=RootRelative
    boundBone {boneName=dyn_bagginess_fr_r}
    componentAppliedLinearAccClamp=<3000 3000 3000>
    componentLinearAccScale=<1 1 1>
    componentLinearVelScale=<1 1 1>
    gravityOverride {z=-980}
    physicsBodyDefinitions [{
      boundBone {boneName=dyn_bagginess_fr_r}
      boxExtents=<20 30 30>
      constraintSetup {angularLimitsMax=<0 20 20>, angularLimitsMin=<0 -20 -20>, angularTarget=<1 0 0>}
      localJointOffset=<6 0 0>
     }]
    simSpaceSettings {maxAngularAcceleration=100, maxAngularVelocity=50, simSpaceAngularAlpha=1}
   }
   n51 = Anim Dynamics - Bone: dyn_bagginess_sd_r(ComponentPose=@n50, Alpha=1, ExternalForce="0,0,0") {
    angularDampingOverride=1
    angularSpringConstant=25000
    bAngularSpring=true
    bOverrideAngularDamping=true
    bUseGravityOverride=true
    gravityScale=0.35
    linearDampingOverride=0.95
    linearSpringConstant=300
    simulationSpace=RootRelative
    boundBone {boneName=dyn_bagginess_sd_r}
    componentAppliedLinearAccClamp=<3000 3000 3000>
    componentLinearAccScale=<1 1 1>
    componentLinearVelScale=<1 1 1>
    gravityOverride {z=-980}
    physicsBodyDefinitions [{
      boundBone {boneName=dyn_bagginess_sd_r}
      boxExtents=<20 30 30>
      constraintSetup {angularLimitsMax=<0 20 20>, angularLimitsMin=<0 -20 -20>, angularTarget=<1 0 0>}
      localJointOffset=<6 0 0>
     }]
    simSpaceSettings {maxAngularAcceleration=100, maxAngularVelocity=50, simSpaceAngularAlpha=1}
   }
   n52 = Anim Dynamics - Bone: dyn_bagginess_bk_r(ComponentPose=@n51, Alpha=1, ExternalForce="0,0,0") {
    angularDampingOverride=1
    angularSpringConstant=25000
    bAngularSpring=true
    bOverrideAngularDamping=true
    bUseGravityOverride=true
    gravityScale=0.35
    linearDampingOverride=0.95
    linearSpringConstant=300
    simulationSpace=RootRelative
    boundBone {boneName=dyn_bagginess_bk_r}
    componentAppliedLinearAccClamp=<3000 3000 3000>
    componentLinearAccScale=<1 1 1>
    componentLinearVelScale=<1 1 1>
    gravityOverride {z=-980}
    physicsBodyDefinitions [{
      boundBone {boneName=dyn_bagginess_bk_r}
      boxExtents=<20 30 30>
      constraintSetup {angularLimitsMax=<0 20 20>, angularLimitsMin=<0 -20 -20>, angularTarget=<1 0 0>}
      localJointOffset=<6 0 0>
     }]
    simSpaceSettings {maxAngularAcceleration=100, maxAngularVelocity=50, simSpaceAngularAlpha=1}
   }
   n53 = Anim Dynamics - Bone: dyn_bagginess_in_r(ComponentPose=@n52, Alpha=1, ExternalForce="0,0,0") {
    angularDampingOverride=1
    angularSpringConstant=25000
    bAngularSpring=true
    bOverrideAngularDamping=true
    bUseGravityOverride=true
    gravityScale=0.35
    linearDampingOverride=0.95
    linearSpringConstant=300
    simulationSpace=RootRelative
    boundBone {boneName=dyn_bagginess_in_r}
    componentAppliedLinearAccClamp=<3000 3000 3000>
    componentLinearAccScale=<1 1 1>
    componentLinearVelScale=<1 1 1>
    gravityOverride {z=-980}
    physicsBodyDefinitions [{
      boundBone {boneName=dyn_bagginess_in_r}
      boxExtents=<20 30 30>
      constraintSetup {angularLimitsMax=<0 20 20>, angularLimitsMin=<0 -20 -20>, angularTarget=<1 0 0>}
      localJointOffset=<6 0 0>
     }]
    simSpaceSettings {maxAngularAcceleration=100, maxAngularVelocity=50, simSpaceAngularAlpha=1}
   }
   n54 = Anim Dynamics - Bone: dyn_boot_tongue_01(ComponentPose=@n53, Alpha=1, ExternalForce="0,0,0") {
    angularDampingOverride=1
    angularSpringConstant=15000
    bAngularSpring=true
    bOverrideAngularDamping=true
    bUseGravityOverride=true
    gravityScale=0.35
    simulationSpace=RootRelative
    boundBone {boneName=dyn_boot_tongue_01}
    componentAppliedLinearAccClamp=<3000 3000 3000>
    componentLinearAccScale=<1 1 1>
    componentLinearVelScale=<1 1 1>
    gravityOverride {z=-980}
    physicsBodyDefinitions [{
      boundBone {boneName=dyn_boot_tongue_01}
      boxExtents=<20 30 30>
      constraintSetup {angularLimitsMax=<0 15 5>, angularLimitsMin=<0 0 -5>, angularTarget=<1 0 0>}
      localJointOffset=<4 0 0>
     }]
    simSpaceSettings {maxAngularAcceleration=100, maxAngularVelocity=50, simSpaceAngularAlpha=0.5}
   }
   n55 = Anim Dynamics - Bone: dyn_boot_tongue_02(ComponentPose=@n54, Alpha=1, ExternalForce="0,0,0") {
    angularDampingOverride=1
    angularSpringConstant=15000
    bAngularSpring=true
    bOverrideAngularDamping=true
    bUseGravityOverride=true
    gravityScale=0.35
    simulationSpace=RootRelative
    boundBone {boneName=dyn_boot_tongue_02}
    componentAppliedLinearAccClamp=<3000 3000 3000>
    componentLinearAccScale=<1 1 1>
    componentLinearVelScale=<1 1 1>
    gravityOverride {z=-980}
    physicsBodyDefinitions [{
      boundBone {boneName=dyn_boot_tongue_02}
      boxExtents=<20 30 30>
      constraintSetup {angularLimitsMax=<0 15 5>, angularLimitsMin=<0 -5 -5>, angularTarget=<1 0 0>}
      localJointOffset=<4 0 0>
     }]
    simSpaceSettings {maxAngularAcceleration=100, maxAngularVelocity=50, simSpaceAngularAlpha=0.5}
   }
   n56 = Anim Dynamics - Bone: dyn_laces_01_l(ComponentPose=@n55, Alpha=1, ExternalForce="0,0,0") {
    angularDampingOverride=1
    angularSpringConstant=14000
    bAngularSpring=true
    bOverrideAngularDamping=true
    bUseGravityOverride=true
    bUsePlanarLimit=false
    bUseSphericalLimits=true
    gravityScale=0.35
    simulationSpace=RootRelative
    boundBone {boneName=dyn_laces_01_l}
    componentAppliedLinearAccClamp=<3000 3000 3000>
    componentLinearAccScale=<0.7 0.7 0.7>
    componentLinearVelScale=<1 1 1>
    gravityOverride {z=-980}
    physicsBodyDefinitions [{
      boundBone {boneName=dyn_laces_01_l}
      boxExtents=<20 30 30>
      constraintSetup {angularLimitsMax=<0 -3 15>, angularLimitsMin=<0 -30 -15>, angularTarget=<1 0 0>}
      localJointOffset=<4 0 0>
     }]
    simSpaceSettings {maxAngularAcceleration=100, maxAngularVelocity=50, simSpaceAngularAlpha=1}
    sphericalLimits [{limitRadius=5.2, drivingBone {boneName=foot_r}, sphereLocalOffset=<3.0054 -4.9131 0.3573>}]
   }
   n57 = Anim Dynamics - Bone: dyn_laces_02_l(ComponentPose=@n56, Alpha=1, ExternalForce="0,0,0") {
    angularDampingOverride=1
    angularSpringConstant=14000
    bAngularSpring=true
    bOverrideAngularDamping=true
    bUseGravityOverride=true
    bUsePlanarLimit=false
    bUseSphericalLimits=true
    gravityScale=0.35
    simulationSpace=RootRelative
    boundBone {boneName=dyn_laces_02_l}
    componentAppliedLinearAccClamp=<3000 3000 3000>
    componentLinearAccScale=<0.7 0.7 0.7>
    componentLinearVelScale=<1 1 1>
    gravityOverride {z=-980}
    physicsBodyDefinitions [{
      boundBone {boneName=dyn_laces_02_l}
      boxExtents=<20 30 30>
      constraintSetup {angularLimitsMax=<0 -3 15>, angularLimitsMin=<0 -30 -15>, angularTarget=<1 0 0>}
      localJointOffset=<4 0 0>
     }]
    simSpaceSettings {maxAngularAcceleration=100, maxAngularVelocity=50, simSpaceAngularAlpha=1}
    sphericalLimits [{limitRadius=5.2, drivingBone {boneName=foot_r}, sphereLocalOffset=<3.0054 -4.9131 0.3573>}]
   }
   n58 = Anim Dynamics - Bone: dyn_laces_03_l(ComponentPose=@n57, Alpha=1, ExternalForce="0,0,0") {
    angularDampingOverride=1
    angularSpringConstant=14000
    bAngularSpring=true
    bOverrideAngularDamping=true
    bUseGravityOverride=true
    bUsePlanarLimit=false
    bUseSphericalLimits=true
    gravityScale=0.35
    simulationSpace=RootRelative
    boundBone {boneName=dyn_laces_03_l}
    componentAppliedLinearAccClamp=<3000 3000 3000>
    componentLinearAccScale=<0.7 0.7 0.7>
    componentLinearVelScale=<1 1 1>
    gravityOverride {z=-980}
    physicsBodyDefinitions [{
      boundBone {boneName=dyn_laces_03_l}
      boxExtents=<20 30 30>
      constraintSetup {angularLimitsMax=<0 -3 15>, angularLimitsMin=<0 -30 -15>, angularTarget=<1 0 0>}
      localJointOffset=<4 0 0>
     }]
    simSpaceSettings {maxAngularAcceleration=100, maxAngularVelocity=50, simSpaceAngularAlpha=1}
    sphericalLimits [{limitRadius=5.2, drivingBone {boneName=foot_r}, sphereLocalOffset=<3.0054 -4.9131 0.3573>}]
   }
   n59 = Anim Dynamics - Bone: dyn_laces_01_r(ComponentPose=@n58, Alpha=1, ExternalForce="0,0,0") {
    angularDampingOverride=1
    angularSpringConstant=15000
    bAngularSpring=true
    bOverrideAngularDamping=true
    bUseGravityOverride=true
    bUsePlanarLimit=false
    bUseSphericalLimits=true
    gravityScale=0.35
    simulationSpace=RootRelative
    boundBone {boneName=dyn_laces_01_r}
    componentAppliedLinearAccClamp=<3000 3000 3000>
    componentLinearAccScale=<1 1 1>
    componentLinearVelScale=<1 1 1>
    gravityOverride {z=-980}
    physicsBodyDefinitions [{
      boundBone {boneName=dyn_laces_01_r}
      boxExtents=<20 30 30>
      constraintSetup {angularLimitsMax=<5 -4.5 15>, angularLimitsMin=<-5 -30 -15>, angularTarget=<1 0 0>}
      localJointOffset=<4 0 0>
     }]
    simSpaceSettings {maxAngularAcceleration=100, maxAngularVelocity=50, simSpaceAngularAlpha=1}
    sphericalLimits [{limitRadius=5, drivingBone {boneName=foot_r}, sphereLocalOffset=<2.9582 -9.818 0.0022>}]
   }
   n60 = Anim Dynamics - Bone: dyn_laces_02_r(ComponentPose=@n59, Alpha=1, ExternalForce="0,0,0") {
    angularDampingOverride=1
    angularSpringConstant=15000
    bAngularSpring=true
    bOverrideAngularDamping=true
    bUseGravityOverride=true
    bUsePlanarLimit=false
    bUseSphericalLimits=true
    gravityScale=0.35
    simulationSpace=RootRelative
    boundBone {boneName=dyn_laces_02_r}
    componentAppliedLinearAccClamp=<3000 3000 3000>
    componentLinearAccScale=<1 1 1>
    componentLinearVelScale=<1 1 1>
    gravityOverride {z=-980}
    physicsBodyDefinitions [{
      boundBone {boneName=dyn_laces_02_r}
      boxExtents=<20 30 30>
      constraintSetup {angularLimitsMax=<5 -4.5 15>, angularLimitsMin=<-5 -30 -15>, angularTarget=<1 0 0>}
      localJointOffset=<4 0 0>
     }]
    simSpaceSettings {maxAngularAcceleration=100, maxAngularVelocity=50, simSpaceAngularAlpha=1}
    sphericalLimits [{limitRadius=5, drivingBone {boneName=foot_r}, sphereLocalOffset=<2.0814 -8.643 0.0022>}]
   }
   n61 = Anim Dynamics - Bone: dyn_laces_03_r(ComponentPose=@n60, Alpha=1, ExternalForce="0,0,0") {
    angularDampingOverride=1
    angularSpringConstant=15000
    bAngularSpring=true
    bOverrideAngularDamping=true
    bUseGravityOverride=true
    bUsePlanarLimit=false
    bUseSphericalLimits=true
    gravityScale=0.35
    simulationSpace=RootRelative
    boundBone {boneName=dyn_laces_03_r}
    componentAppliedLinearAccClamp=<3000 3000 3000>
    componentLinearAccScale=<1 1 1>
    componentLinearVelScale=<1 1 1>
    gravityOverride {z=-980}
    physicsBodyDefinitions [{
      boundBone {boneName=dyn_laces_03_r}
      boxExtents=<20 30 30>
      constraintSetup {angularLimitsMax=<5 -4.5 15>, angularLimitsMin=<-5 -30 -15>, angularTarget=<1 0 0>}
      localJointOffset=<4 0 0>
     }]
    simSpaceSettings {maxAngularAcceleration=100, maxAngularVelocity=50, simSpaceAngularAlpha=1}
    sphericalLimits [{limitRadius=5.5, drivingBone {boneName=foot_r}, sphereLocalOffset=<3.8078 -11.5956 0.28>}]
   }
   Save cached pose 'DynamicsCache'(Pose=Component To Local(ComponentPose=@n61)) {cacheName=DynamicsCache}
  ]}
 EventGraph {nodes [
   Event BlueprintUpdateAnimation()
   > n1 = Sequence!(execute) -> (then_0, then_1, then_2:Exec, then_3:Exec, then_4:Exec, then_5:Exec, then_6:Exec, then_7:Exec, then_8:Exec)
   > n2 = "Gravity Override State Function (Parameter Asset) (Nativized)"(
    CharacterPartAnimInstance=Self-Reference()
    GameTimeInSeconds=GetGameTimeInSeconds()
    DebugJointName=spine_05
    Parameters=&/M_MED_AnvilAlarmRobin_Paremters
   )
   > Set Gravity Override(Gravity Override=@n2)
   > Set RBAN Settings Out(RBAN Settings Out=@n2.OutRigidBodyAnimNodeParameters)
   n3 = Branch(execute=@n1.then_2, Condition=Get bIsSliding())
   > Set RB Chain Alpha(RB Chain Alpha:double=0.3) -> (Output_Get:double)
   n4 = Get bIsSkyDiving()
   n5 = Get bIsPlayingEmote()
   n6 = Get bIsMoving2D()
   n7 = Branch(
    execute=@n3.else
    Condition=AND Boolean!(A=Get bIsInFrontEnd(), B=NOT Boolean(A=@n6), C:bool=NOT Boolean(A=@n4), D:bool=NOT Boolean(A=@n5)) -> (ReturnValue)
   )
   > Set RB Chain Alpha(RB Chain Alpha=0.8)
   Set RB Chain Alpha(execute=@n7.else, RB Chain Alpha=1)
   n8 = PlayerMovement(execute=@n1.then_3, AnimInstance=Self-Reference(), XDotProductCoefficient=100, YDotProductCoefficient=100, ZDotProductCoefficient=100)
   n9 = BreakVector(InVec=MakeVector(X=@n8, Y=@n8.YMovement))
   n10 = "float > float"(A=@n9, B=50)
   n11 = Get bIsCrouching()
   > Set Right_Crouch(Right_Crouch:double="To Float (Boolean)"(InBool=AND Boolean!(A=@n10, B=@n11) -> (ReturnValue))) -> (Output_Get:double)
   > Set Backward(Backward:bool=AND Boolean(A=@n11, B="float < float"(A=@n9.Y, B=-40))) -> (Output_Get:bool)
   n12 = NOT Boolean(A=Get Backward() -> (Backward:bool))
   n13 = Get bIsMoving2D()
   n14 = Get bIsCrouching()
   > Set CrouchingMoving2D(CrouchingMoving2D:double="To Float (Boolean)"(InBool=AND Boolean!(A=@n13, B=@n14, C:bool=@n12) -> (ReturnValue))) -> (Output_Get:double)
   n15 = PlayerMovement(execute=@n1.then_5, AnimInstance=Self-Reference(), XDotProductCoefficient=100, YDotProductCoefficient=100, ZDotProductCoefficient=100)
   n16 = BreakVector(InVec=MakeVector(X=@n15, Y=@n15.YMovement, Z=@n15.ZMovement))
   > n17 = Branch(Condition=AND Boolean!(A="float > float"(A=@n16.Z), B=Get bIsSkyDiving()) -> (ReturnValue))
   > Set Alpha_isSkidiveUp(Alpha_isSkidiveUp:float=1) -> (Output_Get:float)
   Set Alpha_isSkidiveUp(execute=@n17.else, Alpha_isSkidiveUp=0)
   n18 = NOT Boolean(A=Get bIsMoving2D())
   n19 = Get bIsCrouching()
   Set CrouchingIdle(execute=@n1.then_1, CrouchingIdle:double="To Float (Boolean)"(InBool=AND Boolean(A=@n19, B=@n18))) -> (Output_Get:double)
   n20 = Get bIsCrouching()
   > Set NotCrouching(NotCrouching:double="To Float (Boolean)"(InBool=NOT Boolean(A=@n20))) -> (Output_Get:double)
   n21 = Get bIsSkyDiving()
   > Set OnGround(OnGround:double="To Float (Boolean)"(InBool=NOT Boolean(A=@n21))) -> (Output_Get:double)
   n22 = Get bIsMoving2D()
   n23 = Get bIsSkyDiving()
   n24 = Get bIsPlayingEmote()
   n25 = Branch(
    execute=@n1.then_4
    Condition=AND Boolean!(A=Get bIsInFrontEnd(), B=NOT Boolean(A=@n22), C:bool=NOT Boolean(A=@n23), D:bool=NOT Boolean(A=@n24)) -> (ReturnValue)
   )
   > Set FEIdle(FEIdle:double=0.5) -> (Output_Get:double)
   > Set Simulation Space(Simulation Space:AnimPhysSimSpaceType Enum=World) -> (Output_Get:AnimPhysSimSpaceType Enum)
   > Set FEIdle_Necklace(FEIdle_Necklace:double=0.3) -> (Output_Get:double)
   > Set FEIdleBagginess(FEIdleBagginess:double=0.3) -> (Output_Get:double)
   > Set FEIdleDefault(FEIdleDefault:double=1) -> (Output_Get:double)
   Set FEIdle(execute=@n25.else, FEIdle=1)
   > Set Simulation Space()
   > Set FEIdle_Necklace(FEIdle_Necklace=1)
   > Set FEIdleBagginess(FEIdleBagginess=1)
   > Set FEIdleDefault(FEIdleDefault=0)
   n26 = float * float(A=GetRealTimeSeconds(), B=Get SpeedWave() -> (SpeedWave:double)) {is=title K2Node_PromotableOperator}
   n27 = float * float(A=MultiplyByPi!(Value:double="Sin (Radians)"(A=@n26)) -> (ReturnValue:double), B=Get AmplitudWave() -> (AmplitudWave:double))
   n28 = Get OffsetWave() -> (OffsetWave:double)
   Set RotatorWave(execute=@n1.then_7, RotatorWave:Rotator=MakeRotator(Pitch=float + float(A=@n27, B=@n28))) -> (Output_Get:Rotator)
   n29 = Branch(execute=@n1.then_6, Condition=Get bIsSkydiveFloating())
   > Set SpeedWave(SpeedWave:double=8) -> (Output_Get:double)
   > Set AmplitudWave(AmplitudWave:double=2) -> (Output_Get:double)
   > Set OffsetWave(OffsetWave:double=6) -> (Output_Get:double)
   Set SpeedWave(execute=@n29.else, SpeedWave=0)
   > Set AmplitudWave(AmplitudWave=0)
   > Set OffsetWave(OffsetWave=0)
   n30 = GetAnimInstance(self=GetOwnerSkeletalMeshComponent())
   Set EasternBlocEmote(
    execute=@n1.then_8
    EasternBlocEmote:double="To Float (Boolean)"(InBool=Montage_IsPlaying(self=@n30, Montage=/BRCosmetics/Animation/Game/MainPlayer/Emotes/Eastern_Bloc/Emote_EasternBloc_CMM_M))
   ) -> (Output_Get:double)
   Event OnWeaponChanged()
   > n31 = Cast To FortPlayerAnimInstance(Object=GetAnimInstance(self=GetOwnerSkeletalMeshComponent()))
   > Set DetectedPistol(DetectedPistol:double=0) -> (Output_Get:double)
   > n32 = Switch on EFortWeaponCoreAnimation!(execute, Selection:EFortWeaponCoreAnimation Enum=Get WeaponCoreAnim(self=@n31)) -> (
    Melee:Exec
    Pistol:Exec
    Shotgun:Exec
    PaperBlueprint:Exec
    Rifle:Exec
    MeleeOneHand:Exec
    MachinePistol:Exec
    RocketLauncher:Exec
    GrenadeLauncher:Exec
    GoingCommando:Exec
    AssaultRifle:Exec
    TacticalShotgun:Exec
    SniperRifle:Exec
    TrapPlacement:Exec
    ShoulderLauncher:Exec
    AbilityDecoTool:Exec
    Crossbow:Exec
    C4:Exec
    RemoteControl:Exec
    DualWield:Exec
    AR_BullPup:Exec
    AR_ForwardGrip:Exec
    MedPackPaddles:Exec
    SMG_P90:Exec
    AR_DrumGun:Exec
    Consumable_Small:Exec
    Consumable_Large:Exec
    Balloon:Exec
    MountedTurret:Exec
    CreativeTool:Exec
    ExplosiveBow:Exec
    AshtonIndigo:Exec
    AshtonChicago:Exec
    MeleeDualWield:Exec
    Unarmed:Exec
   )
   Set DetectedPistol(execute=@n32.Pistol, DetectedPistol=1)
  ]}
]
variables [
 Alpha_isSkidiveUp {type=number}
 AmplitudWave {type=number}
 Backward {type=boolean}
 CrouchingIdle {type=number}
 CrouchingMoving2D {type=number}
 DetectedPistol {type=number}
 EasternBlocEmote {type=number}
 FEIdle {type=number}
 FEIdle_Necklace {type=number}
 FEIdleBagginess {type=number}
 FEIdleDefault {type=number}
 Gravity Override {type=Vector}
 NotCrouching {type=number}
 OffsetWave {type=number}
 OnGround {type=number}
 RB Chain Alpha {type=number}
 RBAN Settings Out {type=FortRigidBodyAnimNodeParameters}
 Right_Crouch {type=number}
 RotatorWave {type=Rotator}
 Simulation Space {type=AnimPhysSimSpaceType, default=RootRelative}
 SpeedWave {type=number}
]
