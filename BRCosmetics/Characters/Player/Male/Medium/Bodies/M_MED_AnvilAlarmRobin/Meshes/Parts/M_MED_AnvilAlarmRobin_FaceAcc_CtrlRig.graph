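# Control Rig for the AnvilAlarmRobin character's face accessory part, applying corrective posing to hair and hat-tail bones.
# RB Enable selects between RBAN and non-RBAN setup functions.
# One setup blends pose offsets using head and left-arm spherical pose readers, scaled by smoothed crouch and movement flags.
# The other partially constrains dynamic hair and hat tails to spine_05, adding head-driven side corrections, a skydiving branch for tail weights and a floating offset.
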ControlRigBlueprint /BRCosmetics/Characters/Player/Male/Medium/Bodies/M_MED_AnvilAlarmRobin/Meshes/Parts/M_MED_AnvilAlarmRobin_FaceAcc_CtrlRig
graphs [
 Multiply_2_ContainedGraph {nodes [
   n1 = RigVMFunctionEntryNode()
   n2 = &RigVMFunction_MathDoubleMul(A=@n1.A, B=@n1.B)
   RigVMFunctionReturnNode(Result=&RigVMFunction_MathDoubleMul(B=@n1.D, A=&RigVMFunction_MathDoubleMul(B=@n1.C, A=@n2.Result).Result).Result)
  ]}
 New Function_ContainedGraph {nodes [
   n1 = RigVMFunctionEntryNode()
   RigVMFunction_Sequence = &RigVMFunction_Sequence(ExecuteContext=@n1.ExecuteContext)
   n2 = &RigUnit_SphericalPoseReader(
    ActiveRegionScaleFactors=(PositiveWidth=1,NegativeWidth=1,PositiveHeight=3.5,NegativeHeight=3)
    ActiveRegionSize=0.1
    Debug=(bDrawDebug=true,bDraw2D=False,bDrawLocalAxes=False,DebugScale=25,DebugSegments=20,DebugThickness=0.25)
    DriverAxis=(X=1,Y=-1,Z=0)
    DriverItem=(Type=Bone,Name="head")
    FalloffRegionScaleFactors=(PositiveWidth=0.5,NegativeWidth=1,PositiveHeight=6,NegativeHeight=2)
    FalloffSize=0.4
    FlipHeightScaling=False
    FlipWidthScaling=False
    OptionalParentItem=(Type=Bone,Name="spine_05")
    RotationOffset=(X=-90,Y=0,Z=-15)
    ExecutePin=@RigVMFunction_Sequence.A
   )
   RigVMFunction_MathBoolNot = &RigVMFunction_MathBoolNot(Value=RigVMVariableNode() {variable=Crouching}.Value)
   RigVMFunction_MathBoolToFloat = &RigVMFunction_MathBoolToFloat(Value=@RigVMFunction_MathBoolNot.Result)
   AlphaInterp = &RigVMFunction_AlphaInterp(InterpSpeedDecreasing=10, InterpSpeedIncreasing=10, Scale=1, bClampResult=true, Value=@RigVMFunction_MathBoolToFloat.Result)
   Multiply = &RigVMFunction_MathDoubleMul(B=@AlphaInterp.Result, A=@n2.OutputParam)
   > &RigUnit_ModifyTransforms(
    ItemToModify="((Item=(Type=Bone,Name=\"base_hair_bk_c\"),Transform=(Rotation=(X=0,Y=-0.4263,Z=0,W=0.9046),Translation=(X=0,Y=0,Z=0),Scale3D=(X=1,Y=1,Z=1))),(Item=(Type=Bone,Name=\"base_hair_bk_l\"),Transform=(Rotation=(X=-0.1111,Y=0.3367,Z=0.2287,W=-0.9066),Translation=(X=0 ...(truncated)"
    WeightMinimum=0
    Weight=@Multiply.Result
   )
   n3 = &RigVMFunction_AlphaInterp(
    InterpSpeedDecreasing=10
    InterpSpeedIncreasing=10
    Scale=1
    bClampResult=true
    Value=&RigVMFunction_MathBoolToFloat(Value=RigVMVariableNode() {variable=Crouching}.Value).Result
   )
   n4 = RigVMVariableNode() {variable=Moving2D}
   n5 = &RigVMFunction_AlphaInterp(
    InterpSpeedDecreasing=10
    InterpSpeedIncreasing=10
    Scale=1
    bClampResult=true
    Value=&RigVMFunction_MathBoolToFloat(Value=&RigVMFunction_MathBoolNot(Value=@n4.Value).Result).Result
   )
   Multiply_1 = RigVMAggregateNode(D=0.7, B=@n3.Result, C=@n5.Result, A=@n2.OutputParam)
   > &RigUnit_ModifyTransforms(
    ItemToModify="((Item=(Type=Bone,Name=\"base_hair_bk_c\"),Transform=(Rotation=(X=0,Y=-0.5633,Z=0,W=0.8262),Translation=(X=0,Y=0,Z=0),Scale3D=(X=1,Y=1,Z=1))),(Item=(Type=Bone,Name=\"base_hair_bk_l\"),Transform=(Rotation=(X=0.1704,Y=-0.4904,Z=-0.1823,W=0.835),Translation=(X=0, ...(truncated)"
    WeightMinimum=0
    Weight=@Multiply_1.Result
   )
   > n6 = &RigUnit_SphericalPoseReader(
    ActiveRegionScaleFactors=(PositiveWidth=1,NegativeWidth=1,PositiveHeight=1,NegativeHeight=1)
    ActiveRegionSize=0.1
    Debug=(bDrawDebug=false,bDraw2D=False,bDrawLocalAxes=False,DebugScale=25,DebugSegments=20,DebugThickness=0.25)
    DriverAxis=(X=1,Y=-1,Z=0)
    DriverItem=(Type=Bone,Name="head")
    FalloffRegionScaleFactors=(PositiveWidth=0.5,NegativeWidth=1,PositiveHeight=1,NegativeHeight=1)
    FalloffSize=0.35
    FlipHeightScaling=False
    FlipWidthScaling=False
    OptionalParentItem=(Type=Bone,Name="spine_05")
    RotationOffset=(X=-90,Y=0,Z=-65)
   )
   > &RigUnit_ModifyTransforms(
    ItemToModify="((Item=(Type=Bone,Name=\"base_hair_bk_c\"),Transform=(Rotation=(X=0,Y=-0.3332,Z=0,W=0.9428),Translation=(X=-1.1473,Y=0,Z=-4.1959),Scale3D=(X=1,Y=1,Z=1))),(Item=(Type=Bone,Name=\"base_hair_bk_l\"),Transform=(Rotation=(X=-0.1348,Y=0.2724,Z=-0.0009,W=-0.9527),Tra ...(truncated)"
    WeightMinimum=0
    Weight=@n6.OutputParam
   )
   > n7 = &RigUnit_SphericalPoseReader(
    ActiveRegionScaleFactors=(PositiveWidth=1,NegativeWidth=1,PositiveHeight=1,NegativeHeight=1)
    ActiveRegionSize=0.08
    Debug=(bDrawDebug=false,DebugScale=25,DebugSegments=20,DebugThickness=0.25)
    DriverAxis=(X=1,Y=0,Z=0)
    DriverItem=(Type=Bone,Name="upperarm_l")
    FalloffRegionScaleFactors=(PositiveWidth=2,NegativeWidth=2,PositiveHeight=2,NegativeHeight=2)
    FalloffSize=0.3
    OptionalParentItem=(Type=Bone,Name="spine_04")
    RotationOffset=(X=0,Y=0,Z=0)
   )
   > &RigUnit_ModifyTransforms(
    ItemToModify="((Item=(Type=Bone,Name=\"base_hair_bk_c\"),Transform=(Rotation=(X=0,Y=0,Z=0,W=1),Translation=(X=0,Y=0,Z=0),Scale3D=(X=1,Y=1,Z=1))),(Item=(Type=Bone,Name=\"base_hair_bk_l\"),Transform=(Rotation=(X=0,Y=0,Z=0,W=1),Translation=(X=0,Y=0,Z=0),Scale3D=(X=1,Y=1,Z=1))) ...(truncated)"
    WeightMinimum=0
    Weight=@n7.OutputParam
   )
   n8 = &RigVMFunction_AlphaInterp(
    InterpSpeedDecreasing=10
    InterpSpeedIncreasing=10
    Scale=1
    bClampResult=true
    Value=&RigVMFunction_MathBoolToFloat(Value=RigVMVariableNode() {variable=Crouching}.Value).Result
   )
   > &RigUnit_ModifyTransforms(
    ItemToModify="((Item=(Type=Bone,Name=\"base_hair_bk_c\"),Transform=(Rotation=(X=0,Y=0,Z=0,W=1),Translation=(X=0,Y=0,Z=0),Scale3D=(X=1,Y=1,Z=1))),(Item=(Type=Bone,Name=\"base_hair_bk_l\"),Transform=(Rotation=(X=0,Y=0,Z=0,W=1),Translation=(X=0,Y=0,Z=0),Scale3D=(X=1,Y=1,Z=1))) ...(truncated)"
    WeightMinimum=0
    Weight=&RigVMFunction_MathDoubleMul(B=@n8.Result, A=@n7.OutputParam).Result
   )
   RigVMFunctionReturnNode(ExecuteContext=@RigVMFunction_Sequence.B)
  ]}
 New Function_ContainedGraph {nodes [
   n1 = RigVMFunctionEntryNode()
   RigVMFunction_Sequence = &RigVMFunction_Sequence(ExecuteContext=@n1.ExecuteContext)
   &RigUnit_ParentConstraint(
    AdvancedSettings=(InterpolationType=Average,RotationOrderForFilter=XZY)
    Child=(Type=Bone,Name="dyn_hair_sd_02_l")
    Filter=(TranslationFilter=(bX=True,bY=True,bZ=True),RotationFilter=(bX=True,bY=True,bZ=True),ScaleFilter=(bX=False,bY=False,bZ=False))
    Parents=((Item=(Type=Bone,Name="spine_05"),Weight=1))
    Weight=0.5
    ExecutePin=@RigVMFunction_Sequence.A
   )
   > &RigUnit_ParentConstraint(
    AdvancedSettings=(InterpolationType=Average,RotationOrderForFilter=XZY)
    Child=(Type=Bone,Name="dyn_hair_sd_02_r")
    Filter=(TranslationFilter=(bX=True,bY=True,bZ=True),RotationFilter=(bX=True,bY=True,bZ=True),ScaleFilter=(bX=False,bY=False,bZ=False))
    Parents=((Item=(Type=Bone,Name="spine_05"),Weight=1))
    Weight=0.4
   )
   > &RigUnit_ParentConstraint(
    AdvancedSettings=(InterpolationType=Average,RotationOrderForFilter=XZY)
    Child=(Type=Bone,Name="dyn_hair_bk_02_l")
    Filter=(TranslationFilter=(bX=True,bY=True,bZ=True),RotationFilter=(bX=True,bY=True,bZ=True),ScaleFilter=(bX=False,bY=False,bZ=False))
    Parents=((Item=(Type=Bone,Name="spine_05"),Weight=1))
    Weight=0.7
   )
   > &RigUnit_ParentConstraint(
    AdvancedSettings=(InterpolationType=Average,RotationOrderForFilter=XZY)
    Child=(Type=Bone,Name="dyn_hair_bk_02_c")
    Filter=(TranslationFilter=(bX=True,bY=True,bZ=True),RotationFilter=(bX=True,bY=True,bZ=True),ScaleFilter=(bX=False,bY=False,bZ=False))
    Parents=((Item=(Type=Bone,Name="spine_05"),Weight=1))
    Weight=0.7
   )
   > n2 = &RigUnit_ParentConstraint(
    AdvancedSettings=(InterpolationType=Average,RotationOrderForFilter=XZY)
    Child=(Type=Bone,Name="dyn_hair_bk_02_r")
    Filter=(TranslationFilter=(bX=True,bY=True,bZ=True),RotationFilter=(bX=True,bY=True,bZ=True),ScaleFilter=(bX=False,bY=False,bZ=False))
    Parents=((Item=(Type=Bone,Name="spine_05"),Weight=1))
    Weight=0.7
   )
   RigVMFunction_ControlFlowBranch = &RigVMFunction_ControlFlowBranch(ExecuteContext=@n2.ExecutePin, Condition=RigVMVariableNode() {variable=Skyidiving}.Value)
   n3 = &RigUnit_SphericalPoseReader(
    ActiveRegionScaleFactors=(PositiveWidth=1,NegativeWidth=1,PositiveHeight=2.4,NegativeHeight=0.5)
    ActiveRegionSize=0.07
    Debug=(bDrawDebug=false,bDraw2D=False,bDrawLocalAxes=False,DebugScale=25,DebugSegments=20,DebugThickness=0.25)
    DriverAxis=(X=0,Y=1,Z=0)
    DriverItem=(Type=Bone,Name="head")
    FalloffRegionScaleFactors=(PositiveWidth=1,NegativeWidth=1,PositiveHeight=6,NegativeHeight=1)
    FalloffSize=0.22
    FlipHeightScaling=False
    FlipWidthScaling=False
    OptionalParentItem=(Type=Bone,Name="spine_05")
    RotationOffset=(X=148,Y=0,Z=37)
    ExecutePin=@RigVMFunction_ControlFlowBranch.Completed
   )
   > &RigUnit_ModifyTransforms(
    ItemToModify="((Item=(Type=Bone,Name=\"dyn_hair_sd_02_l\"),Transform=(Rotation=(X=0,Y=0,Z=0,W=-1),Translation=(X=0,Y=-3.2278,Z=1.4931),Scale3D=(X=1,Y=1,Z=1))),(Item=(Type=Bone,Name=\"dyn_hair_bk_02_l\"),Transform=(Rotation=(X=0.2295,Y=0,Z=0,W=0.9733),Translation=(X=0,Y=-4.6 ...(truncated)"
    WeightMinimum=0
    Weight=@n3.OutputParam
   )
   > n4 = &RigUnit_SphericalPoseReader(
    ActiveRegionScaleFactors=(PositiveWidth=1,NegativeWidth=1,PositiveHeight=2.4,NegativeHeight=1.4)
    ActiveRegionSize=0.06
    Debug=(bDrawDebug=false,bDraw2D=False,bDrawLocalAxes=False,DebugScale=25,DebugSegments=20,DebugThickness=0.25)
    DriverAxis=(X=0,Y=1,Z=0)
    DriverItem=(Type=Bone,Name="head")
    FalloffRegionScaleFactors=(PositiveWidth=1,NegativeWidth=1,PositiveHeight=6,NegativeHeight=1.5)
    FalloffSize=0.2
    FlipHeightScaling=true
    FlipWidthScaling=false
    OptionalParentItem=(Type=Bone,Name="spine_05")
    RotationOffset=(X=50,Y=0,Z=30)
   )
   > &RigUnit_ModifyTransforms(
    ItemToModify="((Item=(Type=Bone,Name=\"dyn_hair_sd_02_r\"),Transform=(Rotation=(X=0,Y=0,Z=0,W=1),Translation=(X=0,Y=1.3264,Z=0),Scale3D=(X=1,Y=1,Z=1))),(Item=(Type=Bone,Name=\"dyn_hair_bk_02_r\"),Transform=(Rotation=(X=0,Y=0,Z=0,W=1),Translation=(X=0,Y=3.8059,Z=0),Scale3D=( ...(truncated)"
    WeightMinimum=0
    Weight=@n4.OutputParam
   )
   &RigUnit_ParentConstraint(
    AdvancedSettings=(InterpolationType=Average,RotationOrderForFilter=XZY)
    Child=(Type=Bone,Name="dyn_hat_tail_l_02")
    Parents=((Item=(Type=Bone,Name="spine_05"),Weight=1))
    Weight=0.4
    ExecutePin=@RigVMFunction_ControlFlowBranch.False
   )
   > &RigUnit_ParentConstraint(
    AdvancedSettings=(InterpolationType=Average,RotationOrderForFilter=XZY)
    Child=(Type=Bone,Name="dyn_hat_tail_l_03")
    Parents=((Item=(Type=Bone,Name="spine_05"),Weight=1))
    Weight=0.8
   )
   > &RigUnit_ParentConstraint(
    AdvancedSettings=(InterpolationType=Average,RotationOrderForFilter=XZY)
    Child=(Type=Bone,Name="dyn_hat_tail_r_02")
    Parents=((Item=(Type=Bone,Name="spine_05"),Weight=1))
    Weight=0.4
   )
   > &RigUnit_ParentConstraint(
    AdvancedSettings=(InterpolationType=Average,RotationOrderForFilter=XZY)
    Child=(Type=Bone,Name="dyn_hat_tail_r_03")
    Parents=((Item=(Type=Bone,Name="spine_05"),Weight=1))
    Weight=0.8
   )
   &RigUnit_ParentConstraint(
    AdvancedSettings=(InterpolationType=Average,RotationOrderForFilter=XZY)
    Child=(Type=Bone,Name="dyn_hat_tail_l_02")
    Parents=((Item=(Type=Bone,Name="spine_05"),Weight=1))
    Weight=0.5
    ExecutePin=@RigVMFunction_ControlFlowBranch.True
   )
   > &RigUnit_ParentConstraint(
    AdvancedSettings=(InterpolationType=Average,RotationOrderForFilter=XZY)
    Child=(Type=Bone,Name="dyn_hat_tail_l_03")
    Parents=((Item=(Type=Bone,Name="spine_05"),Weight=1))
    Weight=0.5
   )
   > &RigUnit_ParentConstraint(
    AdvancedSettings=(InterpolationType=Average,RotationOrderForFilter=XZY)
    Child=(Type=Bone,Name="dyn_hat_tail_r_02")
    Parents=((Item=(Type=Bone,Name="spine_05"),Weight=1))
    Weight=0.4
   )
   > &RigUnit_ParentConstraint(
    AdvancedSettings=(InterpolationType=Average,RotationOrderForFilter=XZY)
    Child=(Type=Bone,Name="dyn_hat_tail_r_03")
    Parents=((Item=(Type=Bone,Name="spine_05"),Weight=1))
    Weight=0.8
   )
   RigVMFunction_MathBoolToFloat = &RigVMFunction_MathBoolToFloat(Value=RigVMVariableNode() {variable=Floating}.Value)
   AlphaInterp = &RigVMFunction_AlphaInterp(InterpSpeedDecreasing=10, InterpSpeedIncreasing=10, Scale=1, bClampResult=true, Value=@RigVMFunction_MathBoolToFloat.Result)
   > &RigUnit_ModifyTransforms(
    ItemToModify="((Item=(Type=Bone,Name=\"base_hair_bk_l\"),Transform=(Rotation=(X=0.0434,Y=-0.2367,Z=0.1577,W=0.9577),Translation=(X=0,Y=0,Z=0),Scale3D=(X=0.8595,Y=1,Z=1))),(Item=(Type=Bone,Name=\"base_hair_bk_c\"),Transform=(Rotation=(X=-0.0119,Y=-0.0999,Z=0.0154,W=0.9948),T ...(truncated)"
    WeightMinimum=0
    Weight=@AlphaInterp.Result
   )
   RigVMFunctionReturnNode(ExecuteContext=@RigVMFunction_Sequence.B)
  ]}
 RigVMFunctionLibrary {nodes [
   NO RBAN Setup = RigVMCollapseNode(ExecuteContext=())
   RBAN Setup = RigVMCollapseNode(ExecuteContext=())
  ]}
 RigVMModel {nodes [
   RigUnit_BeginExecution = &RigUnit_BeginExecution()
   RigVMFunction_ControlFlowBranch = &RigVMFunction_ControlFlowBranch(ExecuteContext=@RigUnit_BeginExecution.ExecutePin, Condition=RigVMVariableNode() {variable=RB Enable}.Value)
   NO RBAN Setup = RigVMFunctionReferenceNode(ExecuteContext=@RigVMFunction_ControlFlowBranch.False)
   RBAN Setup = RigVMFunctionReferenceNode(ExecuteContext=@RigVMFunction_ControlFlowBranch.True)
  ]}
]
hierarchy [
 root {loc=<0 0 0>, rot=<0 0 0 1>}
 pelvis {parent=root, loc=<0 -0.7071 93.1191>, rot=<0.0127 0.707 -0.0127 -0.707>}
 spine_01 {parent=pelvis, loc=<6.7494 -0.6388 0>, rot=<0 0 0.034 -0.9994>}
 spine_02 {parent=spine_01, loc=<7.2518 -0.035 0>, rot=<0 0 0.0238 -0.9997>}
 spine_03 {parent=spine_02, loc=<6.0104 0 0>, rot=<0 0 0.0149 -0.9999>}
 spine_04 {parent=spine_03, loc=<8.1024 0 0>, rot=<0 0 -0.0132 -0.9999>}
 spine_05 {parent=spine_04, loc=<10.0146 0 0>, rot=<0 0 -0.0015 -1>}
 neck_01 {parent=spine_05, loc=<11.3138 0.1529 0>, rot=<0 0 -0.2003 -0.9797>}
 neck_02 {parent=neck_01, loc=<4.5627 0 0>, rot=<0 0 0.0504 -0.9987>}
 head {parent=neck_02, loc=<5.3914 0 0>, rot=<0 0 0.115 -0.9934>}
 base_hair_fr_l {parent=head, loc=<5.9598 4.2545 -6.2977>, rot=<0.0041 -0.9952 -0.0973 0.0004>}
 dyn_hair_fr_01_l {parent=base_hair_fr_l, loc=<0 0 0>, rot=<0 0 0 1>}
 dyn_hair_fr_02_l {parent=dyn_hair_fr_01_l, loc=<5.6135 0 0>, rot=<0 0 0 1>}
 base_hair_sd_l {parent=head, loc=<4.0306 -1.2219 -7.2607>, rot=<0.0041 -1 0.0039 0>}
 dyn_hair_sd_01_l {parent=base_hair_sd_l, loc=<0 0 0>, rot=<0 0 0 1>}
 dyn_hair_sd_02_l {parent=dyn_hair_sd_01_l, loc=<5.0705 0 0>, rot=<0 0 0 1>}
 base_hair_fr_r {parent=head, loc=<5.9598 4.2545 6.2977>, rot=<0.9952 0.0041 0.0004 0.0973>}
 dyn_hair_fr_01_r {parent=base_hair_fr_r, loc=<0 0 0>, rot=<0 0 0 -1>}
 dyn_hair_fr_02_r {parent=dyn_hair_fr_01_r, loc=<-5.614 0 0>, rot=<0 0 0 -1>}
 base_hair_sd_r {parent=head, loc=<4.031 -1.2219 7.2607>, rot=<1 0.0041 0 -0.0039>}
 dyn_hair_sd_01_r {parent=base_hair_sd_r, loc=<0 0 0>, rot=<0 0 0 1>}
 dyn_hair_sd_02_r {parent=dyn_hair_sd_01_r, loc=<-5.071 0 0>, rot=<0 0 0 1>}
 head_back {parent=head, loc=<6.3246 -6.1611 0>, rot=<0.0029 -0.7071 0.7071 -0.0029>}
 base_hair_bk_c {parent=head_back, loc=<4.4722 -0.8314 1.3324>, rot=<0 0 0 1>}
 dyn_hair_bk_01_c {parent=base_hair_bk_c, loc=<0 0 0>, rot=<0 0 0 1>}
 dyn_hair_bk_02_c {parent=dyn_hair_bk_01_c, loc=<4.6607 0 0>, rot=<0 0 0 1>}
 base_hair_bk_l {parent=head_back, loc=<3.4045 4.9723 -0.7204>, rot=<0.3452 0 0 -0.9385>}
 dyn_hair_bk_01_l {parent=base_hair_bk_l, loc=<0 0 0>, rot=<0 0 0 1>}
 dyn_hair_bk_02_l {parent=dyn_hair_bk_01_l, loc=<5.6262 0 0>, rot=<0 0 0 1>}
 base_hair_bk_r {parent=head_back, loc=<3.405 -4.9723 -0.7204>, rot=<0 -0.9385 -0.3452 0>}
 dyn_hair_bk_01_r {parent=base_hair_bk_r, loc=<0 0 0>, rot=<0 0 0 1>}
 dyn_hair_bk_02_r {parent=dyn_hair_bk_01_r, loc=<-5.626 0 0>, rot=<0 0 0 1>}
 base_hat_tail_l {parent=head_back, loc=<4.7379 1.8102 2.7618>, rot=<0.7023 0.1403 -0.0272 -0.6974>}
 dyn_hat_tail_l_01 {parent=base_hat_tail_l, loc=<0 0 0>, rot=<0 0 0 1>}
 dyn_hat_tail_l_02 {parent=dyn_hat_tail_l_01, loc=<5 0 0>, rot=<0 0 0.0369 -0.9993>}
 dyn_hat_tail_l_03 {parent=dyn_hat_tail_l_02, loc=<5 0 0>, rot=<0 0 0 1>}
 base_hat_tail_r {parent=head_back, loc=<5.1331 -1.3121 2.4672>, rot=<0.6921 -0.011 0.1226 -0.7112>}
 dyn_hat_tail_r_01 {parent=base_hat_tail_r, loc=<0 0 0>, rot=<0 0 0 1>}
 dyn_hat_tail_r_02 {parent=dyn_hat_tail_r_01, loc=<3.5 0 0>, rot=<0 0 0.0369 -0.9993>}
 dyn_hat_tail_r_03 {parent=dyn_hat_tail_r_02, loc=<3.5 0 0>, rot=<0 0 0 1>}
 clavicle_l {parent=spine_05, loc=<4.3575 -0.1466 -2.3223>, rot=<0.0465 -0.7289 -0.009 -0.6829>}
 upperarm_l {parent=clavicle_l, loc=<15.5111 0 0>, rot=<0.0226 -0.3234 -0.0088 -0.9459>}
 clavicle_extra_r {parent=clavicle_l, loc=<0 0 0>, rot=<0 0 0 1>}
 clavicle_extra_l {parent=clavicle_l, loc=<0 0 0>, rot=<0 0 0 1>}
 clavicle_r {parent=spine_05, loc=<4.3575 -0.1466 2.3223>, rot=<0.7289 0.0465 -0.6829 0.009>}
 upperarm_r {parent=clavicle_r, loc=<-15.5111 0 0>, rot=<0.0226 -0.3234 -0.0088 -0.9459>}
]
variables [
 Crouching {type=bool, default=False}
 Floating {type=bool, default=False}
 HangGliding {type=double, default=0}
 Moving2D {type=bool, default=False}
 RB Enable {type=bool, default=False}
 Skyidiving {type=bool, default=False}
]
