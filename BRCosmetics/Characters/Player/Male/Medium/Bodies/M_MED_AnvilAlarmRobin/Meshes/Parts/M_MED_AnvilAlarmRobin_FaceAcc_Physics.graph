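# Physics asset for the AnvilAlarmRobin face accessory mesh, simulating hair locks and two hat tails.
# Two-link hair chains (back, front, side) and three-link hat tail chains hang from small base anchor bodies via swing-limited constraints, mostly twist-locked.
# Back hair uses box bodies, while the other strands use capsules.
# Unconstrained head, neck, clavicle and upper-spine capsules provide body collision shapes, with the left clavicle on an extra bone.
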PhysicsAsset /BRCosmetics/Characters/Player/Male/Medium/Bodies/M_MED_AnvilAlarmRobin/Meshes/Parts/M_MED_AnvilAlarmRobin_FaceAcc_Physics
bodies {
 base_hair_bk_c [{shapeType=Capsule, length=0.505, radius=0.505, shapeName=base_hair_bk_c_capsule, rotation=<rot 0 -90 0>}]
 base_hair_bk_l [{shapeType=Capsule, length=0.505, radius=0.505, shapeName=base_hair_bk_l_capsule, rotation=<rot 0 -90 0>}]
 base_hair_bk_r [{shapeType=Capsule, length=0.505, radius=0.505, shapeName=base_hair_bk_r_capsule, rotation=<rot 0 -90 0>}]
 base_hair_fr_l [{shapeType=Capsule, length=0.505, radius=0.505, shapeName=base_hair_fr_l_capsule, rotation=<rot 0 -90 0>}]
 base_hair_fr_r [{shapeType=Capsule, length=0.505, radius=0.505, shapeName=base_hair_fr_r_capsule, rotation=<rot 0 -90 0>}]
 base_hair_sd_l [{shapeType=Capsule, length=0.505, radius=0.505, shapeName=base_hair_sd_l_capsule, rotation=<rot 0 -90 0>}]
 base_hair_sd_r [{shapeType=Capsule, length=0.505, radius=0.505, shapeName=base_hair_sd_r_capsule, rotation=<rot 0 -90 0>}]
 base_hat_tail_l [{shapeType=Capsule, length=0.505, radius=0.505, shapeName=base_hat_tail_l_capsule, rotation=<rot 0 -90 0>}]
 base_hat_tail_r [{shapeType=Capsule, length=0.505, radius=0.505, shapeName=base_hat_tail_r_capsule, rotation=<rot 0 -90 0>}]
 clavicle_extra_l [{
   shapeType=Capsule
   length=11.6233
   radius=4.9352
   shapeName=clavicle_l_capsule
   center=<15.8566 -2.9379 -0.4268>
   rotation=<rot 83.1122 -0.0006 179.9994>
  }]
 clavicle_r [{
   shapeType=Capsule
   length=10.6705
   radius=6.7602
   shapeName=clavicle_r_capsule
   center=<-15.7379 0 1.4093>
   rotation=<rot 83.1122 -0.0006 -0.0006>
  }]
 dyn_hair_bk_01_c [{shapeType=Box, extentX=5, extentY=5, extentZ=1.346, shapeName=dyn_hair_bk_01_c_box, center=<3.1134 0 0>}]
 dyn_hair_bk_01_l [{
   shapeType=Box
   extentX=5
   extentY=6.3926
   extentZ=2
   shapeName=dyn_hair_bk_01_l_box
   center=<2.4248 -0.131 -0.0073>
   rotation=<rot -0.0001 0.0001 -3.1777>
  }]
 dyn_hair_bk_01_r [{shapeType=Box, extentX=5, extentY=5.623, extentZ=2, shapeName=dyn_hair_bk_01_r_box, center=<-2.4248 0 -0.4329>, rotation=<rot 0 0 -180>}]
 dyn_hair_bk_02_c [{shapeType=Box, extentX=4.8449, extentY=5, extentZ=1.4672, shapeName=dyn_hair_bk_02_c_box, center=<3.4313 0 0.2064>}]
 dyn_hair_bk_02_l [{
   shapeType=Box
   extentX=3.78
   extentY=7.9565
   extentZ=1.7617
   shapeName=dyn_hair_bk_02_l_box
   center=<2.6158 -0.8309 0.1533>
   rotation=<rot 0 0 -6.37>
  }]
 dyn_hair_bk_02_r [{
   shapeType=Box
   extentX=3.92
   extentY=6.3492
   extentZ=1.4477
   shapeName=dyn_hair_bk_02_r_box
   center=<-2.8345 0.3001 0.0202>
   rotation=<rot 0 0 -180>
  }]
 dyn_hair_fr_01_l [{
   shapeType=Capsule
   length=4.65
   radius=1.35
   shapeName=dyn_hair_fr_01_l_capsule
   center=<2.3482 -0.3418 0.8263>
   rotation=<rot 84.2385 90.0003 0.0002>
  }]
 dyn_hair_fr_01_r [{
   shapeType=Capsule
   length=4.65
   radius=1.35
   shapeName=dyn_hair_fr_01_r_capsule
   center=<-2.3482 0.3418 -0.8263>
   rotation=<rot 84.2385 90.0003 -179.9998>
  }]
 dyn_hair_fr_02_l [{
   shapeType=Capsule
   length=4.2397
   radius=1.7603
   shapeName=dyn_hair_fr_02_l_capsule
   center=<3.9298 -1.0651 0.7823>
   rotation=<rot 84.2385 90.0003 0.0002>
  }]
 dyn_hair_fr_02_r [{
   shapeType=Capsule
   length=5.9885
   radius=1.7603
   shapeName=dyn_hair_fr_02_r_capsule
   center=<-5.0404 1.9163 -1.0436>
   rotation=<rot 84.2385 90.0003 -179.9998>
  }]
 dyn_hair_sd_01_l [{
   shapeType=Capsule
   length=4.55
   radius=1.45
   shapeName=dyn_hair_sd_01_l_capsule
   center=<3.6871 0.0233 0.5761>
   rotation=<rot 81.7592 89.9581 0>
  }]
 dyn_hair_sd_01_r [{
   shapeType=Capsule
   length=4.55
   radius=1.45
   shapeName=dyn_hair_sd_01_r_capsule
   center=<-3.6871 -0.0233 -0.5761>
   rotation=<rot 81.7592 89.9581 180>
  }]
 dyn_hair_sd_02_l [{shapeType=Capsule, length=4.02, radius=1.98, shapeName=dyn_hair_sd_02_l_capsule, center=<6.226 0 0.1726>, rotation=<rot 90 90 0>}]
 dyn_hair_sd_02_r [{shapeType=Capsule, length=4.02, radius=1.98, shapeName=dyn_hair_sd_02_r_capsule, center=<-6.226 0 -0.1726>, rotation=<rot 90 -90 0>}]
 dyn_hat_tail_l_01 [{
   shapeType=Capsule
   length=1.5509
   radius=1.5
   shapeName=dyn_hat_tail_l_01_capsule
   center=<3.0802 -0.8044 0.0006>
   rotation=<rot 0 0.0001 8.9248>
  }]
 dyn_hat_tail_l_02 [
  {shapeType=Capsule, length=4, radius=1.5, shapeName=dyn_hat_tail_l_02_capsule, center=<1.8848 -0.9286 0.1581>, rotation=<rot 0 0.0002 13.005>}
  {
   shapeType=Capsule
   length=3.6178
   radius=1.2406
   shapeName=dyn_hat_tail_l_02_capsule
   center=<5.1932 -0.4458 0.0887>
   rotation=<rot 0 0.0001 11.2586>
  }
 ]
 dyn_hat_tail_l_03 [{
   shapeType=Capsule
   length=4
   radius=1.5
   shapeName=dyn_hat_tail_l_03_capsule
   center=<3.9027 -0.7407 -0.0155>
   rotation=<rot 2.7302 0.0003 8.2856>
  }]
 dyn_hat_tail_r_01 [{shapeType=Capsule, length=1.2243, radius=1.5, shapeName=dyn_hat_tail_r_01_capsule, center=<0.5 -1.0521 0>}]
 dyn_hat_tail_r_02 [{
   shapeType=Capsule
   length=2.8492
   radius=1.5
   shapeName=dyn_hat_tail_r_02_capsule
   center=<1.8848 -0.7717 -0.1742>
   rotation=<rot 0 0.0003 -7.0759>
  }]
 dyn_hat_tail_r_03 [{
   shapeType=Capsule
   length=4
   radius=1.5
   shapeName=dyn_hat_tail_r_03_capsule
   center=<2.2808 -1.2058 -0.028>
   rotation=<rot 0 0.0001 -6.3235>
  }]
 head [{shapeType=Capsule, length=6.052, radius=6.5369, shapeName=head_capsule, center=<2.2093 1.3029 0>, rotation=<rot 90 0 0>}]
 neck_01 [{shapeType=Capsule, length=3.7498, radius=6.2502, shapeName=neck_01_capsule, center=<0.3436 0.942 0>, rotation=<rot 90 -5.6783 0>}]
 neck_02 [{shapeType=Capsule, length=3.7498, radius=6.2502, shapeName=neck_02_capsule, center=<0.7614 -0.1533 0>, rotation=<rot 90 -11.3816 0>}]
 spine_05 [
  {shapeType=Capsule, length=11.9165, radius=15.2, shapeName=spine_05_capsule, center=<-0.8587 2.8293 0>, rotation=<rot 0 -90 0>}
  {shapeType=Capsule, length=9.1654, radius=8.2048, shapeName=spine_05_capsule, center=<9.5662 2.7481 0>}
 ]
}
constraints [
 {bone1Name=dyn_hair_bk_01_c, bone2Name=base_hair_bk_c, swing1LimitDegrees=1, swing2LimitDegrees=20, twistLimitDegrees=5, twistMotion=Locked}
 {bone1Name=dyn_hair_bk_02_c, bone2Name=dyn_hair_bk_01_c, swing1LimitDegrees=5, swing2LimitDegrees=20, twistLimitDegrees=5}
 {bone1Name=dyn_hair_bk_01_l, bone2Name=base_hair_bk_l, swing1LimitDegrees=7, swing2LimitDegrees=20, twistLimitDegrees=5, twistMotion=Locked}
 {bone1Name=dyn_hair_bk_02_l, bone2Name=dyn_hair_bk_01_l, swing1LimitDegrees=7, swing2LimitDegrees=20, twistLimitDegrees=5}
 {bone1Name=dyn_hair_bk_01_r, bone2Name=base_hair_bk_r, swing1LimitDegrees=10, swing2LimitDegrees=20, twistLimitDegrees=5, twistMotion=Locked}
 {bone1Name=dyn_hair_bk_02_r, bone2Name=dyn_hair_bk_01_r, swing1LimitDegrees=10, swing2LimitDegrees=20, twistLimitDegrees=15}
 {bone1Name=dyn_hair_fr_01_l, bone2Name=base_hair_fr_l, swing1LimitDegrees=15, swing2LimitDegrees=20, twistLimitDegrees=45, twistMotion=Locked}
 {
  bone1Name=dyn_hair_fr_02_l
  bone2Name=dyn_hair_fr_01_l
  swing1LimitDegrees=15
  swing2LimitDegrees=20
  twistLimitDegrees=45
  twistMotion=Locked
 }
 {bone1Name=dyn_hair_fr_01_r, bone2Name=base_hair_fr_r, swing1LimitDegrees=15, swing2LimitDegrees=20, twistLimitDegrees=45, twistMotion=Locked}
 {
  bone1Name=dyn_hair_fr_02_r
  bone2Name=dyn_hair_fr_01_r
  swing1LimitDegrees=15
  swing2LimitDegrees=20
  twistLimitDegrees=45
  twistMotion=Locked
 }
 {bone1Name=dyn_hair_sd_01_l, bone2Name=base_hair_sd_l, swing1LimitDegrees=20, swing2LimitDegrees=20, twistLimitDegrees=45, twistMotion=Locked}
 {
  bone1Name=dyn_hair_sd_02_l
  bone2Name=dyn_hair_sd_01_l
  swing1LimitDegrees=20
  swing2LimitDegrees=20
  twistLimitDegrees=45
  twistMotion=Locked
 }
 {bone1Name=dyn_hair_sd_01_r, bone2Name=base_hair_sd_r, swing1LimitDegrees=20, swing2LimitDegrees=20, twistLimitDegrees=45, twistMotion=Locked}
 {
  bone1Name=dyn_hair_sd_02_r
  bone2Name=dyn_hair_sd_01_r
  swing1LimitDegrees=20
  swing2LimitDegrees=20
  twistLimitDegrees=45
  twistMotion=Locked
 }
 {
  bone1Name=dyn_hat_tail_l_01
  bone2Name=base_hat_tail_l
  swing1LimitDegrees=40
  swing2LimitDegrees=5
  twistLimitDegrees=45
  twistMotion=Locked
 }
 {
  bone1Name=dyn_hat_tail_l_02
  bone2Name=dyn_hat_tail_l_01
  swing1LimitDegrees=30
  swing2LimitDegrees=15
  twistLimitDegrees=45
  twistMotion=Locked
 }
 {
  bone1Name=dyn_hat_tail_l_03
  bone2Name=dyn_hat_tail_l_02
  swing1LimitDegrees=30
  swing2LimitDegrees=15
  twistLimitDegrees=45
  twistMotion=Locked
 }
 {
  bone1Name=dyn_hat_tail_r_01
  bone2Name=base_hat_tail_r
  swing1LimitDegrees=25
  swing2LimitDegrees=5
  twistLimitDegrees=45
  twistMotion=Locked
 }
 {
  bone1Name=dyn_hat_tail_r_02
  bone2Name=dyn_hat_tail_r_01
  swing1LimitDegrees=20
  swing2LimitDegrees=15
  twistLimitDegrees=45
  twistMotion=Locked
 }
 {
  bone1Name=dyn_hat_tail_r_03
  bone2Name=dyn_hat_tail_r_02
  swing1LimitDegrees=20
  swing2LimitDegrees=15
  twistLimitDegrees=45
  twistMotion=Locked
 }
]
