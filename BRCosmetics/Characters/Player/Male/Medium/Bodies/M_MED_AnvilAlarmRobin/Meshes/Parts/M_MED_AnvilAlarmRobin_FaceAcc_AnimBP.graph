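# Post-process style animation for the AnvilAlarmRobin face accessory part, driving hair and hat-tail secondary motion.
# The parent mesh pose runs through a Control Rig fed movement states, then additive bone offsets around two alternative RigidBody sims.
# The sims work in base-bone space, relative to head during emotes and neck_02 otherwise.
# The EventGraph stores parameter-asset gravity and rigid-body settings, plus a hang-glider alpha set by parachute events.
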AnimBlueprint /BRCosmetics/Characters/Player/Male/Medium/Bodies/M_MED_AnvilAlarmRobin/Meshes/Parts/M_MED_AnvilAlarmRobin_FaceAcc_AnimBP
target_skeleton=&/M_MED_AnvilAlarmRobin_FaceAcc_Skeleton
graphs [
 AnimGraph {nodes [
   n1 = RigidBody!(
    ComponentPose=Local To Component(LocalPose=Use cached pose 'InitialAdjust'())
    Alpha
    OverrideWorldGravity:Vector="0,0,-980"
    ExternalForce
    SimSpaceSettings:Sim Space Settings Structure=(WorldAlpha=0.7,VelocityScaleZ=1,DampingAlpha=1,MaxLinearVelocity=200,MaxAngularVelocity=20,MaxLinearAcceleration=400,MaxAngularAcceleration=100,ExternalLinearDragV=(X=0,Y=0,Z=0),ExternalLinearVelocity=(X=0,Y=0,Z=0),ExternalAngularVelocity=(X=0,Y=0,Z=0))
   ) -> (Pose) {
    bOverrideWorldGravity=true
    overridePhysicsAsset=&/M_MED_AnvilAlarmRobin_FaceAcc_Physics
    simulationSpace=BaseBoneSpace
    alphaScaleBiasClamp {interpSpeedDecreasing=20, interpSpeedIncreasing=20}
    baseBoneRef {boneName=head}
    componentAppliedLinearAccClamp=<0 0 0>
    overrideWorldGravity {z=-980}
    simSpaceSettings {maxAngularAcceleration=100, maxAngularVelocity=20, maxLinearAcceleration=400, maxLinearVelocity=200, worldAlpha=0.7}
   }
   n2 = Component To Local(ComponentPose=@n1)
   n3 = RigidBody(
    ComponentPose=Local To Component(LocalPose=Use cached pose 'InitialAdjust'())
    OverrideWorldGravity="0,0,-980"
    SimSpaceSettings=(WorldAlpha=0.7,VelocityScaleZ=1,DampingAlpha=1,MaxLinearVelocity=200,MaxAngularVelocity=20,MaxLinearAcceleration=400,MaxAngularAcceleration=100,ExternalLinearDragV=(X=0,Y=0,Z=0),ExternalLinearVelocity=(X=0,Y=0,Z=0),ExternalAngularVelocity=(X=0,Y=0,Z=0))
   ) {
    bOverrideWorldGravity=true
    overridePhysicsAsset=&M_MED_AnvilAlarmRobin_FaceAcc_Physics
    simulationSpace=BaseBoneSpace
    alphaScaleBiasClamp {interpSpeedDecreasing=20, interpSpeedIncreasing=20}
    baseBoneRef {boneName=neck_02}
    componentAppliedLinearAccClamp=<0 0 0>
    overrideWorldGravity {z=-980}
    simSpaceSettings {maxAngularAcceleration=100, maxAngularVelocity=20, maxLinearAcceleration=400, maxLinearVelocity=200, worldAlpha=0.7}
   }
   n4 = Component To Local(ComponentPose=@n3)
   n5 = Get bIsPlayingEmote()
   n6 = "Transform (Modify) Bone - Bone: base_hair_sd_r"!(
    ComponentPose=Local To Component(LocalPose=Blend Poses by bool(BlendPose_0=@n2, BlendPose_1=@n4, bActiveValue=@n5))
    bAlphaBoolEnabled:bool=True
    Translation="0.035,-0.795,-0.006"
    Rotation="-14.234,3.723,-0.691"
    Scale="1,1,1"
   ) -> (Pose) {
    alphaInputType=Bool
    rotationMode=BMM_Additive
    rotationSpace=BCS_BoneSpace
    translationMode=BMM_Additive
    translationSpace=BCS_BoneSpace
    alphaBoolBlend {blendInTime=0.2, blendOutTime=0.2}
    boneToModify {boneName=base_hair_sd_r}
    rotation=<rot -14.234 3.723 -0.691>
    translation=<0.035 -0.7954 -0.006>
   }
   n7 = "Transform (Modify) Bone - Bone: base_hair_fr_r"!(ComponentPose=@n6, bAlphaBoolEnabled:bool=True, Translation="0,0,0", Rotation="-3.157,10.226,0.000", Scale="1,1,1") -> (Pose) {
    alphaInputType=Bool
    rotationMode=BMM_Additive
    rotationSpace=BCS_BoneSpace
    translationMode=BMM_Additive
    translationSpace=BCS_BoneSpace
    alphaBoolBlend {blendInTime=0.2, blendOutTime=0.2}
    boneToModify {boneName=base_hair_fr_r}
    rotation=<rot -3.157 10.226 0>
   }
   n8 = "Transform (Modify) Bone - Bone: dyn_hat_tail_l_02"!(ComponentPose=@n7, bAlphaBoolEnabled:bool=True, Translation="0,0,0", Rotation="0,-15,0", Scale="1,1,1") -> (Pose) {
    alphaInputType=Bool
    rotationMode=BMM_Additive
    rotationSpace=BCS_BoneSpace
    translationMode=BMM_Additive
    translationSpace=BCS_BoneSpace
    alphaBoolBlend {blendInTime=0.2, blendOutTime=0.5}
    boneToModify {boneName=dyn_hat_tail_l_02}
    rotation {yaw=-15}
   }
   Output Pose(Result=Component To Local(ComponentPose=@n8)) {layerGroup=None}
   n9 = "Transform (Modify) Bone - Bone: base_hat_tail_l"!(
    ComponentPose=Local To Component(LocalPose=Use cached pose 'CtrlRig Cache'())
    bAlphaBoolEnabled:bool=True
    Translation="0.103,0.527,-0.000"
    Rotation="0,-30,0"
    Scale="1,1,1"
   ) -> (Pose) {
    alphaInputType=Bool
    rotationMode=BMM_Additive
    rotationSpace=BCS_BoneSpace
    translationMode=BMM_Additive
    translationSpace=BCS_BoneSpace
    alphaBoolBlend {blendInTime=0.2, blendOutTime=0.5}
    boneToModify {boneName=base_hat_tail_l}
    rotation {yaw=-30}
    translation=<0.1031 0.5272 0>
   }
   n10 = "Transform (Modify) Bone - Bone: base_hair_bk_l"!(ComponentPose=@n9, bAlphaBoolEnabled:bool=True, Translation="-0.106,0.092,-0.640", Rotation="-8.392,6.457,-8.857", Scale="1,1,1") -> (Pose) {
    alphaInputType=Bool
    rotationMode=BMM_Additive
    rotationSpace=BCS_BoneSpace
    translationMode=BMM_Additive
    translationSpace=BCS_BoneSpace
    alphaBoolBlend {blendInTime=0.2, blendOutTime=0.2}
    boneToModify {boneName=base_hair_bk_l}
    rotation=<rot -8.392 6.457 -8.857>
    translation=<-0.106 0.092 -0.64>
   }
   n11 = "Transform (Modify) Bone - Bone: base_hair_bk_c"!(ComponentPose=@n10, bAlphaBoolEnabled:bool=True, Translation="0,0,0", Rotation="-10,0,0", Scale="1,1,1") -> (Pose) {
    alphaInputType=Bool
    rotationMode=BMM_Additive
    rotationSpace=BCS_BoneSpace
    translationMode=BMM_Additive
    translationSpace=BCS_BoneSpace
    alphaBoolBlend {blendInTime=0.2, blendOutTime=0.2}
    boneToModify {boneName=base_hair_bk_c}
    rotation {pitch=-10}
   }
   n12 = "Transform (Modify) Bone - Bone: clavicle_extra_l"!(ComponentPose=@n11, bAlphaBoolEnabled:bool=True, Translation="0,0,-10", Rotation="0,0,0", Scale="1,1,1") -> (Pose) {
    alphaInputType=Bool
    rotationMode=BMM_Additive
    rotationSpace=BCS_BoneSpace
    translationMode=BMM_Additive
    translationSpace=BCS_BoneSpace
    alphaBoolBlend {blendInTime=0.2, blendOutTime=0.2}
    boneToModify {boneName=clavicle_extra_l}
    translation {z=-10}
   }
   Save cached pose 'InitialAdjust'(Pose=Component To Local(ComponentPose=@n12)) {cacheName=InitialAdjust}
   n13 = Control Rig!(
    Source=Copy Pose From Mesh() {bUseAttachedParent=true}
    Alpha
    Crouching:bool=Get bIsCrouching()
    Moving2D:bool=Get bIsMoving2D()
    RB Enable:bool=Get bRigidBodyEnabled()
    Skyidiving:bool=Get bIsSkyDiving()
    Floating:bool=Get bIsSkydiveFloating()
   ) -> (Pose) {controlRigAssetReference {blueprintRigClass=&/M_MED_AnvilAlarmRobin_FaceAcc_CtrlRig.M_MED_AnvilAlarmRobin_FaceAcc_CtrlRig_C}}
   Save cached pose 'CtrlRig Cache'(Pose=@n13) {cacheName=CtrlRig Cache}
  ]}
 EventGraph {nodes [
   Event BlueprintUpdateAnimation()
   > n1 = "Gravity Override State Function (Parameter Asset) (Nativized)"(
    CharacterPartAnimInstance=Self-Reference()
    GameTimeInSeconds=GetGameTimeInSeconds()
    DebugJointName=spine_05
    Parameters=&/M_MED_AnvilAlarmRobin_FaceAcc_Parameters
   )
   > Set Gravity Override(Gravity Override=@n1)
   > Set RBAN Settings Out(RBAN Settings Out=@n1.OutRigidBodyAnimNodeParameters)
   Event OnParachuteOpened()
   > Switch on EGliderType(Selection=Get PlayerGliderType() -> (PlayerGliderType=HangGlider))
   > Set HangGlidingAlpha(HangGlidingAlpha:double=1) -> (Output_Get:double)
   Event OnParachuteClosed()
   > Set HangGlidingAlpha(HangGlidingAlpha=0)
  ]}
]
variables [
 Gravity Override {type=Vector}
 HangGlidingAlpha {type=number}
 RBAN Settings Out {type=FortRigidBodyAnimNodeParameters}
]
